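# Creative property editor "More Options" popup, an activatable widget for renaming a device, clearing items or resetting its properties, functions and events.
# A switcher flips between the button list, shown per visibility flags applied on Construct, and a rename page with a filtered text field.
# Clicks fire event dispatchers; Reset All also sends an empty rename.
# Rename fires once confirmation (Save, or Enter gated by a mobile check) and sanitized text are both ready.
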WidgetBlueprint /Game/Athena/UI/InGame/Creative/PropertyEditor/Cr_PropertyEditor_MoreOptions
graphs [
 BP_GetDesiredFocusTarget {nodes [
   Get Desired Focus Target()
   > Return Node!(execute, ReturnValue:Widget Object Reference=Get Button_Rename() -> (Button_Rename:Icon Text Button Object Reference))
  ]}
 BackAction {nodes [BackAction() {is=title K2Node_FunctionEntry}]}
 ClearItemsAction {nodes [ClearItemsAction() {is=title K2Node_FunctionEntry}]}
 DefaultNameAction {nodes [DefaultNameAction() {is=title K2Node_FunctionEntry}]}
 EventGraph {nodes [
   "On Button Clicked (Button_Rename)"!() -> (OutputDelegate, then, Button:Common Button Legacy Object Reference) {is=title K2Node_ComponentBoundEvent}
   > SetActiveWidgetIndex(self=Get Switcher() -> (Switcher:Common Activatable Widget Switcher Object Reference), Index=1)
   > SetFocus(self=Get FilteredTextEntryWidget_Rename() -> (FilteredTextEntryWidget_Rename:WBP UIKit Filtered Input Field Object Reference))
   > SetTextFocus(self:WBP UIKit Filtered Input Field Object Reference=Get FilteredTextEntryWidget_Rename())
   > UpdateCharacterCount(self:Fort Editable Filtered Counted Text Box Object Reference=Get FilteredTextEntryWidget_Rename())
   "On Button Clicked (ClearItemsButton)"!() -> (OutputDelegate, then, Button:Common Button Legacy Object Reference) {is=title K2Node_ComponentBoundEvent}
   > Call Clear Items Action!(execute:Exec, self:Cr Property Editor More Options Object Reference) -> (then:Exec) {is=title K2Node_CallDelegate}
   "On Button Clicked (RestoreDefaultsButton)"!() -> (OutputDelegate, then, Button:Common Button Legacy Object Reference) {is=title K2Node_ComponentBoundEvent}
   > Call Reset Action!(execute:Exec, self:Cr Property Editor More Options Object Reference) -> (then:Exec) {is=title K2Node_CallDelegate}
   > Call Rename Action!(
    execute:Exec
    self:Cr Property Editor More Options Object Reference
    NewName:Text=GetEmptyText!() -> (ReturnValue:Text)
    CommitMethod:ETextCommit Enum=Default
   ) -> (then:Exec) {is=title K2Node_CallDelegate}
   "On Button Clicked (Button_Back)"!() -> (OutputDelegate, then, Button:Common Button Legacy Object Reference) {is=title K2Node_ComponentBoundEvent}
   > Call Back Action!(execute:Exec, self:Cr Property Editor More Options Object Reference) -> (then:Exec) {is=title K2Node_CallDelegate}
   "On Button Clicked (ResetProperties)"!() -> (OutputDelegate, then, Button:Common Button Legacy Object Reference) {is=title K2Node_ComponentBoundEvent}
   > Call Reset Properties Action!(execute:Exec, self:Cr Property Editor More Options Object Reference) -> (then:Exec) {is=title K2Node_CallDelegate}
   "On Button Clicked (Button_CancelReName)"!() -> (OutputDelegate, then, Button:Common Button Legacy Object Reference) {is=title K2Node_ComponentBoundEvent}
   > SetActiveWidgetIndex(self=Get Switcher(), Index=0)
   > SetFocus(self=Get Button_Rename())
   "On Button Clicked (Button_Rename_Done)"!() -> (OutputDelegate, then, Button:Common Button Legacy Object Reference) {is=title K2Node_ComponentBoundEvent}
   > Set WantToRename(WantToRename:bool=true) -> (Output_Get:bool)
   > Save Rename if Ready()
   "On Button Clicked (Button_ResetDefaultName)"!() -> (OutputDelegate, then, Button:Common Button Legacy Object Reference) {is=title K2Node_ComponentBoundEvent}
   > Call Default Name Action!(execute:Exec, self:Cr Property Editor More Options Object Reference) -> (then:Exec) {is=title K2Node_CallDelegate}
   "On Button Clicked (ResetFunctions)"!() -> (OutputDelegate, then, Button:Common Button Legacy Object Reference) {is=title K2Node_ComponentBoundEvent}
   > Call Reset Functions Action!(execute:Exec, self:Cr Property Editor More Options Object Reference) -> (then:Exec) {is=title K2Node_CallDelegate}
   "On Button Clicked (ResetEvents)"!() -> (OutputDelegate, then, Button:Common Button Legacy Object Reference) {is=title K2Node_ComponentBoundEvent}
   > Call Reset Events Action!(execute:Exec, self:Cr Property Editor More Options Object Reference) -> (then:Exec) {is=title K2Node_CallDelegate}
   Save Rename if Ready() {is=title K2Node_CustomEvent}
   n1 = Get FilteredTextEntryWidget_Rename()
   > Branch(Condition=AND Boolean(
     A=Get WantToRename() -> (WantToRename:bool)
     B=Get SanitizedTextReady!(self:WBP UIKit Filtered Input Field Object Reference=@n1) -> (SanitizedTextReady:bool)
    ))
   > n2 = GetCurrentText!(execute, self:WBP UIKit Filtered Input Field Object Reference=Get FilteredTextEntryWidget_Rename()) -> (then, Text:Text)
   > Call Rename Action(NewName=@n2, CommitMethod=Default)
   > Set SanitizedTextReady!(execute, SanitizedTextReady:bool=false, self:WBP UIKit Filtered Input Field Object Reference=Get FilteredTextEntryWidget_Rename()) -> (then, Output_Get:bool)
   > Set WantToRename(WantToRename=false)
   Confirm Rename() {is=title K2Node_CustomEvent}
   > Set WantToRename(WantToRename=true)
   On Sanitized Text Ready!() -> (OutputDelegate, then, Text:Text, BadText:bool=~) {is=title K2Node_CustomEvent}
   > Save Rename if Ready() {is=title K2Node_CallFunction}
   Event On Deactivated()
   > SetActiveWidgetIndex(self=Get Switcher(), Index=0)
   Event OnInitialized()
   > Bind Event to On Text Commit by Enter(self:WBP UIKit Filtered Input Field Object Reference=Get FilteredTextEntryWidget_Rename(), Delegate=Create Event()) {is=title K2Node_AddDelegate}
   > Bind Event to On Sanitized Text Ready(self:WBP UIKit Filtered Input Field Object Reference=Get FilteredTextEntryWidget_Rename(), Delegate=Create Event()) {is=title K2Node_AddDelegate}
   Event Construct()
   > UpdateButtonVisibilities()
   > n3 = Cast To WBP_UIKit_FilteredInputField!(execute, Object=Get FilteredTextEntryWidget_Rename()) -> (then, CastFailed, AsWBP UIKit Filtered Input Field:WBP UIKit Filtered Input Field Object Reference) {is=title K2Node_DynamicCast}
   > Set As Filtered Text Entry Widget(As Filtered Text Entry Widget:WBP UIKit Filtered Input Field Object Reference=@n3) -> (Output_Get:WBP UIKit Filtered Input Field Object Reference)
  ]}
 HandleOnTextCommitByEnter {nodes [
   HandleOnTextCommitByEnter() {is=title K2Node_FunctionEntry}
   > Branch(Condition=IsMobileGame())
   > Confirm Rename() {is=title K2Node_CallFunction}
  ]}
 RenameAction {nodes [RenameAction!() -> (then, NewName:Text, CommitMethod:ETextCommit Enum=Default) {is=title K2Node_FunctionEntry}]}
 ResetAction {nodes [ResetAction() {is=title K2Node_FunctionEntry}]}
 ResetEventsAction {nodes [ResetEventsAction() {is=title K2Node_FunctionEntry}]}
 ResetFunctionsAction {nodes [ResetFunctionsAction() {is=title K2Node_FunctionEntry}]}
 ResetPropertiesAction {nodes [ResetPropertiesAction() {is=title K2Node_FunctionEntry}]}
 SetButtonVisibility {nodes [
   n1 = SetButtonVisibility!() -> (
    then
    bShowClearItemsButton:bool=~
    bShowResetButton:bool=~
    bShowRenameButton:bool=~
    ShowResetFunctionsButton:bool=~
    ShowResetEventsButton:bool=~
    ShowResetPropertiesButton:bool=~
   ) {is=title K2Node_FunctionEntry}
   > Set ShowClearButton(ShowClearButton:bool=@n1) -> (Output_Get:bool)
   > Set ShowResetButton(ShowResetButton:bool=@n1.bShowResetButton) -> (Output_Get:bool)
   > Set ShowRenameButton(ShowRenameButton:bool=@n1.bShowRenameButton) -> (Output_Get:bool)
   > Set ShowResetFunctionsButton(ShowResetFunctionsButton:bool=@n1.ShowResetFunctionsButton) -> (Output_Get:bool)
   > Set ShowResetEventsButton(ShowResetEventsButton:bool=@n1.ShowResetEventsButton) -> (Output_Get:bool)
   > Set ShowResetPropertiesButton(ShowResetPropertiesButton:bool=@n1.ShowResetPropertiesButton) -> (Output_Get:bool)
  ]}
 SetCurrentName {nodes [
   n1 = SetCurrentName!() -> (then, ObjectName:Text) {is=title K2Node_FunctionEntry}
   > SetInputFieldText!(
    execute
    self:WBP UIKit Filtered Input Field Object Reference=Get As Filtered Text Entry Widget() -> (As Filtered Text Entry Widget:WBP UIKit Filtered Input Field Object Reference)
    Text:Text=@n1
   ) -> (then)
  ]}
 UpdateButtonVisibilities {nodes [
   UpdateButtonVisibilities() {is=title K2Node_FunctionEntry}
   > SetVisibility(
    self=Get Button_Rename()
    InVisibility=Select(Option 0=Collapsed, Option 1=SelfHitTestInvisible, Index=Get ShowRenameButton() -> (ShowRenameButton:bool))
   )
   > SetVisibility(
    self=Get ClearItemsButton() -> (ClearItemsButton:Icon Text Button Object Reference)
    InVisibility=Select(Option 0=Collapsed, Option 1=SelfHitTestInvisible, Index=Get ShowClearButton() -> (ShowClearButton:bool))
   )
   > SetVisibility(
    self=Get RestoreDefaultsButton() -> (RestoreDefaultsButton:Icon Text Button Object Reference)
    InVisibility=Select(Option 0=Collapsed, Option 1=SelfHitTestInvisible, Index=Get ShowResetButton() -> (ShowResetButton:bool))
   )
   > SetVisibility(
    self=Get ResetEvents() -> (ResetEvents:Icon Text Button Object Reference)
    InVisibility=Select(Option 0=Collapsed, Option 1=SelfHitTestInvisible, Index=Get ShowResetEventsButton() -> (ShowResetEventsButton:bool))
   )
   > SetVisibility(
    self=Get ResetFunctions() -> (ResetFunctions:Icon Text Button Object Reference)
    InVisibility=Select(Option 0=Collapsed, Option 1=SelfHitTestInvisible, Index=Get ShowResetFunctionsButton() -> (ShowResetFunctionsButton:bool))
   )
   > SetVisibility(
    self=Get ResetProperties() -> (ResetProperties:Icon Text Button Object Reference)
    InVisibility=Select(Option 0=Collapsed, Option 1=SelfHitTestInvisible, Index=Get ShowResetPropertiesButton() -> (ShowResetPropertiesButton:bool))
   )
  ]}
]
variables [
 As Filtered Text Entry Widget {type=WBP_UIKit_FilteredInputField_C, default=null}
 BackAction {}
 ClearItemsAction {}
 DefaultNameAction {}
 RenameAction {}
 ResetAction {}
 ResetEventsAction {}
 ResetFunctionsAction {}
 ResetPropertiesAction {}
 ShowClearButton {type=boolean}
 ShowRenameButton {type=boolean}
 ShowResetButton {type=boolean}
 ShowResetEventsButton {type=boolean}
 ShowResetFunctionsButton {type=boolean}
 ShowResetPropertiesButton {type=boolean}
 WantToRename {type=boolean}
]
widgets {
 info {inheritedWidgetCount=0, parentClass=&CommonActivatableWidget, rootWidgetClass=&Overlay, widgetCount=28}
 widgets [
  Overlay_0 {widgetClassPath=&Overlay, widget=Overlay_0}
  CommonBorder_0 {
   widgetClassPath=&CommonBorder
   parent=Overlay_0
   widget=CommonBorder_0
   properties {padding=<ltrb 0 0 0 0>}
   slot {class=OverlaySlot, properties {horizontalAlignment=HAlign_Fill, verticalAlignment=VAlign_Fill}}
  }
  Overlay_1 {widgetClassPath=&Overlay, parent=CommonBorder_0, widget=Overlay_1, slot {class=BorderSlot, properties {padding=<ltrb 0 0 0 0>}}}
  Image_96 {
   widgetClassPath=&Image
   bIsVariable=true
   parent=Overlay_1
   widget=Image_96
   properties {
    brush {
     bIsDynamicallyLoaded=false
     drawAs=Image
     imageType=NoImage
     mirroring=NoMirror
     resourceName=None
     resourceObject=/Game/Playgrounds/Items/Portals/UI/Art/M_SimpleSkewBox_Cr_PortalCalloutBacking
     tiling=NoTile
     imageSize=<32 32>
     margin=<ltrb 0 0 0 0>
     outlineSettings {
      bUseBrushTransparency=false
      roundingType=HalfHeightRadius
      width=0
      color {colorUseRule=UseColor_Specified, specifiedColor=<rgba 0 0 0 0>}
      cornerRadii=<0 0 0 0>
     }
     tintColor {colorUseRule=UseColor_Specified, specifiedColor=<rgba 1 1 1 1>}
     uVRegion {bIsValid=false, max=<0 0>, min=<0 0>}
    }
    colorAndOpacity=<rgba 0 0.1221 0.7835 1>
    renderTransform {angle=0, scale=<1.3 1.3>, shear=<0 0>, translation=<0 0>}
   }
   slot {class=OverlaySlot, properties {horizontalAlignment=HAlign_Fill, verticalAlignment=VAlign_Fill}}
  }
  Image_2 {
   widgetClassPath=&Image
   bIsVariable=true
   parent=Overlay_1
   widget=Image_2
   properties {
    brush {
     bIsDynamicallyLoaded=false
     drawAs=Image
     imageType=NoImage
     mirroring=NoMirror
     resourceName=None
     resourceObject=&M_SimpleSkewBox_Cr_PortalCalloutBacking
     tiling=NoTile
     imageSize=<32 32>
     margin=<ltrb 0 0 0 0>
     outlineSettings {
      bUseBrushTransparency=false
      roundingType=HalfHeightRadius
      width=0
      color {colorUseRule=UseColor_Specified, specifiedColor=<rgba 0 0 0 0>}
      cornerRadii=<0 0 0 0>
     }
     tintColor {colorUseRule=UseColor_Specified, specifiedColor=<rgba 1 1 1 1>}
     uVRegion {bIsValid=false, max=<0 0>, min=<0 0>}
    }
    colorAndOpacity=<rgba 0 0.0111 0.1146 1>
    renderTransform {angle=0, scale=<1.3 1.3>, shear=<0 0>, translation=<0 0>}
   }
   slot {class=OverlaySlot, properties {horizontalAlignment=HAlign_Fill, verticalAlignment=VAlign_Fill, padding=<ltrb 12 12 12 12>}}
  }
  SizeBox_0 {
   widgetClassPath=&SizeBox
   parent=Overlay_0
   widget=SizeBox_0
   properties {bOverride_MinDesiredHeight=true, bOverride_MinDesiredWidth=true, minDesiredHeight=356, minDesiredWidth=556, widthOverride=507.4627}
   slot {class=OverlaySlot, properties {horizontalAlignment=HAlign_Center, verticalAlignment=VAlign_Center}}
  }
  Switcher {
   widgetClassPath=&CommonActivatableWidgetSwitcher
   bIsVariable=true
   parent=SizeBox_0
   widget=Switcher
   properties {transitionDuration=0}
   slot {class=SizeBoxSlot}
  }
  Overlay_2 {
   widgetClassPath=&Overlay
   parent=Switcher
   widget=Overlay_2
   slot {class=WidgetSwitcherSlot, properties {padding=<ltrb 64 64 64 64>}}
  }
  VBox_Root {
   widgetClassPath=&VerticalBox
   parent=Overlay_2
   widget=VBox_Root
   properties {navigation=&self:WidgetTree.VBox_Root.WidgetNavigation_0}
   slot {class=OverlaySlot, properties {horizontalAlignment=HAlign_Fill}}
  }
  Button_Rename {
   widgetClassPath=&IconTextButton
   bIsVariable=true
   parent=VBox_Root
   widget=Button_Rename
   properties {
    buttonText=RENAME
    minHeight=72
    minWidth=200
    onButtonBaseClicked=(&self:WidgetTree.Button_Rename.HandleOnButtonClicked)
    onButtonBaseDoubleClicked=(&self:WidgetTree.Button_Rename.HandleOnButtonDoubleClicked)
    onButtonBaseHovered=(&self:WidgetTree.Button_Rename.HandleOnButtonHovered)
    onButtonBaseUnhovered=(&self:WidgetTree.Button_Rename.HandleOnButtonUnhovered)
    onSelectedChangedBase=(&self:WidgetTree.Button_Rename.HandleOnSelectedChanged)
    style=/FortUILibrary/Foundation/Buttons/ButtonStyles/Athena/Athena_ButtonStyle_AngledBlueMenuButton.Athena_ButtonStyle_AngledBlueMenuButton_C
   }
   slot {class=VerticalBoxSlot, properties {padding=<ltrb 0 0 0 6>}}
  }
  ClearItemsButton {
   widgetClassPath=&IconTextButton
   bIsVariable=true
   parent=VBox_Root
   widget=ClearItemsButton
   properties {
    buttonText=CLEAR ITEMS
    minHeight=72
    minWidth=200
    onButtonBaseClicked=(&self:WidgetTree.ClearItemsButton.HandleOnButtonClicked)
    onButtonBaseDoubleClicked=(&self:WidgetTree.ClearItemsButton.HandleOnButtonDoubleClicked)
    onButtonBaseHovered=(&self:WidgetTree.ClearItemsButton.HandleOnButtonHovered)
    onButtonBaseUnhovered=(&self:WidgetTree.ClearItemsButton.HandleOnButtonUnhovered)
    onSelectedChangedBase=(&self:WidgetTree.ClearItemsButton.HandleOnSelectedChanged)
    style=&Athena_ButtonStyle_AngledBlueMenuButton_C
    visibility=Collapsed
   }
   slot {class=VerticalBoxSlot, properties {padding=<ltrb 0 0 0 6>}}
  }
  RestoreDefaultsButton {
   widgetClassPath=&IconTextButton
   bIsVariable=true
   parent=VBox_Root
   widget=RestoreDefaultsButton
   properties {
    buttonText=RESET ALL
    minHeight=72
    minWidth=200
    onButtonBaseClicked=(&self:WidgetTree.RestoreDefaultsButton.HandleOnButtonClicked)
    onButtonBaseDoubleClicked=(&self:WidgetTree.RestoreDefaultsButton.HandleOnButtonDoubleClicked)
    onButtonBaseHovered=(&self:WidgetTree.RestoreDefaultsButton.HandleOnButtonHovered)
    onButtonBaseUnhovered=(&self:WidgetTree.RestoreDefaultsButton.HandleOnButtonUnhovered)
    onSelectedChangedBase=(&self:WidgetTree.RestoreDefaultsButton.HandleOnSelectedChanged)
    style=&Athena_ButtonStyle_AngledBlueMenuButton_C
   }
   slot {class=VerticalBoxSlot, properties {padding=<ltrb 0 0 0 6>}}
  }
  ResetProperties {
   widgetClassPath=&IconTextButton
   bIsVariable=true
   parent=VBox_Root
   widget=ResetProperties
   properties {
    buttonText=RESET PROPERTIES
    minHeight=72
    minWidth=200
    onButtonBaseClicked=(&self:WidgetTree.ResetProperties.HandleOnButtonClicked)
    onButtonBaseDoubleClicked=(&self:WidgetTree.ResetProperties.HandleOnButtonDoubleClicked)
    onButtonBaseHovered=(&self:WidgetTree.ResetProperties.HandleOnButtonHovered)
    onButtonBaseUnhovered=(&self:WidgetTree.ResetProperties.HandleOnButtonUnhovered)
    onSelectedChangedBase=(&self:WidgetTree.ResetProperties.HandleOnSelectedChanged)
    style=&Athena_ButtonStyle_AngledBlueMenuButton_C
   }
   slot {class=VerticalBoxSlot, properties {padding=<ltrb 0 0 0 6>}}
  }
  ResetFunctions {
   widgetClassPath=&IconTextButton
   bIsVariable=true
   parent=VBox_Root
   widget=ResetFunctions
   properties {
    buttonText=RESET FUNCTIONS
    minHeight=72
    minWidth=200
    onButtonBaseClicked=(&self:WidgetTree.ResetFunctions.HandleOnButtonClicked)
    onButtonBaseDoubleClicked=(&self:WidgetTree.ResetFunctions.HandleOnButtonDoubleClicked)
    onButtonBaseHovered=(&self:WidgetTree.ResetFunctions.HandleOnButtonHovered)
    onButtonBaseUnhovered=(&self:WidgetTree.ResetFunctions.HandleOnButtonUnhovered)
    onSelectedChangedBase=(&self:WidgetTree.ResetFunctions.HandleOnSelectedChanged)
    style=&Athena_ButtonStyle_AngledBlueMenuButton_C
   }
   slot {class=VerticalBoxSlot, properties {padding=<ltrb 0 0 0 6>}}
  }
  ResetEvents {
   widgetClassPath=&IconTextButton
   bIsVariable=true
   parent=VBox_Root
   widget=ResetEvents
   properties {
    buttonText=RESET EVENTS
    minHeight=72
    minWidth=200
    onButtonBaseClicked=(&self:WidgetTree.ResetEvents.HandleOnButtonClicked)
    onButtonBaseDoubleClicked=(&self:WidgetTree.ResetEvents.HandleOnButtonDoubleClicked)
    onButtonBaseHovered=(&self:WidgetTree.ResetEvents.HandleOnButtonHovered)
    onButtonBaseUnhovered=(&self:WidgetTree.ResetEvents.HandleOnButtonUnhovered)
    onSelectedChangedBase=(&self:WidgetTree.ResetEvents.HandleOnSelectedChanged)
    style=&Athena_ButtonStyle_AngledBlueMenuButton_C
   }
   slot {class=VerticalBoxSlot, properties {padding=<ltrb 0 0 0 6>}}
  }
  Button_Back {
   widgetClassPath=&IconTextButton
   bIsVariable=true
   parent=VBox_Root
   widget=Button_Back
   properties {
    buttonText=BACK
    minHeight=72
    minWidth=200
    onButtonBaseClicked=(&self:WidgetTree.Button_Back.HandleOnButtonClicked)
    onButtonBaseDoubleClicked=(&self:WidgetTree.Button_Back.HandleOnButtonDoubleClicked)
    onButtonBaseHovered=(&self:WidgetTree.Button_Back.HandleOnButtonHovered)
    onButtonBaseUnhovered=(&self:WidgetTree.Button_Back.HandleOnButtonUnhovered)
    onSelectedChangedBase=(&self:WidgetTree.Button_Back.HandleOnSelectedChanged)
    style=&Athena_ButtonStyle_AngledBlueMenuButton_C
    triggeringInputAction {dataTable=/FortUILibrary/Input/CreativeInputActionDataTable, rowName=Generic_Back}
   }
   slot {class=VerticalBoxSlot, properties {verticalAlignment=VAlign_Bottom, size {sizeRule=Fill, value=1}}}
  }
  Overlay_3 {widgetClassPath=&Overlay, parent=Switcher, widget=Overlay_3, slot {class=WidgetSwitcherSlot}}
  Image {
   widgetClassPath=&Image
   bIsVariable=true
   parent=Overlay_3
   widget=Image
   properties {
    brush {
     bIsDynamicallyLoaded=false
     drawAs=NoDrawType
     imageType=NoImage
     mirroring=NoMirror
     resourceName=None
     resourceObject=&M_SimpleSkewBox_Cr_PortalCalloutBacking
     tiling=NoTile
     imageSize=<32 302.1818>
     margin=<ltrb 0 0 0 0>
     outlineSettings {
      bUseBrushTransparency=false
      roundingType=HalfHeightRadius
      width=0
      color {colorUseRule=UseColor_Specified, specifiedColor=<rgba 0 0 0 0>}
      cornerRadii=<0 0 0 0>
     }
     tintColor {colorUseRule=UseColor_Specified, specifiedColor=<rgba 1 1 1 1>}
     uVRegion {bIsValid=false, max=<0 0>, min=<0 0>}
    }
    colorAndOpacity=<rgba 0 0.1221 0.7835 0.2701>
    renderTransform {angle=-0.3, scale=<1.3 1.3>, shear=<0 0>, translation=<0 0>}
   }
   slot {class=OverlaySlot, properties {horizontalAlignment=HAlign_Fill, verticalAlignment=VAlign_Bottom, padding=<ltrb 6 8 6 6>}}
  }
  VBox_Rename {
   widgetClassPath=&VerticalBox
   parent=Overlay_3
   widget=VBox_Rename
   properties {navigation=&self:WidgetTree.VBox_Rename.WidgetNavigation_0}
   slot {class=OverlaySlot, properties {horizontalAlignment=HAlign_Fill, padding=<ltrb 64 42 64 64>}}
  }
  CommonTextBlock_174 {
   widgetClassPath=&CommonTextBlock
   parent=VBox_Rename
   widget=CommonTextBlock_174
   properties {
    justification=Center
    style=/FortUILibrary/Styles/Typography/Foundation/TextStyle_BurbankSmall_L_White.TextStyle_BurbankSmall_L_White_C
    text=Rename Device
    textTransformPolicy=ToUpper
    font {
     bForceMonospaced=false
     bMaterialIsStencil=false
     fontMaterial=null
     fontObject=&BurbankSmall
     letterSpacing=0
     monospacedWidth=1
     size=36
     skewAmount=0
     typefaceFontName=Black
     outlineSettings {
      bApplyOutlineToDropShadows=false
      bMiteredCorners=false
      bSeparateFillAlpha=false
      outlineMaterial=null
      outlineSize=0
      outlineColor=<rgba 0 0 0 1>
     }
    }
    shadowOffset=<0 0>
   }
   slot {class=VerticalBoxSlot}
  }
  Overlay {widgetClassPath=&Overlay, parent=VBox_Rename, widget=Overlay, slot {class=VerticalBoxSlot, properties {padding=<ltrb 0 5 0 5>}}}
  CommonBorder_2 {
   widgetClassPath=&CommonBorder
   parent=Overlay
   widget=CommonBorder_2
   properties {
    style=/FortUILibrary/Styles/Border/Border-SolidBG_White.Border-SolidBG_White_C
    background {
     bIsDynamicallyLoaded=false
     drawAs=Box
     imageType=NoImage
     mirroring=NoMirror
     resourceName=None
     resourceObject=/FortUILibrary/Styles/Border/Materials/M-Border-TabM
     tiling=NoTile
     imageSize=<32 40>
     margin=<ltrb 0.0708 0 0.9027 0.0708>
     outlineSettings {
      bUseBrushTransparency=false
      roundingType=HalfHeightRadius
      width=0
      color {colorUseRule=UseColor_Specified, specifiedColor=<rgba 0 0 0 0>}
      cornerRadii=<0 0 0 0>
     }
     tintColor {colorUseRule=UseColor_Specified, specifiedColor=<rgba 1 1 1 1>}
     uVRegion {bIsValid=false, max=<0 0>, min=<0 0>}
    }
    brushColor=<rgba 0.651 0.651 0.651 1>
    padding=<ltrb 0 0 0 0>
   }
   slot {class=OverlaySlot, properties {horizontalAlignment=HAlign_Fill, verticalAlignment=VAlign_Fill}}
  }
  SizeBox_1 {
   widgetClassPath=&SizeBox
   parent=VBox_Rename
   widget=SizeBox_1
   properties {bOverride_MinDesiredHeight=true, bOverride_WidthOverride=true, minDesiredHeight=124, widthOverride=700}
   slot {class=VerticalBoxSlot, properties {horizontalAlignment=HAlign_Center, padding=<ltrb 0 12 0 12>}}
  }
  FilteredTextEntryWidget_Rename {
   widgetClassPath=/FortUILibrary/UIKit/InputField/WBP_UIKit_FilteredInputField
   bIsVariable=true
   parent=SizeBox_1
   widget=FilteredTextEntryWidget_Rename
   properties {bIsFocusable=true, entryCharLimit=60, justification=Center, renderOpacity=2, suggestionText=Device name}
   slot {class=SizeBoxSlot}
  }
  Spacer_142 {widgetClassPath=&Spacer, parent=VBox_Rename, widget=Spacer_142, properties {size=<1 12>}, slot {class=VerticalBoxSlot}}
  Button_Rename_Done {
   widgetClassPath=&IconTextButton
   bIsVariable=true
   parent=VBox_Rename
   widget=Button_Rename_Done
   properties {
    buttonText=Save
    minHeight=72
    minWidth=500
    onButtonBaseClicked=(&self:WidgetTree.Button_Rename_Done.HandleOnButtonClicked)
    onButtonBaseDoubleClicked=(&self:WidgetTree.Button_Rename_Done.HandleOnButtonDoubleClicked)
    onButtonBaseHovered=(&self:WidgetTree.Button_Rename_Done.HandleOnButtonHovered)
    onButtonBaseUnhovered=(&self:WidgetTree.Button_Rename_Done.HandleOnButtonUnhovered)
    onSelectedChangedBase=(&self:WidgetTree.Button_Rename_Done.HandleOnSelectedChanged)
    style=/FortUILibrary/Foundation/Buttons/ButtonStyles/Athena/Athena_ButtonStyle_AngledYellowMenuButton.Athena_ButtonStyle_AngledYellowMenuButton_C
   }
   slot {class=VerticalBoxSlot, properties {horizontalAlignment=HAlign_Center, padding=<ltrb 0 0 0 6>}}
  }
  Button_ResetDefaultName {
   widgetClassPath=&IconTextButton
   bIsVariable=true
   parent=VBox_Rename
   widget=Button_ResetDefaultName
   properties {
    buttonText=Reset to Default
    minHeight=72
    minWidth=500
    onButtonBaseClicked=(&self:WidgetTree.Button_ResetDefaultName.HandleOnButtonClicked)
    onButtonBaseDoubleClicked=(&self:WidgetTree.Button_ResetDefaultName.HandleOnButtonDoubleClicked)
    onButtonBaseHovered=(&self:WidgetTree.Button_ResetDefaultName.HandleOnButtonHovered)
    onButtonBaseUnhovered=(&self:WidgetTree.Button_ResetDefaultName.HandleOnButtonUnhovered)
    onSelectedChangedBase=(&self:WidgetTree.Button_ResetDefaultName.HandleOnSelectedChanged)
    style=&Athena_ButtonStyle_AngledBlueMenuButton_C
   }
   slot {class=VerticalBoxSlot, properties {horizontalAlignment=HAlign_Center, padding=<ltrb 0 0 0 6>}}
  }
  Button_CancelReName {
   widgetClassPath=&IconTextButton
   bIsVariable=true
   parent=VBox_Rename
   widget=Button_CancelReName
   properties {
    buttonText=Cancel
    minHeight=72
    minWidth=500
    onButtonBaseClicked=(&self:WidgetTree.Button_CancelReName.HandleOnButtonClicked)
    onButtonBaseDoubleClicked=(&self:WidgetTree.Button_CancelReName.HandleOnButtonDoubleClicked)
    onButtonBaseHovered=(&self:WidgetTree.Button_CancelReName.HandleOnButtonHovered)
    onButtonBaseUnhovered=(&self:WidgetTree.Button_CancelReName.HandleOnButtonUnhovered)
    onSelectedChangedBase=(&self:WidgetTree.Button_CancelReName.HandleOnSelectedChanged)
    style=&Athena_ButtonStyle_AngledBlueMenuButton_C
   }
   slot {class=VerticalBoxSlot, properties {horizontalAlignment=HAlign_Center}}
  }
 ]
}
