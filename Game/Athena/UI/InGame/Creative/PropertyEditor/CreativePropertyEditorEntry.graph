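# Creative property editor row, a FortMatchmakingKnobsSpinnerButton widget showing a setting name, asterisked when modified, and its value.
# Hover swaps the read-only value for an arrow-button option picker and the name for a scrolling copy.
# Arrows slide-fade the option out, shift it when the animation finishes, then reverse; clicking the row shifts right.
# Rotation broadcasts the index and refreshes the read-only value, which prefers non-empty text-field input over the option text.
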WidgetBlueprint /Game/Athena/UI/InGame/Creative/PropertyEditor/CreativePropertyEditorEntry
animations [
 Left {
  binding_count=1
  display_label=""
  length_seconds=0.1
  object=&self:Left
  sequence {bindings [MyText {
     id=#1
     possessed_class=CommonTextBlock
     tracks [
      {
       type=MovieScene2DTransformTrack
       sections [{
         type=MovieScene2DTransformSection
         end_frame=2
         start_frame=0
         channels [
          Translation.X {type=MovieSceneScriptingFloatChannel, keys [[0,0],[2,-100]]}
          Translation.Y {type=MovieSceneScriptingFloatChannel, const [0,1]}
          Angle {type=MovieSceneScriptingFloatChannel, const [0,1]}
          Scale.X {type=MovieSceneScriptingFloatChannel, const [1,1]}
          Scale.Y {type=MovieSceneScriptingFloatChannel, const [1,1]}
          Shear.X {type=MovieSceneScriptingFloatChannel, const [0,1]}
          Shear.Y {type=MovieSceneScriptingFloatChannel, const [0,1]}
         ]
        }]
      }
      {
       type=MovieSceneFloatTrack
       display_name=Render Opacity
       sections [{type=MovieSceneFloatSection, end_frame=2, start_frame=0, channels [None {type=MovieSceneScriptingFloatChannel, keys [[0,1],[2,0]]}]}]
      }
     ]
    }]}
 }
 Right {
  binding_count=1
  display_label=""
  length_seconds=0.1
  object=&self:Right
  sequence {bindings [MyText {
     id=#1
     possessed_class=CommonTextBlock
     tracks [
      {
       type=MovieScene2DTransformTrack
       sections [{
         type=MovieScene2DTransformSection
         end_frame=2
         start_frame=0
         channels [
          Translation.X {type=MovieSceneScriptingFloatChannel, keys [[0,0],[2,100]]}
          Translation.Y {type=MovieSceneScriptingFloatChannel, const [0,1]}
          Angle {type=MovieSceneScriptingFloatChannel, const [0,1]}
          Scale.X {type=MovieSceneScriptingFloatChannel, const [1,1]}
          Scale.Y {type=MovieSceneScriptingFloatChannel, const [1,1]}
          Shear.X {type=MovieSceneScriptingFloatChannel, const [0,1]}
          Shear.Y {type=MovieSceneScriptingFloatChannel, const [0,1]}
         ]
        }]
      }
      {
       type=MovieSceneFloatTrack
       display_name=Render Opacity
       sections [{type=MovieSceneFloatSection, end_frame=2, start_frame=0, channels [None {type=MovieSceneScriptingFloatChannel, keys [[0,1],[2,0]]}]}]
      }
     ]
    }]}
 }
 In {
  binding_count=1
  length_seconds=0.35
  object=&self:In
  sequence {bindings [Border_Fill {
     id=#2
     possessed_class=Border
     tracks [
      {
       type=MovieScene2DTransformTrack
       sections [{
         type=MovieScene2DTransformSection
         end_frame=7
         start_frame=0
         channels [
          Translation.X {type=MovieSceneScriptingFloatChannel, keys [[0,300],[5,-25],[7,0]]}
          Translation.Y {type=MovieSceneScriptingFloatChannel}
          Angle {type=MovieSceneScriptingFloatChannel}
          Scale.X {type=MovieSceneScriptingFloatChannel}
          Scale.Y {type=MovieSceneScriptingFloatChannel}
          Shear.X {type=MovieSceneScriptingFloatChannel}
          Shear.Y {type=MovieSceneScriptingFloatChannel}
         ]
        }]
      }
      {
       type=MovieSceneFloatTrack
       display_name=Render Opacity
       sections [{type=MovieSceneFloatSection, end_frame=7, start_frame=0, channels [None {type=MovieSceneScriptingFloatChannel, keys [[0,0],[7,1]]}]}]
      }
     ]
    }]}
 }
 In2 {
  binding_count=0
  length_seconds=5
  object=&self:In2
  sequence {bindings []}
 }
]
graphs [
 DoCustomNavigation_0 {nodes [
   DoCustomNavigation_0!() -> (then, Navigation:EUINavigation Enum) {is=title K2Node_FunctionEntry}
   > ShiftTextRight(self:Common Rotator Object Reference)
   > PlayAnimationForward(InAnimation=Get Right() -> (Right:Widget Animation Object Reference))
   > Return Node!(execute, ReturnValue:Widget Object Reference=Self-Reference())
   ShiftTextLeft(self:Common Rotator Object Reference)
   Get Left() -> (Left:Widget Animation Object Reference)
  ]}
 EventGraph {nodes [
   n1 = "On Button Clicked (LeftButton)"!() -> (OutputDelegate, then, Button:Common Button Legacy Object Reference) {is=title K2Node_ComponentBoundEvent}
   n2 = AfterleftisDone() {is=title K2Node_CustomEvent}
   n3 = Get Left()
   BindToAnimationFinished!(execute=@n1.then, self:Widget Animation Object Reference=@n3, Widget:User Widget Object Reference=Self-Reference(), Delegate=@n2) -> (then)
   > PlayAnimation(InAnimation=@n3)
   n4 = "On Button Clicked (RightButton)"!() -> (OutputDelegate, then, Button:Common Button Legacy Object Reference) {is=title K2Node_ComponentBoundEvent}
   n5 = AfterRightIsDone() {is=title K2Node_CustomEvent}
   n6 = Get Right()
   BindToAnimationFinished(execute=@n4.then, self=@n6, Widget=Self-Reference(), Delegate=@n5)
   > PlayAnimation(InAnimation=@n6)
   n7 = InputModeChanged!() -> (OutputDelegate, then, GamePad:bool=~, InputType:ECommonInputType Enum) {is=title K2Node_CustomEvent}
   > n8 = Set isGamepad(isGamepad:bool=@n7.GamePad) -> (Output_Get:bool)
   > n9 = Branch(Condition=@n8)
   ClearSelection(execute=@n9.else, self:Common Button Base Object Reference)
   n10 = InputUpdated!() -> (OutputDelegate, then, bNewInputType:ECommonInputType Enum) {is=title K2Node_CustomEvent}
   > InputModeChanged!(execute, self, GamePad:bool="Equal (Enum)"(A=@n10.bNewInputType, B=Gamepad), InputType:ECommonInputType Enum=@n10.bNewInputType) -> (then) {is=title K2Node_CallFunction}
   Event On Clicked()
   > ShiftTextRight()
   > UpdateReadOnlyValues()
   Event On Hovered()
   n11 = Get OptionSelectionArea() -> (OptionSelectionArea:Horizontal Box Object Reference)
   n12 = Get Image_Selection_Hover() -> (Image_Selection_Hover:Image Object Reference)
   n13 = Get OptionDisplayName_Bounce() -> (OptionDisplayName_Bounce:Common Text Object Reference)
   > SetVisibility(self=[@n11, @n12, @n13], InVisibility=SelfHitTestInvisible)
   n14 = Get OptionDisplayName() -> (OptionDisplayName:Common Text Object Reference)
   n15 = Get TextDisplayValue_Readonly() -> (TextDisplayValue_Readonly:Common Text Object Reference)
   > SetVisibility(self=[@n14, @n15], InVisibility=Collapsed)
   > SetColorAndOpacity(
    self=[Get OptionDisplayName(), Get OptionDisplayName_Bounce()]
    InColorAndOpacity=Get TextColor_Focused() -> (TextColor_Focused:Slate Color Structure)
   )
   n16 = Event OnLoadingStateChanged!() -> (OutputDelegate, then, bCurrentlyLoading:bool) {is=title K2Node_Event}
   > SetVisibility(
    self=Get SaveSpinner() -> (SaveSpinner:Image Object Reference)
    InVisibility=Select(Option 0=Collapsed, Option 1=Visible, Index=@n16.bCurrentlyLoading)
   )
   Event Construct()
   n17 = Get CommonInputSubsystem!() -> (ReturnValue:Common Input Object Reference) {is=title K2Node_GetBlueprintContextEx}
   n18 = GetCurrentInputType(self=@n17)
   > InputModeChanged(GamePad="Equal (Enum)"(A=@n18, B=Gamepad), InputType=@n18)
   > Assign On Input Method Changed!(execute:Exec, self:Common Input Object Reference=@n17, Delegate:"Delegate (by ref)"=@n10) -> (then:Exec) {is=title K2Node_AssignDelegate}
   > Set Button Sizes() {is=title K2Node_CallFunction}
   Event On Unhovered()
   > SetVisibility(self=[@n11, @n12, @n13], InVisibility=Collapsed)
   > SetVisibility(self=[@n14, @n15], InVisibility=SelfHitTestInvisible)
   > SetColorAndOpacity(
    self=[Get OptionDisplayName(), Get OptionDisplayName_Bounce()]
    InColorAndOpacity=Get TextColor_Normal() -> (TextColor_Normal:Slate Color Structure)
   )
   n19 = Event OnSpinnerRotated!() -> (OutputDelegate, then, Value:int=0, bUserInitiated:bool) {is=title K2Node_Event}
   > Call On Spinner Value Changed!(execute:Exec, self:Creative Property Editor Entry Object Reference, Index:int=@n19.Value) -> (then:Exec) {is=title K2Node_CallDelegate}
   > UpdateReadOnlyValues()
   n20 = Branch(Condition=Event SetSpinnerOptionsVisiblity!() -> (OutputDelegate, then, bVisible:bool) {is=title K2Node_Event}.bVisible)
   > n21 = Set VisibilitySetting(VisibilitySetting:ESlateVisibility Enum=Visible) -> (Output_Get:ESlateVisibility Enum=Visible)
   SetVisibility(
    execute=[@n22.then, @n21.then]
    self=Get OptionSelectionArea()
    InVisibility=Get VisibilitySetting() -> (VisibilitySetting:ESlateVisibility Enum=Visible)
   )
   n22 = Set VisibilitySetting(execute=@n20.else, VisibilitySetting=Collapsed) -> (Output_Get=Visible)
   UpdateReadOnlyValues(execute=[@n24.then, @n28.then])
   n23 = Get Right()
   n24 = PlayAnimation(execute=@n27.then, InAnimation=@n23, PlayMode=Reverse)
   n25 = Self-Reference()
   n26 = Get Left()
   n27 = UnbindAllFromAnimationFinished!(execute=@n30.then, self:Widget Animation Object Reference=[@n23, @n26], Widget:User Widget Object Reference=@n25) -> (then)
   UnbindAllFromAnimationFinished(execute=@n29.then, self=[@n23, @n26], Widget=@n25)
   > n28 = PlayAnimation(InAnimation=@n26, PlayMode=Reverse)
   n29 = ShiftTextRight(execute=@n5.then)
   n30 = ShiftTextLeft(execute=@n2.then)
  ]}
 OnFocusReceived {nodes [
   OnFocusReceived!() -> (then, MyGeometry:Geometry Structure, InFocusEvent:Focus Event Structure) {is=title K2Node_FunctionEntry}
   n1 = GetVisibility!(self:Widget Object Reference=Get SizeBox_TextEntry() -> (SizeBox_TextEntry:Size Box Object Reference)) -> (ReturnValue:ESlateVisibility Enum=Visible)
   n2 = GetVisibility(self=Get FilteredTextEntry_Widget() -> (FilteredTextEntry_Widget:Fort Editable Filtered Counted Text Box Object Reference)) -> (ReturnValue=Visible)
   > n3 = Branch(Condition=OR Boolean(A="Equal (Enum)"(A:ESlateVisibility Enum=@n2, B:ESlateVisibility Enum=Collapsed), B="Equal (Enum)"(A=@n1, B=Collapsed)))
   > Return Node(ReturnValue:Event Reply Structure=Unhandled())
   Return Node(execute=@n3.else, ReturnValue=Handled())
  ]}
 OnSpinnerValueChanged {nodes [OnSpinnerValueChanged!() -> (then, Index:int) {is=title K2Node_FunctionEntry}]}
 Set Button Sizes {nodes [
   Set Button Sizes() {is=title K2Node_FunctionEntry}
   > Branch(Condition=IsMobileGame())
   n1 = Get Mobile Size Override() -> (Mobile Size Override:double)
   n2 = Get LeftButtonSize() -> (LeftButtonSize:Size Box Object Reference)
   n3 = Get RightButtonSize() -> (RightButtonSize:Size Box Object Reference)
   > SetHeightOverride(self=[@n2, @n3], InHeightOverride=@n1)
   > SetWidthOverride(self=[@n2, @n3], InWidthOverride=@n1)
  ]}
 Update {nodes [
   Update!() -> (then, isModifiedFromDefaultValue:bool=~) {is=title K2Node_FunctionEntry}
   > n1 = Sequence!(execute) -> (then_0, then_1, then_2:Exec)
   n2 = GetEmptyText!() -> (ReturnValue:Text)
   n3 = Get isModifiedFromDefaultValue() -> (isModifiedFromDefaultValue:bool)
   > "SetText (Text)"(
    self=[
     Get OptionDisplayName()
     Get OptionDisplayName_Bounce()
     Get TextBlock_OptionDisplayName() -> (TextBlock_OptionDisplayName:Common Text Object Reference)
    ]
    InText=Format Text!(
     Format="LOCTEXT(\"{modified}{title}\")"
     modified:Text=Select(Option 0:Text=@n2, Option 1:Text="LOCTEXT(\"*\")", Index=@n3) -> (ReturnValue:Text)
     title:Text=GetTitleText!(self:Fort Matchmaking Knobs Spinner Button Object Reference) -> (ReturnValue:Text)
    ) -> (Result)
   )
   UpdateReadOnlyValues(execute=@n1.then_2)
  ]}
 UpdateReadOnlyValues {nodes [
   n1 = Cast To WBP_UIKit_FilteredInputField!(execute, Object=Get FilteredTextEntry_Widget()) -> (then, CastFailed, AsWBP UIKit Filtered Input Field:WBP UIKit Filtered Input Field Object Reference) {is=title K2Node_DynamicCast}
   > n2 = GetCurrentText!(execute, self:WBP UIKit Filtered Input Field Object Reference=@n1) -> (then, Text:Text)
   > n3 = Branch(Condition="Equal Exactly (Text)"!(A:"Text (by ref)"=@n2, B:"Text (by ref)"=GetEmptyText()) -> (ReturnValue:bool))
   "SetText (Text)"(
    execute=[@n4.then, @n3.then]
    self=Get TextDisplayValue_Readonly()
    InText="GetText (Text)"!(self:Text Object Reference=Get MyText() -> (MyText:Common Text Object Reference)) -> (ReturnValue:Text)
   )
   "SetText (Text)"(execute=@n3.else, self=Get TextDisplayValue_Readonly(), InText=@n2)
   n4 = UpdateReadOnlyValues() {is=title K2Node_FunctionEntry}
  ]}
 UpdateSyle {nodes [UpdateSyle!() -> (then, InputType:ECommonInputType Enum, isUsing GamePad:bool=~) {is=title K2Node_FunctionEntry}]}
]
variables [
 isGamepad {type=boolean}
 Mobile Size Override {type=number, default=50}
 OnSpinnerValueChanged {}
 TextColor_Focused {type=SlateColor, default {colorUseRule=UseColor_Specified, specifiedColor=<rgba 0 0.0887 0.9823 1>}}
 TextColor_Normal {type=SlateColor, default {colorUseRule=UseColor_Specified, specifiedColor=<rgba 0.3095 0.7157 0.9911 1>}}
 VisibilitySetting {type=ESlateVisibility, default=Visible}
]
widgets {
 info {inheritedWidgetCount=5, parentClass=/Script/FortniteUI.FortMatchmakingKnobsSpinnerButton, rootWidgetClass=&Border, widgetCount=27}
 widgets [
  Border_Fill {
   widgetClassPath=&Border
   bIsVariable=true
   widget=Border_Fill
   properties {
    navigation=&self:WidgetTree.Border_Fill.WidgetNavigation_0
    verticalAlignment=VAlign_Center
    background {
     bIsDynamicallyLoaded=false
     drawAs=NoDrawType
     imageType=NoImage
     mirroring=NoMirror
     resourceName=None
     resourceObject=null
     tiling=NoTile
     imageSize=<32 32>
     margin=<ltrb 0 0 0 0>
     outlineSettings {
      bUseBrushTransparency=false
      roundingType=HalfHeightRadius
      width=0
      color {colorUseRule=UseColor_Specified, specifiedColor=<rgba 0 0 0 0>}
      cornerRadii=<0 0 0 0>
     }
     tintColor {colorUseRule=UseColor_Specified, specifiedColor=<rgba 0.7454 0.7454 0.7454 0.05>}
     uVRegion {bIsValid=false, max=<0 0>, min=<0 0>}
    }
    padding=<ltrb 0 0 0 0>
   }
  }
  Overlay_2 {
   widgetClassPath=&Overlay
   parent=Border_Fill
   widget=Overlay_2
   slot {class=BorderSlot, properties {verticalAlignment=VAlign_Center, padding=<ltrb 0 0 0 0>}}
  }
  Image_Selection_Unselected {
   widgetClassPath=&Image
   bIsVariable=true
   parent=Overlay_2
   widget=Image_Selection_Unselected
   properties {
    visibility=HitTestInvisible
    brush {
     bIsDynamicallyLoaded=false
     drawAs=Image
     imageType=NoImage
     mirroring=NoMirror
     resourceName=None
     resourceObject=null
     tiling=NoTile
     imageSize=<32 82.5632>
     margin=<ltrb 0 0 0 0>
     outlineSettings {
      bUseBrushTransparency=false
      roundingType=HalfHeightRadius
      width=0
      color {colorUseRule=UseColor_Specified, specifiedColor=<rgba 0 0 0 0>}
      cornerRadii=<0 0 0 0>
     }
     tintColor {colorUseRule=UseColor_Specified, specifiedColor=<rgba 1 1 1 1>}
     uVRegion {bIsValid=false, max=<0 0>, min=<0 0>}
    }
    colorAndOpacity=<rgba 0.0235 0.0275 0.0902 1>
   }
   slot {class=OverlaySlot, properties {horizontalAlignment=HAlign_Fill, verticalAlignment=VAlign_Fill, padding=<ltrb 16 2 16 2>}}
  }
  Image_Selection_Hover {
   widgetClassPath=&Image
   bIsVariable=true
   parent=Overlay_2
   widget=Image_Selection_Hover
   properties {
    visibility=Collapsed
    brush {
     bIsDynamicallyLoaded=false
     drawAs=Image
     imageType=NoImage
     mirroring=NoMirror
     resourceName=None
     resourceObject=/Game/Athena/UI/InGame/Creative/Art/M_SimpleSkewBox_CreativeQuickMenuSelection
     tiling=NoTile
     imageSize=<32 52>
     margin=<ltrb 0 0 0 0>
     outlineSettings {
      bUseBrushTransparency=false
      roundingType=HalfHeightRadius
      width=0
      color {colorUseRule=UseColor_Specified, specifiedColor=<rgba 0 0 0 0>}
      cornerRadii=<0 0 0 0>
     }
     tintColor {colorUseRule=UseColor_Specified, specifiedColor=<rgba 1 1 1 1>}
     uVRegion {bIsValid=false, max=<0 0>, min=<0 0>}
    }
    renderTransform {angle=0, scale=<1.31 1.3>, shear=<0 0>, translation=<0 0>}
   }
   slot {class=OverlaySlot, properties {horizontalAlignment=HAlign_Fill, verticalAlignment=VAlign_Fill, padding=<ltrb 0 2 0 2>}}
  }
  SizeBox_1 {
   widgetClassPath=&SizeBox
   parent=Overlay_2
   widget=SizeBox_1
   properties {heightOverride=64, minDesiredHeight=56, widthOverride=1000}
   slot {class=OverlaySlot, properties {horizontalAlignment=HAlign_Fill, verticalAlignment=VAlign_Center, padding=<ltrb 16 0 16 0>}}
  }
  CommonBorder_Container {
   widgetClassPath=&CommonBorder
   parent=SizeBox_1
   widget=CommonBorder_Container
   properties {padding=<ltrb 6 6 6 6>}
   slot {class=SizeBoxSlot}
  }
  VerticalBox_0 {
   widgetClassPath=&VerticalBox
   parent=CommonBorder_Container
   widget=VerticalBox_0
   slot {class=BorderSlot, properties {padding=<ltrb 6 6 6 6>}}
  }
  HorizontalBox_0 {widgetClassPath=&HorizontalBox, parent=VerticalBox_0, widget=HorizontalBox_0, slot {class=VerticalBoxSlot}}
  SizeBox_0 {
   widgetClassPath=&SizeBox
   parent=HorizontalBox_0
   widget=SizeBox_0
   properties {bOverride_WidthOverride=true, clipping=ClipToBounds, maxDesiredWidth=669.8986, widthOverride=500}
   slot {
    class=HorizontalBoxSlot
    properties {horizontalAlignment=HAlign_Left, verticalAlignment=VAlign_Center, padding=<ltrb 16 0 24 0>, size {sizeRule=Fill, value=1}}
   }
  }
  Overlay_0 {
   widgetClassPath=&Overlay
   parent=SizeBox_0
   widget=Overlay_0
   slot {class=SizeBoxSlot, properties {verticalAlignment=VAlign_Center}}
  }
  OptionDisplayName {
   widgetClassPath=&CommonTextBlock
   bIsVariable=true
   parent=Overlay_0
   widget=OptionDisplayName
   properties {
    text="[Parameter name]"
    colorAndOpacity {colorUseRule=UseColor_Specified, specifiedColor=<rgba 0.3095 0.7157 0.9911 1>}
    font {
     bForceMonospaced=false
     bMaterialIsStencil=false
     fontMaterial=null
     fontObject=&BurbankSmall
     letterSpacing=0
     monospacedWidth=1
     size=20
     skewAmount=0
     typefaceFontName=Bold
     outlineSettings {
      bApplyOutlineToDropShadows=false
      bMiteredCorners=false
      bSeparateFillAlpha=false
      outlineMaterial=null
      outlineSize=0
      outlineColor=<rgba 0.0052 0.0052 0.0056 0.2>
     }
    }
   }
   slot {class=OverlaySlot}
  }
  OptionDisplayName_Bounce {
   widgetClassPath=&CommonTextBlock
   bIsVariable=true
   parent=Overlay_0
   widget=OptionDisplayName_Bounce
   properties {
    scrollStyle=&TextScrollStyle-Base_C
    text="[Parameter name]"
    visibility=Collapsed
    colorAndOpacity {colorUseRule=UseColor_Specified, specifiedColor=<rgba 0.3095 0.7157 0.9911 1>}
    font {
     bForceMonospaced=false
     bMaterialIsStencil=false
     fontMaterial=null
     fontObject=&BurbankSmall
     letterSpacing=0
     monospacedWidth=1
     size=20
     skewAmount=0
     typefaceFontName=Bold
     outlineSettings {
      bApplyOutlineToDropShadows=false
      bMiteredCorners=false
      bSeparateFillAlpha=false
      outlineMaterial=null
      outlineSize=0
      outlineColor=<rgba 0.0052 0.0052 0.0056 0.2>
     }
    }
   }
   slot {class=OverlaySlot}
  }
  Overlay_1 {widgetClassPath=&Overlay, parent=HorizontalBox_0, widget=Overlay_1, slot {class=HorizontalBoxSlot}}
  TextDisplayValue_Readonly {
   widgetClassPath=&CommonTextBlock
   bIsVariable=true
   parent=Overlay_1
   widget=TextDisplayValue_Readonly
   properties {
    clipping=ClipToBounds
    text="[VALUE]"
    colorAndOpacity {colorUseRule=UseColor_Specified, specifiedColor=<rgba 0.9823 0.9823 0.9823 1>}
    font {
     bForceMonospaced=false
     bMaterialIsStencil=false
     fontMaterial=null
     fontObject=&BurbankSmall
     letterSpacing=0
     monospacedWidth=1
     size=20
     skewAmount=0
     typefaceFontName=Bold
     outlineSettings {
      bApplyOutlineToDropShadows=false
      bMiteredCorners=false
      bSeparateFillAlpha=false
      outlineMaterial=null
      outlineSize=0
      outlineColor=<rgba 0 0 0 0>
     }
    }
    shadowColorAndOpacity=<rgba 0.0052 0.0052 0.0056 0>
   }
   slot {class=OverlaySlot, properties {horizontalAlignment=HAlign_Right, verticalAlignment=VAlign_Center, padding=<ltrb 20 0 20 0>}}
  }
  OptionSelectionArea {
   widgetClassPath=&HorizontalBox
   bIsVariable=true
   parent=Overlay_1
   widget=OptionSelectionArea
   properties {visibility=Collapsed}
   slot {class=OverlaySlot, properties {verticalAlignment=VAlign_Fill, padding=<ltrb 32 0 0 0>}}
  }
  LeftButtonSize {
   widgetClassPath=&SizeBox
   bIsVariable=true
   parent=OptionSelectionArea
   widget=LeftButtonSize
   properties {bOverride_HeightOverride=true, bOverride_WidthOverride=true, heightOverride=48, widthOverride=42}
   slot {class=HorizontalBoxSlot, properties {horizontalAlignment=HAlign_Center, verticalAlignment=VAlign_Center}}
  }
  LeftButton {
   widgetClassPath=/Game/Athena/UI/Matchmaking/SimpleCommonButton
   bIsVariable=true
   parent=LeftButtonSize
   widget=LeftButton
   properties {
    onButtonBaseClicked=(&self:WidgetTree.LeftButton.HandleOnButtonClicked)
    onButtonBaseDoubleClicked=(&self:WidgetTree.LeftButton.HandleOnButtonDoubleClicked)
    onButtonBaseHovered=(&self:WidgetTree.LeftButton.HandleOnButtonHovered)
    onButtonBaseUnhovered=(&self:WidgetTree.LeftButton.HandleOnButtonUnhovered)
    onSelectedChangedBase=(&self:WidgetTree.LeftButton.HandleOnSelectedChanged)
    style=/FortUILibrary/Foundation/Buttons/ButtonStyles/ButtonStyle_SimpleCycleArrowEmphasizedClick_Left.ButtonStyle_SimpleCycleArrowEmphasizedClick_Left_C
    colorAndOpacity=<rgba 0.0545 0.4735 0.9647 1>
   }
   slot {class=SizeBoxSlot, properties {verticalAlignment=VAlign_Center, padding=<ltrb 0 0 4 0>}}
  }
  SizeBox_2 {
   widgetClassPath=&SizeBox
   parent=OptionSelectionArea
   widget=SizeBox_2
   properties {
    bOverride_HeightOverride=true
    bOverride_WidthOverride=true
    clipping=ClipToBounds
    heightOverride=42
    maxDesiredWidth=816
    widthOverride=320
   }
   slot {class=HorizontalBoxSlot, properties {horizontalAlignment=HAlign_Center, verticalAlignment=VAlign_Center}}
  }
  Border_4 {
   widgetClassPath=&Border
   parent=SizeBox_2
   widget=Border_4
   properties {
    verticalAlignment=VAlign_Center
    background {
     bIsDynamicallyLoaded=false
     drawAs=Image
     imageType=NoImage
     mirroring=NoMirror
     resourceName=None
     resourceObject=null
     tiling=NoTile
     imageSize=<32 32>
     margin=<ltrb 0 0 0 0>
     outlineSettings {
      bUseBrushTransparency=false
      roundingType=HalfHeightRadius
      width=0
      color {colorUseRule=UseColor_Specified, specifiedColor=<rgba 0 0 0 0>}
      cornerRadii=<0 0 0 0>
     }
     tintColor {colorUseRule=UseColor_Specified, specifiedColor=<rgba 0.0545 0.4735 0.9647 1>}
     uVRegion {bIsValid=false, max=<0 0>, min=<0 0>}
    }
    brushColor=<rgba 1 1 1 0.2>
    padding=<ltrb 8 0 8 0>
   }
   slot {class=SizeBoxSlot, properties {padding=<ltrb 12 0 12 0>}}
  }
  MyText {
   widgetClassPath=&CommonTextBlock
   bInherited=true
   bIsVariable=true
   parent=Border_4
   widget=MyText
   properties {
    justification=Center
    scrollStyle=&TextScrollStyle-Base_C
    text=" Option Text"
    colorAndOpacity {colorUseRule=UseColor_Specified, specifiedColor=<rgba 0 0 0 1>}
    font {
     bForceMonospaced=false
     bMaterialIsStencil=false
     fontMaterial=null
     fontObject=&BurbankSmall
     letterSpacing=0
     monospacedWidth=1
     size=20
     skewAmount=0
     typefaceFontName=Bold
     outlineSettings {
      bApplyOutlineToDropShadows=false
      bMiteredCorners=false
      bSeparateFillAlpha=false
      outlineMaterial=null
      outlineSize=0
      outlineColor=<rgba 0 0.0182 0.2031 0.2>
     }
    }
    shadowColorAndOpacity=<rgba 0.0052 0.0052 0.0056 0>
    shadowOffset=<2 1>
   }
   slot {class=BorderSlot, properties {verticalAlignment=VAlign_Center, padding=<ltrb 8 0 8 0>}}
  }
  RightButtonSize {
   widgetClassPath=&SizeBox
   bIsVariable=true
   parent=OptionSelectionArea
   widget=RightButtonSize
   properties {bOverride_HeightOverride=true, bOverride_WidthOverride=true, heightOverride=48, widthOverride=42}
   slot {class=HorizontalBoxSlot, properties {horizontalAlignment=HAlign_Center, verticalAlignment=VAlign_Center}}
  }
  RightButton {
   widgetClassPath=&SimpleCommonButton
   bIsVariable=true
   parent=RightButtonSize
   widget=RightButton
   properties {
    onButtonBaseClicked=(&self:WidgetTree.RightButton.HandleOnButtonClicked)
    onButtonBaseDoubleClicked=(&self:WidgetTree.RightButton.HandleOnButtonDoubleClicked)
    onButtonBaseHovered=(&self:WidgetTree.RightButton.HandleOnButtonHovered)
    onButtonBaseUnhovered=(&self:WidgetTree.RightButton.HandleOnButtonUnhovered)
    onSelectedChangedBase=(&self:WidgetTree.RightButton.HandleOnSelectedChanged)
    style=/FortUILibrary/Foundation/Buttons/ButtonStyles/ButtonStyle_SimpleCycleArrow_Right.ButtonStyle_SimpleCycleArrow_Right_C
    colorAndOpacity=<rgba 0.0545 0.4735 0.9647 1>
   }
   slot {class=SizeBoxSlot, properties {padding=<ltrb 4 0 0 0>}}
  }
  SizeBox_TextEntry {
   widgetClassPath=&SizeBox
   bInherited=true
   bIsVariable=true
   parent=VerticalBox_0
   widget=SizeBox_TextEntry
   properties {bOverride_MaxDesiredWidth=true, maxDesiredWidth=750, visibility=Collapsed}
   slot {class=VerticalBoxSlot, properties {padding=<ltrb 32 12 32 0>}}
  }
  FilteredTextEntry_Widget {
   widgetClassPath=/FortUILibrary/UIKit/InputField/WBP_UIKit_FilteredInputField
   bInherited=true
   bIsVariable=true
   parent=SizeBox_TextEntry
   widget=FilteredTextEntry_Widget
   slot {class=SizeBoxSlot}
  }
  TextBlock_OptionDisplayName {
   widgetClassPath=&CommonTextBlock
   bInherited=true
   parent=Overlay_2
   widget=TextBlock_OptionDisplayName
   properties {
    style=&TextStyle-Base_C
    text=Text Block
    visibility=Collapsed
    font {
     bForceMonospaced=false
     bMaterialIsStencil=false
     fontMaterial=null
     fontObject=&NotoSans
     letterSpacing=0
     monospacedWidth=1
     size=24
     skewAmount=0
     typefaceFontName=NotoSans-Regular
     outlineSettings {
      bApplyOutlineToDropShadows=false
      bMiteredCorners=false
      bSeparateFillAlpha=false
      outlineMaterial=null
      outlineSize=0
      outlineColor=<rgba 0 0 0 1>
     }
    }
    shadowOffset=<0 0>
   }
   slot {class=OverlaySlot}
  }
  SaveSpinner {
   widgetClassPath=&Image
   bIsVariable=true
   parent=Overlay_2
   widget=SaveSpinner
   properties {
    visibility=Collapsed
    brush {
     bIsDynamicallyLoaded=false
     drawAs=Image
     imageType=NoImage
     mirroring=NoMirror
     resourceName=None
     resourceObject=/FortUILibrary/UIKit/Throbber/MI_UIKit_Color_Throbber
     tiling=NoTile
     imageSize=<24 24>
     margin=<ltrb 0 0 0 0>
     outlineSettings {
      bUseBrushTransparency=false
      roundingType=HalfHeightRadius
      width=0
      color {colorUseRule=UseColor_Specified, specifiedColor=<rgba 0 0 0 0>}
      cornerRadii=<0 0 0 0>
     }
     tintColor {colorUseRule=UseColor_Specified, specifiedColor=<rgba 1 1 1 1>}
     uVRegion {bIsValid=false, max=<0 0>, min=<0 0>}
    }
   }
   slot {class=OverlaySlot, properties {padding=<ltrb 8 8 0 0>}}
  }
  InputActionWidget {widgetClassPath=&CommonActionWidget, bInherited=true, widget=null}
 ]
}
