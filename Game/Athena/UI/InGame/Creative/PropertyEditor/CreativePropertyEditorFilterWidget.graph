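# Filter bar widget for the Creative property editor, showing the current filter name, previous/next buttons, a row of filter tab buttons and a text search toggle.
# When the C++ parent creates a filter button, the graph passes it the tab label info and shows its left-side icon.
# Toggling search fires selected or deselected dispatchers, and SetSearchButtonState sets the toggle's selection without click feedback.
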WidgetBlueprint /Game/Athena/UI/InGame/Creative/PropertyEditor/CreativePropertyEditorFilterWidget
graphs [
 EventGraph {nodes [
   n1 = "On Selected Changed (Button_TextSearch)"!() -> (OutputDelegate, then, Button:Common Button Legacy Object Reference, Selected:bool) {is=title K2Node_ComponentBoundEvent}
   > n2 = Branch(Condition=@n1.Selected)
   > Call On Text Search Selected!(execute:Exec, self:Creative Property Editor Filter Widget Object Reference) -> (then:Exec) {is=title K2Node_CallDelegate}
   Call On Text Search Deselected!(execute:Exec=@n2.else, self:Creative Property Editor Filter Widget Object Reference) -> (then:Exec) {is=title K2Node_CallDelegate}
   n3 = Event OnFilterButtonCreated!() -> (
    OutputDelegate
    then
    FilterButton:Common Button Base Object Reference
    TabInfo_TabNameId:Name
    TabInfo_bHidden:bool=~
    TabInfo_bAllowedInZone:bool=~
    TabInfo_TabLabelInfo:Fort Tab Button Label Info Structure
    TabInfo_TabButtonType:Common Button Base Class Reference
    TabInfo_TabContentType:Common User Widget Class Reference
    TabInfo_TabScene:Dynamic UIScene Object Reference
   ) {is=title K2Node_Event}
   > n4 = Cast To FortTabButtonInterface!(execute, Object=@n3.FilterButton) -> (then, CastFailed, AsFort Tab Button Interface:Fort Tab Button Interface Interface) {is=title K2Node_DynamicCast}
   > SetTabLabelInfo!(execute, self:Fort Tab Button Interface Interface=@n4, TabLabelInfo:"Fort Tab Button Label Info Structure (by ref)"=@n3.TabInfo_TabLabelInfo) -> (then)
   > n5 = Cast To IconTabButton_Legacy!(execute, Object=@n3.FilterButton) -> (then, CastFailed, AsIcon Tab Button Legacy:Icon Tab Button Legacy Object Reference) {is=title K2Node_DynamicCast}
   > SetVisibility(
    self=Get LeftSideImage!(self:Icon Tab Button Legacy Object Reference=@n5) -> (LeftSideImage:Image Object Reference)
    InVisibility=SelfHitTestInvisible
   )
  ]}
 OnTextSearchDeselected {nodes [OnTextSearchDeselected() {is=title K2Node_FunctionEntry}]}
 OnTextSearchSelected {nodes [OnTextSearchSelected() {is=title K2Node_FunctionEntry}]}
 SetSearchButtonState {nodes [
   n1 = SetSearchButtonState!() -> (then, bIsSelected:bool=~) {is=title K2Node_FunctionEntry}
   > SetIsSelected!(
    execute
    self:Common Button Base Object Reference=Get Button_TextSearch() -> (Button_TextSearch:Icon Tab Button Legacy Object Reference)
    InSelected:bool=@n1
    bGiveClickFeedback:bool=false
   ) -> (then)
  ]}
]
variables [
 OnTextSearchDeselected {}
 OnTextSearchSelected {}
]
widgets {
 info {
  inheritedWidgetCount=4
  parentClass=/Script/FortniteUI.FortCreativePropertyEditorFilter
  rootWidgetClass=&HorizontalBox
  widgetCount=10
 }
 widgets [
  HorizontalBox_0 {widgetClassPath=&HorizontalBox, widget=HorizontalBox_0, properties {flowDirectionPreference=Culture}}
  Text_FilterName {
   widgetClassPath=&CommonTextBlock
   bInherited=true
   parent=HorizontalBox_0
   widget=Text_FilterName
   properties {
    text=Basic options
    textTransformPolicy=ToUpper
    colorAndOpacity {colorUseRule=UseColor_Specified, specifiedColor=<rgba 0.1329 0.6038 1 1>}
    font {
     bForceMonospaced=false
     bMaterialIsStencil=false
     fontMaterial=null
     fontObject=&BurbankSmall
     letterSpacing=0
     monospacedWidth=1
     size=18
     skewAmount=0
     typefaceFontName=Black
     outlineSettings {
      bApplyOutlineToDropShadows=false
      bMiteredCorners=false
      bSeparateFillAlpha=false
      outlineMaterial=null
      outlineSize=0
      outlineColor=<rgba 0 0 0 1>
     }
    }
    shadowOffset=<0 0>
   }
   slot {class=HorizontalBoxSlot, properties {verticalAlignment=VAlign_Center}}
  }
  HorizontalBox_1 {
   widgetClassPath=&HorizontalBox
   parent=HorizontalBox_0
   widget=HorizontalBox_1
   slot {class=HorizontalBoxSlot, properties {horizontalAlignment=HAlign_Right, size {sizeRule=Fill, value=1}}}
  }
  Button_PreviousFilter {
   widgetClassPath=/Game/UI/SocialPanel/UserActionMenuInputButton
   bInherited=true
   parent=HorizontalBox_1
   widget=Button_PreviousFilter
   properties {
    onButtonBaseClicked=(&self:WidgetTree.Button_PreviousFilter.HandleOnButtonClicked)
    onButtonBaseDoubleClicked=(&self:WidgetTree.Button_PreviousFilter.HandleOnButtonDoubleClicked)
    onButtonBaseHovered=(&self:WidgetTree.Button_PreviousFilter.HandleOnButtonHovered)
    onButtonBaseUnhovered=(&self:WidgetTree.Button_PreviousFilter.HandleOnButtonUnhovered)
    onSelectedChangedBase=(&self:WidgetTree.Button_PreviousFilter.HandleOnSelectedChanged)
    triggeringInputAction {dataTable=/FortUILibrary/Input/CreativeInputActionDataTable, rowName=ChangePropertyFilterLeft}
   }
   slot {class=HorizontalBoxSlot}
  }
  SizeBox_6 {
   widgetClassPath=&SizeBox
   parent=HorizontalBox_1
   widget=SizeBox_6
   properties {bOverride_HeightOverride=true, heightOverride=64, widthOverride=340}
   slot {class=HorizontalBoxSlot, properties {verticalAlignment=VAlign_Center}}
  }
  EntryBox_FilterButtons {
   widgetClassPath=&DynamicEntryBox
   bInherited=true
   parent=SizeBox_6
   widget=EntryBox_FilterButtons
   properties {
    entryWidgetClass=/FortUILibrary/Foundation/Buttons/IconTabButton_Legacy.IconTabButton_Legacy_C
    maxElementSize=68
    numDesignerPreviewEntries=0
    entrySpacing=<1 0>
   }
   slot {class=SizeBoxSlot}
  }
  Button_NextFilter {
   widgetClassPath=&UserActionMenuInputButton
   bInherited=true
   parent=HorizontalBox_1
   widget=Button_NextFilter
   properties {
    onButtonBaseClicked=(&self:WidgetTree.Button_NextFilter.HandleOnButtonClicked)
    onButtonBaseDoubleClicked=(&self:WidgetTree.Button_NextFilter.HandleOnButtonDoubleClicked)
    onButtonBaseHovered=(&self:WidgetTree.Button_NextFilter.HandleOnButtonHovered)
    onButtonBaseUnhovered=(&self:WidgetTree.Button_NextFilter.HandleOnButtonUnhovered)
    onSelectedChangedBase=(&self:WidgetTree.Button_NextFilter.HandleOnSelectedChanged)
    triggeringInputAction {dataTable=&CreativeInputActionDataTable, rowName=ChangePropertyFilterRight}
   }
   slot {class=HorizontalBoxSlot}
  }
  SizeBox_4 {
   widgetClassPath=&SizeBox
   parent=HorizontalBox_0
   widget=SizeBox_4
   properties {bOverride_HeightOverride=true, bOverride_WidthOverride=true, heightOverride=64, widthOverride=68}
   slot {class=HorizontalBoxSlot, properties {horizontalAlignment=HAlign_Right, verticalAlignment=VAlign_Center, padding=<ltrb 12 0 0 0>}}
  }
  Button_TextSearch {
   widgetClassPath=/FortUILibrary/Foundation/Buttons/IconTabButton_Legacy
   bIsVariable=true
   parent=SizeBox_4
   widget=Button_TextSearch
   properties {
    bInteractableWhenSelected=true
    bIsFocusable=false
    bSelectable=true
    bToggleable=true
    buttonText=""
    onButtonBaseClicked=(&self:WidgetTree.Button_TextSearch.HandleOnButtonClicked)
    onButtonBaseDoubleClicked=(&self:WidgetTree.Button_TextSearch.HandleOnButtonDoubleClicked)
    onButtonBaseHovered=(&self:WidgetTree.Button_TextSearch.HandleOnButtonHovered)
    onButtonBaseUnhovered=(&self:WidgetTree.Button_TextSearch.HandleOnButtonUnhovered)
    onSelectedChangedBase=(&self:WidgetTree.Button_TextSearch.HandleOnSelectedChanged)
    style=/FortUILibrary/Foundation/Buttons/ButtonStyles/Tab/Athena_ButtonStyle_LockerTab.Athena_ButtonStyle_LockerTab_C
    iconBrush {
     bIsDynamicallyLoaded=false
     drawAs=Image
     imageType=NoImage
     mirroring=NoMirror
     resourceName=None
     resourceObject=/Game/Athena/UI/InGame/Creative/PropertyEditor/Art/Icon-PropertyEditor-Search
     tiling=NoTile
     imageSize=<42 42>
     margin=<ltrb 0 0 0 0>
     outlineSettings {
      bUseBrushTransparency=false
      roundingType=HalfHeightRadius
      width=0
      color {colorUseRule=UseColor_Specified, specifiedColor=<rgba 0 0 0 0>}
      cornerRadii=<0 0 0 1>
     }
     tintColor {colorUseRule=UseColor_Specified, specifiedColor=<rgba 1 1 1 1>}
     uVRegion {bIsValid=false, max=<0 0>, min=<0 0>}
    }
    triggeringInputAction {dataTable=&CreativeInputActionDataTable, rowName=GenericFaceButtonTop}
   }
   slot {class=SizeBoxSlot}
  }
  CommonActionWidget_134 {
   widgetClassPath=&CommonActionWidget
   parent=HorizontalBox_0
   widget=CommonActionWidget_134
   properties {visibility=SelfHitTestInvisible, inputActions [{dataTable=&CreativeInputActionDataTable, rowName=GenericFaceButtonTop}]}
   slot {class=HorizontalBoxSlot}
  }
 ]
}
