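# Creative-mode property editor modal that lists an actor's options for editing, filtering, renaming and resetting.
# It animates in and out, hiding HUD elements while active and closing on island, minigame or loading-state changes.
# A filter bar and timer-debounced text search regenerate the option list, with focus handling for gamepad.
# A lazily created More Options menu drives rename, clear-items and per-category resets, while closing through OK commits pending changes.
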WidgetBlueprint /Game/Athena/UI/InGame/Creative/PropertyEditor/CreativePropertyEditor
animations [
 In {
  binding_count=1
  display_label=""
  length_seconds=0.2
  object=&self:In
  sequence {bindings [Overlay_Shell {
     id=#1
     possessed_class=Overlay
     tracks [
      {
       type=MovieScene2DTransformTrack
       sections [{
         type=MovieScene2DTransformSection
         end_frame=4
         start_frame=0
         channels [
          Translation.X {type=MovieSceneScriptingFloatChannel, keys [[0,-1000],[4,0]]}
          Translation.Y {type=MovieSceneScriptingFloatChannel, const [0,2]}
         ]
        }]
      }
      {
       type=MovieSceneFloatTrack
       display_name=Render Opacity
       sections [{type=MovieSceneFloatSection, end_frame=4, start_frame=0, channels [None {type=MovieSceneScriptingFloatChannel, keys [[0,0],[4,1]]}]}]
      }
     ]
    }]}
 }
 Out {
  binding_count=1
  display_label=""
  length_seconds=0.2
  object=&self:Out
  sequence {bindings [Overlay_Shell {
     id=#2
     possessed_class=Overlay
     tracks [
      {
       type=MovieScene2DTransformTrack
       sections [{
         type=MovieScene2DTransformSection
         end_frame=4
         start_frame=0
         channels [
          Translation.X {type=MovieSceneScriptingFloatChannel, keys [[0,0],[4,-1000]]}
          Translation.Y {type=MovieSceneScriptingFloatChannel, const [0,2]}
         ]
        }]
      }
      {
       type=MovieSceneFloatTrack
       display_name=Render Opacity
       sections [{type=MovieSceneFloatSection, end_frame=4, start_frame=0, channels [None {type=MovieSceneScriptingFloatChannel, keys [[0,1],[4,0]]}]}]
      }
     ]
    }]}
 }
 ShowTextSearch {
  binding_count=4
  length_seconds=0.2
  object=&self:ShowTextSearch
  sequence {bindings [
    SizeBox_TextSearch {id=#3, possessed_class=SizeBox, tracks []}
    WidgetSwitcher_ListViews {id=#4, possessed_class=CommonWidgetSwitcherLegacy, tracks []}
    VerticalBoxSlot_2 {
     id=#5
     parent_id=#4
     possessed_class=VerticalBoxSlot
     tracks [{
       type=MovieSceneMarginTrack
       display_name=Padding
       sections [{
         type=MovieSceneMarginSection
         end_frame=4
         start_frame=0
         channels [
          Left {type=MovieSceneScriptingFloatChannel}
          Top {type=MovieSceneScriptingFloatChannel, keys [[0,0],[4,92]]}
          Right {type=MovieSceneScriptingFloatChannel}
          Bottom {type=MovieSceneScriptingFloatChannel}
         ]
        }]
      }]
    }
    OverlaySlot_2 {
     id=#6
     parent_id=#3
     possessed_class=OverlaySlot
     tracks [{
       type=MovieSceneMarginTrack
       display_name=Padding
       sections [{
         type=MovieSceneMarginSection
         end_frame=4
         start_frame=0
         channels [
          Left {type=MovieSceneScriptingFloatChannel}
          Top {type=MovieSceneScriptingFloatChannel, keys [[0,86],[4,176]]}
          Right {type=MovieSceneScriptingFloatChannel}
          Bottom {type=MovieSceneScriptingFloatChannel}
         ]
        }]
      }]
    }
   ]}
 }
 ShowResetMessage {
  binding_count=1
  length_seconds=2
  object=&self:ShowResetMessage
  sequence {
   bindings [Overlay_ResetMessage {
     id=#7
     possessed_class=Overlay
     tracks [
      {
       type=MovieScene2DTransformTrack
       sections [{
         type=MovieScene2DTransformSection
         end_frame=21
         start_frame=0
         channels [
          Angle {type=MovieSceneScriptingFloatChannel}
          Scale.X {type=MovieSceneScriptingFloatChannel}
          Scale.Y {type=MovieSceneScriptingFloatChannel, keys [[0,0.01],[6,2],[12,1]]}
          Shear.X {type=MovieSceneScriptingFloatChannel}
          Shear.Y {type=MovieSceneScriptingFloatChannel}
         ]
        }]
      }
      {
       type=MovieSceneByteTrack
       display_name=Visibility
       sections [{type=MovieSceneByteSection, end_frame=120, start_frame=0, channels [None {type=MovieSceneScriptingByteChannel, keys [[0,4],[120,1]]}]}]
      }
      {
       type=MovieSceneFloatTrack
       display_name=Render Opacity
       sections [{
         type=MovieSceneFloatSection
         end_frame=120
         start_frame=0
         channels [None {type=MovieSceneScriptingFloatChannel, keys [[0,0],[6,1],[90,1],[120,0]]}]
        }]
      }
     ]
    }]
   display_rate {numerator=60}
  }
 }
]
graphs [
 ApplyFilterAndRefresh {nodes [
   n1 = ApplyFilterAndRefresh!() -> (then, ClearFilter:bool=~, TextToFilter:Text) {is=title K2Node_FunctionEntry}
   > n2 = Branch(Condition=@n1)
   > n3 = Set FilterText(FilterText:Text=GetEmptyText!() -> (ReturnValue:Text)) -> (Output_Get:Text)
   > SetInputFieldText!(
    execute
    self:WBP UIKit Filtered Input Field Object Reference=Get EditableText_Search() -> (EditableText_Search:WBP UIKit Filtered Input Field Object Reference)
    Text:Text=@n3
   ) -> (then)
   > RegenerateAllEntries(self:List View Base Object Reference=Get OptionsList() -> (OptionsList:Common List View Object Reference))
   > OnSearchTextChanged!(execute, self, Text:"Text (by ref)"=Get FilterText() -> (FilterText:Text)) -> (then)
   RegenerateAllEntries(execute=@n2.else, self=Get OptionsList())
  ]}
 AreAnyEntriesVisible {nodes [
   AreAnyEntriesVisible() {is=title K2Node_FunctionEntry}
   n1 = Get OptionsList()
   > n2 = Branch(Condition="Equal (Integer)"(A=GetNumItems!(self:List View Object Reference=@n1) -> (ReturnValue:int=0), B=0) {is=title K2Node_CallFunction})
   > Return Node!(execute, AnyEntryVisible:bool=false)
   Return Node(execute=@n2.else, AnyEntryVisible=true)
   n3 = GetDisplayedEntryWidgets!(self:List View Base Object Reference=@n1) -> (ReturnValue:Array of User Widget Object References)
   n4 = For Each Loop(Array:Array of User Widget Object References=@n3) -> (Array Element:User Widget Object Reference)
   Return Node(execute=@n4.Completed, AnyEntryVisible=false)
   > n5 = Switch on ESlateVisibility!(execute, Selection:ESlateVisibility Enum=GetVisibility!(self:Widget Object Reference=@n4) -> (ReturnValue:ESlateVisibility Enum=Visible)) -> (Visible:Exec, Collapsed:Exec, Hidden:Exec, HitTestInvisible:Exec, SelfHitTestInvisible:Exec) {is=title K2Node_SwitchEnum}
   Return Node(execute=[@n5.HitTestInvisible, @n5.SelfHitTestInvisible, @n5.Visible], AnyEntryVisible=true)
   Length(TargetArray:Array of User Widget Object References=@n3)
  ]}
 BindFilterBarEvents {nodes [
   BindFilterBarEvents() {is=title K2Node_FunctionEntry}
   > n1 = Sequence()
   n2 = Get CreativePropertyEditorFilterWidget() -> (CreativePropertyEditorFilterWidget:Creative Property Editor Filter Widget Object Reference)
   > Bind Event to On Text Search Deselected(self:Creative Property Editor Filter Widget Object Reference=@n2, Delegate=Create Event()) {is=title K2Node_AddDelegate}
   > Bind Event to On Text Search Selected(self:Creative Property Editor Filter Widget Object Reference=@n2, Delegate=Create Event()) {is=title K2Node_AddDelegate}
   > Bind Event to On Selected Filter Changed(self:Fort Creative Property Editor Filter Object Reference=@n2, Delegate=Create Event()) {is=title K2Node_AddDelegate}
   HandleSelectedFilterChanged!(
    execute=@n1.then_1
    self
    SelectedFilterId:Name=GetSelectedFilterButtonId!(self:Fort Creative Property Editor Filter Object Reference=Get CreativePropertyEditorFilterWidget()) -> (ReturnValue:Name=None)
    PreviouslySelectedFilterId:Name=None
   ) -> (then)
  ]}
 CanShowRenameButton {nodes [
   CanShowRenameButton() {is=title K2Node_FunctionEntry}
   n1 = Get Label Override Interface!(self:Fort Matchmaking Knobs Modal Object Reference) -> (ReturnValue:Fort Actor Label Override Interface Interface) {is=title K2Node_CallFunction}
   > n2 = Is Valid(InputObject="To Object (Interface)"!(Interface:"Interface Interface (by ref)"=@n1) -> (ReturnValue:Object Reference))
   Return Node!(execute=@n2.Is Not Valid, ShowRenameButton:bool=false)
   n3 = IsInCabinMode!() -> (ReturnValue:bool)
   > Return Node(ShowRenameButton=AND Boolean(A=CanOverrideLabel!(self:Fort Actor Label Override Interface Interface=@n1) -> (ReturnValue:bool), B=NOT Boolean(A=@n3)))
  ]}
 DirectSetSearchTextVisibility {nodes [
   DirectSetSearchTextVisibility!() -> (then, Show:bool=~) {is=title K2Node_FunctionEntry}
   > n1 = Sequence()
   > SetVisibility(
    self=[
     Get Spacer_TextSearch() -> (Spacer_TextSearch:Spacer Object Reference)
     Get SizeBox_TextSearch() -> (SizeBox_TextSearch:Size Box Object Reference)
    ]
    InVisibility=Select(Option 0=Collapsed, Option 1=SelfHitTestInvisible, Index=Get Show() -> (Show:bool))
   )
   InvalidateLayoutAndVolatility(execute=@n1.then_1, self:Widget Object Reference)
  ]}
 EventGraph {nodes [
   n1 = "On Entry Generated (OptionsList)"!() -> (OutputDelegate, then, Widget:User Widget Object Reference) {is=title K2Node_ComponentBoundEvent}
   > n2 = Cast To CreativePropertyEditorEntry!(execute, Object=@n1.Widget) -> (then, CastFailed, AsCreative Property Editor Entry:Creative Property Editor Entry Object Reference) {is=title K2Node_DynamicCast}
   > PlayAnimationForward(self=@n2, InAnimation=Get In2!(self:Creative Property Editor Entry Object Reference=@n2) -> (In2:Widget Animation Object Reference))
   > Update!(
    execute
    self:Creative Property Editor Entry Object Reference=@n2
    isModifiedFromDefaultValue:bool="Equal (Enum)"(
     A:EFortModifiedStatus Enum=GetSpinnerModificationStatus!(self:Fort Matchmaking Knobs Modal Object Reference, SpinnerButton:Fort Matchmaking Knobs Spinner Button Object Reference=@n2) -> (ReturnValue:EFortModifiedStatus Enum=IsDefault)
     B:EFortModifiedStatus Enum=IsModified
    )
   ) -> (then)
   n3 = Cast To CreativePropertyEditorEntry_V2!(execute=@n2.CastFailed, Object=@n1.Widget) -> (then, CastFailed, AsCreative Property Editor Entry V2:Creative Property Editor Entry V2 Object Reference) {is=title K2Node_DynamicCast}
   > Unbind all Events from On Knob Entry Edit Widget Context Tags Changed!(execute:Exec, self:Fort Creative Property Knob Entry Object Reference=@n3) -> (then:Exec) {is=title K2Node_ClearDelegate}
   > Bind Event to On Knob Entry Edit Widget Context Tags Changed(self:Fort Creative Property Knob Entry Object Reference=@n3, Delegate=Create Event()) {is=title K2Node_AddDelegate}
   n4 = GetListItemObject!(UserObjectListEntry:User Object List Entry Interface=@n3) -> (ReturnValue:Object Reference)
   > UpdateModifiedState!(
    execute
    self:Creative Property Editor Entry V2 Object Reference=@n3
    isModifiedFromDefaultValue:bool="Equal (Enum)"(
     A=GetModificationStatus!(self:Fort Matchmaking Knobs Modal Object Reference, Option:Object Reference=@n4) -> (ReturnValue:EFortModifiedStatus Enum=IsDefault)
     B=IsModified
    )
   ) -> (then)
   n5 = "On Clicked (OverrideCancelButton)"!() -> (OutputDelegate, then, Button:Common Button Base Object Reference) {is=title K2Node_ComponentBoundEvent}
   n6 = Branch(
    execute=[@n5.then, @n78.then]
    Condition=IsOpen!(self:Menu Anchor Object Reference=Get MenuAnchor_MoreOptions() -> (MenuAnchor_MoreOptions:Menu Anchor Object Reference)) -> (ReturnValue:bool)
   )
   > Close(self:Menu Anchor Object Reference=Get MenuAnchor_MoreOptions())
   HandleCloseWithAnimation!(execute=@n6.else, self, bForceClose:bool=false) -> (then)
   n7 = "On Item Selection Changed (OptionsList)"!() -> (OutputDelegate, then, Item:Object Reference, bIsSelected:bool) {is=title K2Node_ComponentBoundEvent}
   > Branch(Condition=Get isUsingGamePad() -> (isUsingGamePad:bool))
   > Delay(Duration=0)
   > Delay(Duration=0)
   > Scroll Item Into View!(execute, self:List View Object Reference=Get OptionsList(), Item:Object Reference=@n7.Item) -> (then)
   "On Button Clicked (MoreActionsButton)"!() -> (OutputDelegate, then, Button:Common Button Legacy Object Reference) {is=title K2Node_ComponentBoundEvent}
   > ToggleOpen!(execute, self:Menu Anchor Object Reference=Get MenuAnchor_MoreOptions(), bFocusOnOpen:bool=true) -> (then)
   n8 = "On Clicked (Button_RightDialogMouseCapture)"() {is=title K2Node_ComponentBoundEvent}
   CloseDialog!(execute=[@n8.then, @n10.then], self:Fort Creative Message Knobs Dialog Object Reference, bSetFocusToOptionsList:bool=true) -> (then)
   "On Clicked (Button_RenameInputButton)"!() -> (OutputDelegate, then, Button:Common Button Base Object Reference) {is=title K2Node_ComponentBoundEvent}
   > NativeHandleOnRenameInput(self:Fort Creative Message Knobs Dialog Object Reference)
   n9 = Get WBP_PropertyEditor_NameEntryWidget() -> (WBP_PropertyEditor_NameEntryWidget:WBP UIKit Filtered Input Field Object Reference)
   > CenterWidget!(execute, ToCenter:Widget Object Reference=@n9) -> (then)
   > SetFocus(self=Get InputField!(self:WBP UIKit Filtered Input Field Object Reference=@n9) -> (InputField:WBP UIKit Input Field Object Reference))
   n10 = "On Clicked (Button_LeftDialogMouseCapture)"() {is=title K2Node_ComponentBoundEvent}
   "On Clicked (Button_EventBrowser)"!() -> (OutputDelegate, then, Button:Common Button Base Object Reference) {is=title K2Node_ComponentBoundEvent}
   > HandleDeviceBrowserButtonPressed(self:Fort Creative Message Knobs Dialog Object Reference)
   n11 = "On Clicked (OK_Button)"!() -> (OutputDelegate, then, Button:Common Button Base Object Reference) {is=title K2Node_ComponentBoundEvent}
   n12 = CheckDestructiveSettings!(execute=[@n11.then, @n53.then], self:Fort Creative Message Knobs Dialog Object Reference) -> (then, ReturnValue:bool)
   > n13 = Branch(Condition=@n12)
   Set bApplyChanges(execute=[@n60.then, @n13.else], bApplyChanges:bool=true) -> (Output_Get:bool)
   > HandleCloseWithAnimation(bForceClose=true)
   n14 = "On Text Changed (EditableText_Search)"!() -> (OutputDelegate, then, Text:Text) {is=title K2Node_ComponentBoundEvent}
   > n15 = Set FilterText(FilterText=@n14.Text)
   n16 = Get FilterText()
   n17 = Branch(execute=[@n49.then, @n15.then], Condition=NOT Boolean(A=TextIsEmpty(InText=@n16)))
   > n18 = Clear and Invalidate Timer by Handle(Handle=Get TextSearchTimer() -> (TextSearchTimer:Timer Handle Structure))
   n19 = HandleSearchTimer() {is=title K2Node_CustomEvent}
   n20 = Set Timer by Event(execute=@n18.then, Delegate=@n19, Time=Get SearchDelay() -> (SearchDelay:double), bLooping=false)
   > Set TextSearchTimer(TextSearchTimer:Timer Handle Structure=@n20) -> (Output_Get:Timer Handle Structure)
   RemoveStringFilter!(execute=@n17.else, self:Fort Creative Message Knobs Dialog Object Reference, UpdateSelection:bool=false) -> (then)
   > GenerateFilteredEntries!(
    execute
    self:Fort Creative Message Knobs Dialog Object Reference
    bUpdateSelectionForGamepad:bool=Get UpdateSelectionOnListRegeneration() -> (UpdateSelectionOnListRegeneration:bool)
    bUpdateSelectionForMouse:bool=false
   ) -> (then)
   n21 = CreativeModeEnabledChanged!() -> (OutputDelegate, then, bIsEnabled:bool=~) {is=title K2Node_CustomEvent}
   > n22 = Branch(Condition=@n21.bIsEnabled)
   HandleCloseWithAnimation(execute=@n22.else, bForceClose=true)
   HandleRestoreAction() {is=title K2Node_CustomEvent}
   > StartRestoreAction!(execute, self, InText:Text=Get ResetMessage() -> (ResetMessage:Text)) -> (then)
   > RestoreOptionsToDefault!(execute, self:Fort Matchmaking Knobs Modal Object Reference, bRestoreAllOptions:bool=true) -> (then)
   HandleClearItemsAction() {is=title K2Node_CustomEvent}
   > Close(self=Get MenuAnchor_MoreOptions())
   > n23 = Branch(Condition=HasClearItemsConfirmation!(self:Fort Creative Message Knobs Dialog Object Reference) -> (ReturnValue:bool))
   > SetVisibility(InVisibility=Hidden)
   > n24 = ShowSimpleConfirmationDialog_CustomInput!(
    execute:Exec
    Icon:Slate Brush Structure
    Title:Text="LOCTEXT(\"Warning\")"
    Message:Text=GetClearItemsConfirmationText!(self:Fort Creative Message Knobs Dialog Object Reference) -> (ReturnValue:Text)
    ConfirmAction:Name=Confirm
    DeclineAction:Name=Cancel
    AdditionalContent:Widget Object Reference
    LeftAdditionalContent:Widget Object Reference
    Cancelable:bool=false
    TargetPlayer:Local Player Object Reference
   ) -> (then:Exec, DialogResult:Exec, Result:EFortDialogResult Enum, ResultName:Name) {is=title K2Node_AsyncAction}
   SetVisibility(execute=@n24.DialogResult, InVisibility=Visible)
   > n25 = Branch(Condition="Equal (Enum)"(A:EFortDialogResult Enum=@n24, B:EFortDialogResult Enum=Confirmed))
   "SetText (Text)"(
    execute=[@n25.then, @n23.else]
    self=Get Text_Message() -> (Text_Message:Common Text Object Reference)
    InText=Get ClearMessage() -> (ClearMessage:Text)
   )
   > PlayAnimationForward(InAnimation=Get ShowResetMessage() -> (ShowResetMessage:Widget Animation Object Reference))
   > HandleClearItemsClicked(self:Fort Creative Message Knobs Dialog Object Reference)
   HandleCloseMoreOptionsAction() {is=title K2Node_CustomEvent}
   > Close(self=Get MenuAnchor_MoreOptions())
   n26 = HandleRename!() -> (OutputDelegate, then, Text:"Text (by ref)", CommitMethod:ETextCommit Enum=Default) {is=title K2Node_CustomEvent}
   > SetTextWithForcedSanitization!(
    execute
    self:Fort Editable Filtered Counted Text Box Object Reference=Get WBP_PropertyEditor_NameEntryWidget()
    Text:"Text (by ref)"=@n26.Text
   ) -> (then)
   > n27 = Set ItemName(ItemName:Text=@n26.Text) -> (Output_Get:Text)
   > SetCurrentName!(
    execute
    self:Cr Property Editor More Options Object Reference=Get MoreOptionsMenu() -> (MoreOptionsMenu:Cr Property Editor More Options Object Reference)
    ObjectName:Text=@n27
   ) -> (then)
   n28 = Get Label Override Interface()
   > n29 = Is Valid(InputObject="To Object (Interface)"(Interface=@n28))
   > n30 = SetLabelOverride!(
    execute
    self:Fort Actor Label Override Interface Interface=@n28
    PlayerController:Player Controller Object Reference=GetOwningPlayer()
    NewLabelOverride:String="To String (Text)"(InText=@n27)
   ) -> (then, ReturnValue:bool)
   Delay(execute=[@n30.then, @n29.Is Not Valid], Duration=0)
   > Close(self=Get MenuAnchor_MoreOptions())
   HandleRestoreDefaultName() {is=title K2Node_CustomEvent}
   n31 = Get Label Override Interface()
   > n32 = Is Valid(InputObject="To Object (Interface)"(Interface=@n31))
   > ResetLabelOverride!(execute, self:Fort Actor Label Override Interface Interface=@n31, PlayerController:Player Controller Object Reference=GetOwningPlayer()) -> (then, ReturnValue:bool)
   > n33 = SetTextWithForcedSanitization(
    self=Get WBP_PropertyEditor_NameEntryWidget()
    Text=GetDisplayLabel!(self:Fort Actor Label Override Interface Interface=@n31) -> (ReturnValue:Text)
   )
   Close(execute=[@n33.then, @n32.Is Not Valid], self=Get MenuAnchor_MoreOptions())
   UnregisterGameplayMessages() {is=title K2Node_CustomEvent}
   n34 = Get SpatialLoadingStateUpdatedHandle() -> (SpatialLoadingStateUpdatedHandle:Async Action Start Listening To Event Object Reference)
   > Is Valid(InputObject=@n34)
   > StopListeningToEvent(self:Async Action Start Listening To Event Object Reference=@n34)
   n35 = HandleInputMethodChanged!() -> (OutputDelegate, then, bUsingGamepad:bool=~) {is=title K2Node_CustomEvent}
   n36 = Get Action_SelectedItemAdditionalAction() -> (Action_SelectedItemAdditionalAction:Common Action Widget Object Reference)
   SetVisibility(
    execute=[@n76.then, @n35.then, @n59.then]
    self=Get HBox_SelectedItemAdditionalAction() -> (HBox_SelectedItemAdditionalAction:Horizontal Box Object Reference)
    InVisibility=GetDesiredAdditionalActionButtonsVisiblity!(
     self:Fort Creative Message Knobs Dialog Object Reference
     RequestedVisibility:ESlateVisibility Enum=GetVisibility(self=@n36) -> (ReturnValue=Visible)
    ) -> (ReturnValue:ESlateVisibility Enum=Visible)
   )
   n37 = OnRefreshTagFilter() {is=title K2Node_CustomEvent}
   RefreshTagFilterForGrantedContextTags!(
    execute=[@n37.then, @n54.else]
    self:Fort Creative Message Knobs Dialog Object Reference
    bRefreshOnlyWhenChanged:bool=true
    bUpdateSelectionForGamepad:bool=true
    bUpdateSelectForMouse:bool=true
   ) -> (then)
   n38 = LeaveIsland!() -> (OutputDelegate, then, PlayerPawn:Fort Player Pawn Object Reference) {is=title K2Node_CustomEvent}
   > HandleCloseWithAnimation(bForceClose=true)
   HandleResetPropertiesAction() {is=title K2Node_CustomEvent}
   > StartRestoreAction(InText=Get ResetPropertiesMessage() -> (ResetPropertiesMessage:Text))
   > HandleRestorePropertiesClicked(self:Fort Matchmaking Knobs Modal Object Reference)
   n39 = ForcedOffIsland() {is=title K2Node_CustomEvent}
   > HandleCloseWithAnimation(bForceClose=true)
   HandleResetFunctionsAction() {is=title K2Node_CustomEvent}
   > StartRestoreAction(InText=Get ResetFunctionsMessage() -> (ResetFunctionsMessage:Text))
   > HandleRestoreFunctionsClicked(self:Fort Matchmaking Knobs Modal Object Reference)
   n40 = StartRestoreAction!() -> (OutputDelegate, then, InText:Text) {is=title K2Node_CustomEvent}
   > "SetText (Text)"(self=Get Text_Message(), InText=@n40.InText)
   > PlayAnimationForward(InAnimation=Get ShowResetMessage())
   > Close(self=Get MenuAnchor_MoreOptions())
   SetupGameplayMessages() {is=title K2Node_CustomEvent}
   > n41 = StartListeningToEventBackwardCompatible!(
    execute:Exec
    Channel:Event Message Tag Structure=(TagName="EventMessage.Component.Playspace.SpatialGameplayActorLoadingState.Updated.Player")
    Target:Gameplay Event Router Component Object Reference=GetGlobalEventRouter()
    EventType:Script Struct Object Reference=/Script/FortniteGame.SpatialGameplayActorLoadingStateUpdatedPlayerContext
    Context:Object Reference
    EventBubblingRule:EEventBubblingRule Enum=Down
   ) -> (
    then:Exec
    AsyncTaskProxy:Async Action Start Listening To Event Object Reference
    OnEventReceived:Exec
    Payload:Spatial Gameplay Actor Loading State Updated Player Context Structure
   ) {is=title K2Node_StartListeningToEventAsyncAction}
   > Set SpatialLoadingStateUpdatedHandle(SpatialLoadingStateUpdatedHandle:Async Action Start Listening To Event Object Reference=@n41) -> (Output_Get:Async Action Start Listening To Event Object Reference)
   HandleCloseWithAnimation(execute=@n41.OnEventReceived, bForceClose=true)
   n42 = HandleCloseWithAnimation!() -> (OutputDelegate, then, bForceClose:bool=~) {is=title K2Node_CustomEvent}
   > n43 = Branch(Condition=AND Boolean(A=Get TextSearchIsOpen() -> (TextSearchIsOpen:bool), B=NOT Boolean(A=@n42.bForceClose)))
   > SetSearchButtonState!(execute, self:Creative Property Editor Filter Widget Object Reference=Get CreativePropertyEditorFilterWidget(), bIsSelected:bool=false) -> (then)
   > Set TextSearchIsOpen(TextSearchIsOpen:bool=false) -> (Output_Get:bool)
   > HandleHideTextSearch()
   n44 = GetCurrentInputType(self=Get CommonInputSubsystem())
   > Branch(Condition=OR Boolean(A="Equal (Enum)"(A:ECommonInputType Enum=@n44, B:ECommonInputType Enum=MouseAndKeyboard), B="Equal (Enum)"(A=@n44, B=Touch)))
   > HandleCloseWithAnimation(bForceClose=true) {is=title K2Node_CallFunction}
   n45 = Get Out() -> (Out:Widget Animation Object Reference)
   n46 = Branch(execute=@n43.else, Condition=IsAnimationPlaying(InAnimation=@n45))
   PlayAnimation(execute=@n46.else, InAnimation=@n45)
   n47 = MinigameStateChanged!() -> (OutputDelegate, then, Minigame:Fort Minigame Object Reference, MinigameState:EFortMinigameState Enum=PreGame) {is=title K2Node_CustomEvent}
   > HandleCloseWithAnimation(bForceClose=true)
   n48 = OnSearchTextChanged!() -> (OutputDelegate, then, Text:"Text (by ref)") {is=title K2Node_CustomEvent}
   > n49 = Set FilterText(FilterText=@n48.Text)
   n50 = OnFilterCompleted!() -> (OutputDelegate, then, NumberOfItemsInFilter:int) {is=title K2Node_CustomEvent}
   > n51 = Branch(Condition="integer <= integer"!(A:int=@n50.NumberOfItemsInFilter, B:int=0) -> (ReturnValue:bool) {is=title K2Node_CallFunction})
   n52 = Get WidgetSwitcher_ListViews() -> (WidgetSwitcher_ListViews:Common Widget Switcher Legacy Object Reference)
   > SetActiveWidgetIndex(self=@n52, Index=1)
   SetActiveWidgetIndex(execute=@n51.else, self=@n52, Index=0)
   > Delay(Duration=0)
   > Delay(Duration=0)
   > Delay(Duration=0)
   > ForceFocusOnCurrentlySelectedItem()
   n53 = KB Confirm Rename() {is=title K2Node_CustomEvent}
   HandleResetEventsAction() {is=title K2Node_CustomEvent}
   > StartRestoreAction(InText=Get ResetEventsMessage() -> (ResetEventsMessage:Text)) {is=title K2Node_CallFunction}
   > HandleRestoreEventsClicked(self:Fort Matchmaking Knobs Modal Object Reference)
   OnKnobGrantedContextTagsChanged() {is=title K2Node_CustomEvent}
   > n54 = Branch(Condition="float > float"(A=Get RecheckContextTagsTimeout() -> (RecheckContextTagsTimeout:double)))
   > Clear and Invalidate Timer by Handle(Handle=Get RecheckContextTagsTimerHandle() -> (RecheckContextTagsTimerHandle:Timer Handle Structure))
   > n55 = Set Timer by Event(Delegate=@n37, Time=Get RecheckContextTagsTimeout(), bLooping=false)
   > Set RecheckContextTagsTimerHandle(RecheckContextTagsTimerHandle:Timer Handle Structure=@n55) -> (Output_Get:Timer Handle Structure)
   n56 = Event OnDialogOpenStateChanged!() -> (OutputDelegate, then, bIsOpen:bool) {is=title K2Node_Event}
   > SetVisibility(
    self=[
     Get Button_LeftDialogMouseCapture() -> (Button_LeftDialogMouseCapture:Button Object Reference)
     Get Button_RightDialogMouseCapture() -> (Button_RightDialogMouseCapture:Button Object Reference)
    ]
    InVisibility=Select(Option 0=Collapsed, Option 1=Visible, Index=@n56.bIsOpen)
   )
   n57 = Event OnInitForUserOptionValues!() -> (OutputDelegate, then, UserOptionValues:"User Option Values Interface Interface (by ref)") {is=title K2Node_Event}
   > Parent: OnInitForUserOptionValues!(execute, UserOptionValues:"User Option Values Interface Interface (by ref)"=@n57.UserOptionValues) -> (then) {is=title K2Node_CallParentFunction}
   n58 = Get Label Override Interface()
   > SetItemInfo!(
    execute
    self
    ItemName:Text=GetDisplayLabel(self=@n58)
    ItemBlock:Text=GetBaseLabel!(self:Fort Actor Label Override Interface Interface=@n58) -> (ReturnValue:Text)
    ItemDescription:Text=GetUserOptionsDescription!(self:User Option Definition Container Interface Interface=Get User Option Definitions!(self:User Option Values Interface Interface=@n57.UserOptionValues) -> (ReturnValue:User Option Definition Container Interface Interface) {is=title K2Node_CallFunction}) -> (ReturnValue:Text)
   ) -> (then)
   n59 = Event OnEntryAdditionalActionChanged() {is=title K2Node_Event}
   n60 = Event BP_DialogConfirmed() {is=title K2Node_Event}
   n61 = Event OnShowLoadingSpinnerChanged!() -> (OutputDelegate, then, bShowLoadingSpinner:bool) {is=title K2Node_Event}
   > SetVisibility(
    self=Get Overlay_LoadingSpinner() -> (Overlay_LoadingSpinner:Overlay Object Reference)
    InVisibility=Select(Option 0=Collapsed, Option 1=Visible, Index=@n61.bShowLoadingSpinner)
   )
   > SetRenderOpacity(
    self=Get WidgetSwitcher_ListViews()
    InOpacity=Select(Option 0:float=1, Option 1:float=0, Index=@n61.bShowLoadingSpinner) -> (ReturnValue:float)
   )
   Event OnAddedToFocusPath!() -> (OutputDelegate, then, InFocusEvent:Focus Event Structure) {is=title K2Node_Event}
   > Set bIsOnFocusPath(bIsOnFocusPath:bool=true) -> (Output_Get:bool)
   Event OnRemovedFromFocusPath!() -> (OutputDelegate, then, InFocusEvent:Focus Event Structure) {is=title K2Node_Event}
   > Set bIsOnFocusPath(bIsOnFocusPath=false)
   n62 = Event OnInitForOptionsComponent!() -> (OutputDelegate, then, OptionsComponent:Fort Actor Options Component Object Reference) {is=title K2Node_Event}
   > Is Valid(InputObject=@n62.OptionsComponent)
   > n63 = Cast To BuildingActor!(execute, Object=GetOwner(self=@n62.OptionsComponent)) -> (then, CastFailed, AsBuilding Actor:Building Actor Object Reference) {is=title K2Node_DynamicCast}
   > SetItemInfo(
    ItemName=GetDisplayLabel(self=@n63)
    ItemBlock=GetBaseLabel(self=@n63)
    ItemDescription=GetDescription!(self:Building Actor Object Reference=@n63) -> (ReturnValue:Text)
   )
   n64 = Event InitFromObject!() -> (OutputDelegate, then, InitObject:Object Reference) {is=title K2Node_Event}
   > n65 = HasUEFNOnlyOptions!(execute:Exec, self:Object Reference=@n64.InitObject) -> (then:Exec, HasUEFNOnlyOptions:bool, TooltipText:Text) {is=title K2Node_Message}
   > HandleUEFNOnlyOption!(execute, self, bHasUEFNOnlyOption:bool=@n65, TooltipText:Text=@n65.TooltipText) -> (then)
   > n66 = Cast To Actor!(execute, Object=@n64.InitObject) -> (then, CastFailed, AsActor:Actor Object Reference) {is=title K2Node_DynamicCast}
   > InitForToyOptionsActor!(execute, self:Fort Matchmaking Knobs Modal Object Reference, InActor:Actor Object Reference=@n66) -> (then)
   > InitializeData(self:Fort Matchmaking Knobs Modal Object Reference)
   Event On Activated()
   > Bind Event to On Input Method Changed(self=Get CommonInputSubsystem(), Delegate=Create Event())
   > SetupGameplayMessages() {is=title K2Node_CallFunction}
   > SetVisibility(self=Get OptionsList(), InVisibility=Hidden)
   > PlayAnimation(InAnimation=Get In() -> (In:Widget Animation Object Reference))
   > PlaySound2D(Sound=Get Open Menu Sound() -> (Open Menu Sound:Sound Base Object Reference))
   > n67 = Cast To FortPlayerControllerAthena(Object=GetOwningPlayer())
   > Bind Event to State Changed Delegate(
    self:Fort Minigame Object Reference=GetMinigame!(ContextActor:Actor Object Reference=@n67) -> (ReturnValue:Fort Minigame Object Reference)
    Delegate=@n47
   ) {is=title K2Node_AddDelegate}
   > Assign On Client Forced Off Island!(execute:Exec, self:Fort Player Controller Athena Object Reference=@n67, Delegate:"Delegate (by ref)"=@n39) -> (then:Exec) {is=title K2Node_AssignDelegate}
   > Assign On Client Leave Island!(execute:Exec, self:Fort Player Controller Athena Object Reference=@n67, Delegate:"Delegate (by ref)"=@n38) -> (then:Exec) {is=title K2Node_AssignDelegate}
   > Bind Event to On Creative Mode Enabled Changed Delegate(self:Fort Player Controller Gameplay Object Reference=@n67, Delegate=@n21) {is=title K2Node_AddDelegate}
   > SetHUDElementVisibilityWithContext!(
    execute
    self:Fort Player Controller Object Reference=@n67
    HUDElementTags:"Gameplay Tag Container Structure (by ref)"=Get HideGameplayTags() -> (HideGameplayTags:Gameplay Tag Container Structure)
    bHideElements:bool=true
    ReasonName:Name=None
    LogContext:String
   ) -> (then)
   K2Node_Event_4 = Handle Input Mode Changed!() -> (OutputDelegate, then, InputType:ECommonInputType Enum) {is=title K2Node_CustomEvent}
   n68 = "Equal (Enum)"(A=@K2Node_Event_4.InputType, B=Gamepad)
   > n69 = Branch(Condition=AND Boolean(A=Get bIsOnFocusPath() -> (bIsOnFocusPath:bool), B=@n68))
   > n70 = NavigateToIndex!(execute, self:List View Object Reference=Get OptionsList(), Index:int=0) -> (then)
   n71 = Set isUsingGamePad(execute=[@n70.then, @n69.else], isUsingGamePad:bool=@n68) -> (Output_Get:bool)
   > UpdateHoldText!(execute, self, ShowHold:bool=@n71) -> (then)
   Event On Deactivated()
   > Unbind Event from On Input Method Changed(self=Get CommonInputSubsystem(), Delegate=Create Event())
   > UnregisterGameplayMessages() {is=title K2Node_CallFunction}
   > PlaySound2D(Sound=Get Close Menu Sound() -> (Close Menu Sound:Sound Base Object Reference))
   > StopAllAnimations(self:User Widget Object Reference)
   > n72 = Branch(Condition=Get bApplyChanges() -> (bApplyChanges:bool))
   > Set bApplyChanges(bApplyChanges=false)
   > OnAcceptAnimationComplete!(execute, self:Fort Matchmaking Knobs Modal Object Reference, bApplyAllOptions:bool=true) -> (then)
   > n73 = HandleRename!(execute, self, Text:"Text (by ref)"=Get ItemName() -> (ItemName:Text), CommitMethod:ETextCommit Enum=Default) -> (then) {is=title K2Node_CallFunction}
   n74 = Cast To FortPlayerControllerAthena(execute=[@n73.then, @n72.else], Object=GetOwningPlayer())
   > SetHUDElementVisibilityWithContext(self=@n74, HUDElementTags=Get HideGameplayTags(), bHideElements=false, ReasonName=None)
   > Unbind Event from On Creative Mode Enabled Changed Delegate(self:Fort Player Controller Gameplay Object Reference=@n74, Delegate=Create Event()) {is=title K2Node_RemoveDelegate}
   > Unbind Event from On Client Forced Off Island(self:Fort Player Controller Athena Object Reference=@n74, Delegate=Create Event()) {is=title K2Node_RemoveDelegate}
   > Unbind Event from On Client Leave Island(self:Fort Player Controller Athena Object Reference=@n74, Delegate=Create Event()) {is=title K2Node_RemoveDelegate}
   > Unbind Event from State Changed Delegate(
    self:Fort Minigame Object Reference=GetMinigame!(self:Fort Player Controller Gameplay Object Reference=@n74) -> (ReturnValue)
    Delegate=Create Event()
   ) {is=title K2Node_RemoveDelegate}
   Event Construct()
   > BindFilterBarEvents() {is=title K2Node_CallFunction}
   > Bind Event to On Filter List Completed(self:Fort Creative Message Knobs Dialog Object Reference, Delegate=Create Event()) {is=title K2Node_AddDelegate}
   > Bind Event to On Menu Open Changed(self:Menu Anchor Object Reference=Get MenuAnchor_MoreOptions(), Delegate=Create Event()) {is=title K2Node_AddDelegate}
   n75 = Get WBP_PropertyEditor_NameEntryWidget()
   > Bind Event to On Sanitized Text Ready(self:WBP UIKit Filtered Input Field Object Reference=@n75, Delegate=Create Event()) {is=title K2Node_AddDelegate}
   > Bind Event to On Text Commit by Enter(self:WBP UIKit Filtered Input Field Object Reference=@n75, Delegate=Create Event()) {is=title K2Node_AddDelegate}
   > SetInputAction!(
    execute
    self:Common Action Widget Object Reference=Get Action_SelectedItemAdditionalAction()
    InputActionRow:Data Table Row Handle Structure
   ) -> (then)
   > "SetText (Text)"(self=Get Text_SelectedItemAdditionalAction() -> (Text_SelectedItemAdditionalAction:Common Text Object Reference), InText=GetEmptyText())
   > n76 = Bind Event to On Input Method Changed(self:Common Action Widget Object Reference=Get Action_SelectedItemAdditionalAction(), Delegate=@n35)
   n77 = Event SetOptionsButtonsEnabled!() -> (OutputDelegate, then, bInButtonsEnabled:bool) {is=title K2Node_Event}
   > SetIsEnabled(
    self=[
     Get OK_Button() -> (OK_Button:WBP UIKit Button Loud Object Reference)
     Get OverrideCancelButton() -> (OverrideCancelButton:WBP UIKit Button Regular Object Reference)
     Get MoreActionsButton() -> (MoreActionsButton:Icon Text Button Object Reference)
     Get HBox_SelectedItemAdditionalAction()
    ]
    bInIsEnabled=@n77.bInButtonsEnabled
   )
   n78 = Event HandleMainMenuInput() {is=title K2Node_Event}
   Event OnConfirmationDialogWidgetCreated() {is=title K2Node_Event}
   > n79 = GetConfirmationDialogWidget!(execute, self:Fort Creative Message Knobs Dialog Object Reference) -> (then, ReturnValue:Athena Confirmation Window Object Reference)
   > n80 = AddChildToOverlay!(
    execute
    self:Overlay Object Reference=Get Overlay_ConfirmDialog() -> (Overlay_ConfirmDialog:Overlay Object Reference)
    Content:Widget Object Reference=@n79
   ) -> (then, ReturnValue:Overlay Slot Object Reference)
   > SetHorizontalAlignment(self=@n80, InHorizontalAlignment=HAlign_Center)
   > SetVerticalAlignment(self=@n80, InVerticalAlignment=VAlign_Center)
   "Animation Finished (In)"!() -> (OutputDelegate:Delegate, then:Exec) {is=title K2Node_WidgetAnimationEvent}
   > SetVisibility(self=Get OptionsList(), InVisibility=Visible)
   n81 = Get CommonInputSubsystem!() -> (ReturnValue:Common Input Object Reference) {is=title K2Node_GetBlueprintContextEx}
   > n82 = Set isUsingGamePad(isUsingGamePad="Equal (Enum)"(A=GetCurrentInputType(self=@n81), B=Gamepad))
   > UpdateHoldText(ShowHold=@n82)
   > GenerateFilteredEntries(bUpdateSelectionForGamepad=Get UpdateSelectionOnListRegeneration(), bUpdateSelectionForMouse=false)
   > ForceFocusOnCurrentlySelectedItem()
   > Set UpdateSelectionOnListRegeneration(UpdateSelectionOnListRegeneration:bool=true) -> (Output_Get:bool)
   > Branch(Condition=Get isUsingGamePad())
   > SetSelectedIndex!(execute, self:List View Object Reference=Get OptionsList(), Index:int=0) -> (then)
   "Animation Finished (Out)"!() -> (OutputDelegate:Delegate, then:Exec) {is=title K2Node_WidgetAnimationEvent}
   > DeactivateWidget()
   Clear and Invalidate Timer by Handle(execute=@n84.then, Handle=Get TextSearchTimer())
   n83 = Get FilterText()
   n84 = FilterEntriesByString!(
    execute=@n86.else
    self:Fort Creative Message Knobs Dialog Object Reference
    FilterString:String="To String (Text)"(InText=@n83)
    UpdateSelection:bool=true
   ) -> (then)
   n85 = TextIsEmpty(InText=@n83)
   n86 = Branch(execute=@n19.then, Condition=@n85)
  ]}
 ForceFocusOnCurrentlySelectedItem {nodes [
   ForceFocusOnCurrentlySelectedItem() {is=title K2Node_FunctionEntry}
   n1 = Get Selected Item!(self:List View Object Reference=Get OptionsList()) -> (ReturnValue:Object Reference) {is=title K2Node_CallFunction}
   > Is Valid(InputObject=@n1)
   > SetFocusToItem!(execute, self:Fort Creative Message Knobs Dialog Object Reference, Item:Object Reference=@n1) -> (then)
  ]}
 GetMoreOptionsMenu {nodes [
   GetMoreOptionsMenu() {is=title K2Node_FunctionEntry}
   > n1 = Get!(execute:Exec) -> (then:Exec, else:Exec, MoreOptionsMenu:Cr Property Editor More Options Object Reference)
   n2 = Get MoreOptionsMenu()
   SetButtonVisibility!(
    execute=[@n5.then, @n1.then]
    self:Cr Property Editor More Options Object Reference=@n2
    bShowClearItemsButton:bool=ContainsItem!(self:Fort Creative Message Knobs Dialog Object Reference) -> (ReturnValue:bool)
    bShowResetButton:bool=true
    bShowRenameButton:bool=CanShowRenameButton!() -> (ShowRenameButton:bool) {is=title K2Node_CallFunction}
    ShowResetFunctionsButton:bool=HasEventFunctionProperty!(self:Fort Matchmaking Knobs Modal Object Reference) -> (ReturnValue:bool)
    ShowResetEventsButton:bool=HasEventDescriptorProperty!(self:Fort Matchmaking Knobs Modal Object Reference) -> (ReturnValue:bool)
    ShowResetPropertiesButton:bool=HasNonEventProperties!(self:Fort Matchmaking Knobs Modal Object Reference) -> (ReturnValue:bool)
   ) -> (then)
   > SetCurrentName(self=@n2, ObjectName=Get ItemName())
   > Return Node!(execute, ReturnValue:Widget Object Reference=@n2)
   n3 = Create Widget!(
    execute:Exec=@n1.else
    Class:User Widget Class Reference=&/Cr_PropertyEditor_MoreOptions.Cr_PropertyEditor_MoreOptions_C
    OwningPlayer:Player Controller Object Reference=GetOwningPlayer()
   ) -> (then:Exec, ReturnValue:Cr Property Editor More Options Object Reference)
   > n4 = Set MoreOptionsMenu(MoreOptionsMenu:Cr Property Editor More Options Object Reference=@n3) -> (Output_Get:Cr Property Editor More Options Object Reference)
   > Bind Event to Back Action(self:Cr Property Editor More Options Object Reference=@n4, Delegate=Create Event()) {is=title K2Node_AddDelegate}
   > Bind Event to Clear Items Action(self:Cr Property Editor More Options Object Reference=@n4, Delegate=Create Event()) {is=title K2Node_AddDelegate}
   > Bind Event to Default Name Action(self:Cr Property Editor More Options Object Reference=@n4, Delegate=Create Event()) {is=title K2Node_AddDelegate}
   > Bind Event to Rename Action(self:Cr Property Editor More Options Object Reference=@n4, Delegate=Create Event()) {is=title K2Node_AddDelegate}
   > Bind Event to Reset Action(self:Cr Property Editor More Options Object Reference=@n4, Delegate=Create Event()) {is=title K2Node_AddDelegate}
   > Bind Event to Reset Properties Action(self:Cr Property Editor More Options Object Reference=@n3, Delegate=Create Event()) {is=title K2Node_AddDelegate}
   > Bind Event to Reset Functions Action(self:Cr Property Editor More Options Object Reference=@n3, Delegate=Create Event()) {is=title K2Node_AddDelegate}
   > n5 = Bind Event to Reset Events Action(self:Cr Property Editor More Options Object Reference=@n3, Delegate=Create Event()) {is=title K2Node_AddDelegate}
  ]}
 GetOptionsListScrollSpeed {nodes [
   GetOptionsListScrollSpeed() {is=title K2Node_FunctionEntry}
   > Return Node(ReturnValue:float=Get WheelScrollMultiplier!(self:List View Base Object Reference=Get OptionsList()) -> (WheelScrollMultiplier:float))
  ]}
 Handle on Text Sanitized {nodes [
   n1 = Handle on Text Sanitized!() -> (then, Text:"Text (by ref)", IsBadText:bool=~) {is=title K2Node_FunctionEntry}
   > Set ItemName(ItemName=@n1)
   > Branch(Condition=Get bTextCommittedByEnter() -> (bTextCommittedByEnter:bool))
   > Set bTextCommittedByEnter(bTextCommittedByEnter:bool=false) -> (Output_Get:bool)
   > KB Confirm Rename() {is=title K2Node_CallFunction}
  ]}
 HandleFiltration {nodes [
   HandleFiltration!() -> (then, inRow:Creative Property Editor Entry Object Reference) {is=title K2Node_FunctionEntry}
   > n1 = Branch(Condition=TextIsEmpty(InText=Get FilterText()))
   n2 = Get inRow() -> (inRow:Creative Property Editor Entry Object Reference)
   > SetVisibility(self=@n2, InVisibility=SelfHitTestInvisible)
   n3 = GetTitleText!(self:Fort Matchmaking Knobs Spinner Button Object Reference=Get inRow()) -> (ReturnValue:Text)
   n4 = Get FilterText()
   n5 = Branch(execute=@n1.else, Condition=Contains(SearchIn="To String (Text)"(InText=@n3), Substring="To String (Text)"(InText=@n4)))
   > SetVisibility(self=@n2, InVisibility=SelfHitTestInvisible)
   SetVisibility(execute=@n5.else, self=@n2, InVisibility=Collapsed)
  ]}
 HandleHideTextSearch {nodes [
   HandleHideTextSearch() {is=title K2Node_FunctionEntry}
   > DirectSetSearchTextVisibility!(execute, self, Show:bool=false) -> (then) {is=title K2Node_CallFunction}
   > n1 = CloseDialog(bSetFocusToOptionsList=true)
   Set TextSearchIsOpen(execute=[@n3.then, @n1.then], TextSearchIsOpen=false)
   > ApplyFilterAndRefresh!(execute, self, ClearFilter:bool=true, TextToFilter:Text) -> (then) {is=title K2Node_CallFunction}
   n2 = Branch(Condition=IsAnimationPlaying(InAnimation=Get ShowTextSearch() -> (ShowTextSearch:Widget Animation Object Reference)))
   SetVisibility(execute=@n2.else, self=Get EditableText_Search(), InVisibility=Hidden)
   > n3 = PlayAnimationReverse(InAnimation=Get ShowTextSearch())
  ]}
 HandleOnEditWidgetOpened {nodes [HandleOnEditWidgetOpened!() -> (then, Opened:bool=~) {is=title K2Node_FunctionEntry}]}
 HandleSelectedFilterChanged {nodes [
   n1 = HandleSelectedFilterChanged!() -> (then, SelectedFilterId:Name, PreviouslySelectedFilterId:Name) {is=title K2Node_FunctionEntry}
   > n2 = Sequence()
   > n3 = Switch on Name!(execute:Exec, Selection:Name=@n1) -> (
    Default:Exec
    Filter_Basic:Exec
    Filter_Advanced:Exec
    Filter_Modified:Exec
    Filter_Channels:Exec
    Filter_EventFunctions:Exec
    Filter_EventDescriptors:Exec
   ) {
    is=title K2Node_SwitchName
    pinNames [Filter_Basic, Filter_Advanced, Filter_Modified, Filter_Channels, Filter_EventFunctions, Filter_EventDescriptors]
   }
   OnAdvancedFilterButtonSelectionChanged!(execute=@n3.Filter_Advanced, self, bSelected:bool=true) -> (then)
   OnChannelFilterButtonSelectionChanged!(execute=@n3.Filter_Channels, self, bSelected:bool=true) -> (then)
   OnEventDescriptorFilterButtonSelectionChanged!(execute=@n3.Filter_EventDescriptors, self, bSelected:bool=true) -> (then)
   OnEventFunctionFilterButtonSelectionChanged!(execute=@n3.Filter_EventFunctions, self, bSelected:bool=true) -> (then)
   OnModifiedFilterButtonSelectionChanged!(execute=@n3.Filter_Modified, self, bSelected:bool=true) -> (then)
   OnBasicFilterButtonSelectionChanged!(execute=[@n3.Default, @n3.Filter_Basic], self, bSelected:bool=true) -> (then)
   n4 = Switch on Name!(execute=@n2.then_1, Selection=@n1.PreviouslySelectedFilterId) -> (Filter_Basic, Filter_Advanced, Filter_Modified, Filter_Channels, Filter_EventFunctions, Filter_EventDescriptors) {
    bHasDefaultPin=false
    pinNames [Filter_Basic, Filter_Advanced, Filter_Modified, Filter_Channels, Filter_EventFunctions, Filter_EventDescriptors]
   }
   OnAdvancedFilterButtonSelectionChanged(execute=@n4.Filter_Advanced, bSelected=false)
   OnBasicFilterButtonSelectionChanged(execute=@n4.Filter_Basic, bSelected=false)
   OnChannelFilterButtonSelectionChanged(execute=@n4.Filter_Channels, bSelected=false)
   OnEventDescriptorFilterButtonSelectionChanged(execute=@n4.Filter_EventDescriptors, bSelected=false)
   OnEventFunctionFilterButtonSelectionChanged(execute=@n4.Filter_EventFunctions, bSelected=false)
   OnModifiedFilterButtonSelectionChanged(execute=@n4.Filter_Modified, bSelected=false)
  ]}
 HandleShowMoreOptions {nodes [
   HandleShowMoreOptions!() -> (then, bIsOpen:bool=~) {is=title K2Node_FunctionEntry}
   > SetVisibility(
    self=Get Image_LightBoxLeftSide() -> (Image_LightBoxLeftSide:Image Object Reference)
    InVisibility=Select(Option 0:ESlateVisibility Enum=Collapsed, Option 1:ESlateVisibility Enum=Collapsed, Index=Get bIsOpen() -> (bIsOpen:bool)) -> (ReturnValue:ESlateVisibility Enum)
   )
   > SetVisibility(self=[Get OK_Button(), Get OverrideCancelButton()], InVisibility=Select(Option 0=SelfHitTestInvisible, Option 1=Hidden, Index=Get bIsOpen()))
   > SetIsInteractionEnabled!(execute, self:Common Button Base Object Reference=Get OverrideCancelButton(), bInIsInteractionEnabled:bool=NOT Boolean(A=Get bIsOpen())) -> (then)
   > n1 = Branch(Condition=Get TextSearchIsOpen())
   n2 = Get bIsOpen()
   n3 = Select(Option 0=SelfHitTestInvisible, Option 1=Hidden, Index=@n2)
   > n4 = SetVisibility(self=Get SizeBox_TextSearch(), InVisibility=@n3)
   SetVisibility(
    execute=[@n4.then, @n1.else]
    self=[Get CreativePropertyEditorFilterWidget(), Get TextBlock_DescriptionText() -> (TextBlock_DescriptionText:Common Text Object Reference)]
    InVisibility=@n3
   )
   n5 = Select(Option 0=Visible, Option 1=Hidden, Index=@n2)
   > SetVisibility(self=Get OptionsList(), InVisibility=@n5)
   > SetVisibility(self=Get HeaderBarSB() -> (HeaderBarSB:Size Box Object Reference), InVisibility=@n5)
   > n6 = Branch(Condition=@n2)
   n7 = Get MoreOptionsMenu()
   > ActivateWidget(self=@n7)
   DeactivateWidget(execute=@n6.else, self=@n7)
   > SetFocus(self=Get OptionsList())
   > ForceFocusOnCurrentlySelectedItem() {is=title K2Node_CallFunction}
  ]}
 HandleShowTextSearch {nodes [
   HandleShowTextSearch() {is=title K2Node_FunctionEntry}
   > DirectSetSearchTextVisibility(Show=true)
   > n1 = CloseDialog(bSetFocusToOptionsList=true)
   Set TextSearchIsOpen(execute=[@n5.then, @n1.then], TextSearchIsOpen=true)
   > CenterWidget(ToCenter=Get EditableText_Search())
   n2 = Get CommonInputSubsystem() {is=title K2Node_GetSubsystem}
   > n3 = Branch(Condition="Equal (Enum)"(A=GetCurrentInputType(self=@n2), B=Gamepad))
   SetTextFocus(execute=@n3.else, self:WBP UIKit Filtered Input Field Object Reference=Get EditableText_Search())
   n4 = Branch(Condition=IsAnimationPlaying(InAnimation=Get ShowTextSearch()))
   SetVisibility(execute=@n4.else, self=Get EditableText_Search(), InVisibility=SelfHitTestInvisible)
   > n5 = PlayAnimationForward(InAnimation=Get ShowTextSearch())
  ]}
 HandleTextCommitByEnter {nodes [
   HandleTextCommitByEnter() {is=title K2Node_FunctionEntry}
   > Set bTextCommittedByEnter(bTextCommittedByEnter=true)
  ]}
 HandleUEFNOnlyOption {nodes [
   HandleUEFNOnlyOption!() -> (then, bHasUEFNOnlyOption:bool=~, TooltipText:Text) {is=title K2Node_FunctionEntry}
   > Set Visibility(
    Visibility="To Visibility (Boolean)"!(
     bIsVisible:bool=Get bHasUEFNOnlyOption() -> (bHasUEFNOnlyOption:bool)
     TrueVisibility:ESlateVisibility Enum=Visible
     FalseVisibility:ESlateVisibility Enum=Collapsed
    ) -> (ReturnValue:ESlateVisibility Enum=Visible)
    self=Get Overlay_WarningMessage() -> (Overlay_WarningMessage:Overlay Object Reference)
   )
   > Set Visibility(
    Visibility="To Visibility (Boolean)"(bIsVisible=Get bHasUEFNOnlyOption(), TrueVisibility=SelfHitTestInvisible, FalseVisibility=Collapsed) -> (ReturnValue=Visible)
    self=Get Hbox_UEFNOverrideDesc() -> (Hbox_UEFNOverrideDesc:Horizontal Box Object Reference)
   )
   > Set Text(Text=Get TooltipText() -> (TooltipText:Text), self=Get UEFNOverride_Description() -> (UEFNOverride_Description:Common Text Object Reference))
  ]}
 HasAnyFilterTags {nodes [
   HasAnyFilterTags() {is=title K2Node_FunctionEntry}
   n1 = Get FilterTags() -> (FilterTags:Gameplay Tag Container Structure)
   > n2 = Branch(Condition="Equal (Integer)"(A=GetNumGameplayTagsInContainer!(TagContainer:"Gameplay Tag Container Structure (by ref)"=@n1) -> (ReturnValue:int=0), B=0))
   > Return Node!(execute, bHasFilterTags:bool=false)
   Return Node(execute=@n2.else, bHasFilterTags=true)
  ]}
 HasUEFNOnlyOptions {nodes [
   HasUEFNOnlyOptions() {is=title K2Node_FunctionEntry}
   > Return Node!(execute, HasUEFNOnlyOptions:bool=false, TooltipText:Text)
  ]}
 NavigateDownFromNameTextEntry {nodes [
   n1 = NavigateDownFromNameTextEntry!() -> (then, Navigation:EUINavigation Enum) {is=title K2Node_FunctionEntry}
   > n2 = Switch on EUINavigation!(execute, Selection:EUINavigation Enum=@n1) -> (Left:Exec, Right:Exec, Up:Exec, Down:Exec, Next:Exec, Previous:Exec, Invalid:Exec) {is=title K2Node_SwitchEnum}
   n3 = Branch(execute=@n2.Down, Condition=Get TextSearchIsOpen())
   > Return Node!(execute, ReturnValue:Widget Object Reference=Get EditableText_Search())
   SetSelectionToFirstVisibleEntryWidget(execute=@n3.else)
   > Return Node()
  ]}
 NavigateDownFromSearch {nodes [
   n1 = NavigateDownFromSearch!() -> (then, Navigation:EUINavigation Enum) {is=title K2Node_FunctionEntry}
   > n2 = Switch on EUINavigation(Selection=@n1)
   SetSelectionToFirstVisibleEntryWidget(execute=@n2.Down)
   > Return Node()
  ]}
 NavigateUpToSearchIfPossible {nodes [
   n1 = NavigateUpToSearchIfPossible!() -> (then, Navigation:EUINavigation Enum) {is=title K2Node_FunctionEntry}
   > n2 = Switch on EUINavigation(Selection=@n1)
   n3 = Get EditableText_Search()
   Return Node(
    execute=@n2.Up
    ReturnValue=Select(Option 0:Widget Object Reference, Option 1:Widget Object Reference=@n3, Index=IsVisible(self=@n3)) -> (ReturnValue:Widget Object Reference)
   )
  ]}
 OnAdvancedFilterButtonSelectionChanged {nodes [
   n1 = OnAdvancedFilterButtonSelectionChanged!() -> (then, bSelected:bool=~) {is=title K2Node_FunctionEntry}
   > CloseDialog(bSetFocusToOptionsList=true)
   > Branch(Condition=@n1)
   > RemoveTagFilter!(execute, self:Fort Creative Message Knobs Dialog Object Reference, UpdateSelection:bool=Get UpdateSelectionOnListRegeneration()) -> (then)
   > RegenerateAllEntries(self=Get OptionsList())
   > RemoveGameplayTag!(
    execute
    TagContainer:"Gameplay Tag Container Structure (by ref)"=Get FilterTags()
    Tag:Gameplay Tag Structure=Get AdvancedTag() -> (AdvancedTag:Gameplay Tag Structure)
   ) -> (then, ReturnValue:bool)
   > RemoveGameplayTag(TagContainer=Get FilterTags(), Tag=Get BasicTag() -> (BasicTag:Gameplay Tag Structure))
   > RemoveGameplayTag(TagContainer=Get FilterTags(), Tag=Get ChannelTag() -> (ChannelTag:Gameplay Tag Structure))
   > RemoveGameplayTag(TagContainer=Get FilterTags(), Tag=Get EventDescriptorTag() -> (EventDescriptorTag:Gameplay Tag Structure))
   > RemoveGameplayTag(TagContainer=Get FilterTags(), Tag=Get EventFunctionTag() -> (EventFunctionTag:Gameplay Tag Structure))
  ]}
 OnBasicFilterButtonSelectionChanged {nodes [
   n1 = OnBasicFilterButtonSelectionChanged!() -> (then, bSelected:bool=~) {is=title K2Node_FunctionEntry}
   > CloseDialog(bSetFocusToOptionsList=true)
   > n2 = Branch(Condition=@n1)
   > n3 = Sequence!(execute) -> (then_0, then_1, then_2:Exec)
   FilterEntriesByTag!(
    execute=@n3.then_2
    self:Fort Creative Message Knobs Dialog Object Reference
    Tags:Gameplay Tag Container Structure=Get FilterTags()
    UpdateSelection:bool=Get UpdateSelectionOnListRegeneration()
   ) -> (then)
   RemoveGameplayTag(execute=@n3.then_0, TagContainer=Get FilterTags(), Tag=Get ChannelTag())
   > RemoveGameplayTag(TagContainer=Get FilterTags(), Tag=Get EventDescriptorTag())
   > RemoveGameplayTag(TagContainer=Get FilterTags(), Tag=Get EventFunctionTag())
   AddGameplayTag!(execute=@n3.then_1, TagContainer:"Gameplay Tag Container Structure (by ref)"=Get FilterTags(), Tag:Gameplay Tag Structure=Get BasicTag()) -> (then)
   RemoveGameplayTag(execute=@n2.else, TagContainer=Get FilterTags(), Tag=Get BasicTag())
   > n4 = Branch(Condition=HasAnyFilterTags!() -> (bHasFilterTags:bool) {is=title K2Node_CallFunction})
   RemoveTagFilter(execute=@n4.else, UpdateSelection=Get UpdateSelectionOnListRegeneration())
  ]}
 OnChannelFilterButtonSelectionChanged {nodes [
   n1 = OnChannelFilterButtonSelectionChanged!() -> (then, bSelected:bool=~) {is=title K2Node_FunctionEntry}
   > CloseDialog(bSetFocusToOptionsList=true)
   > n2 = Branch(Condition=@n1)
   > n3 = Sequence()
   FilterEntriesByTag(execute=@n3.then_2, Tags=Get FilterTags(), UpdateSelection=Get UpdateSelectionOnListRegeneration())
   RemoveGameplayTag(execute=@n3.then_0, TagContainer=Get FilterTags(), Tag=Get BasicTag())
   > RemoveGameplayTag(TagContainer=Get FilterTags(), Tag=Get EventDescriptorTag())
   > RemoveGameplayTag(TagContainer=Get FilterTags(), Tag=Get EventFunctionTag())
   AddGameplayTag(execute=@n3.then_1, TagContainer=Get FilterTags(), Tag=Get ChannelTag())
   RemoveGameplayTag(execute=@n2.else, TagContainer=Get FilterTags(), Tag=Get ChannelTag())
   > n4 = Branch(Condition=HasAnyFilterTags())
   RemoveTagFilter(execute=@n4.else, UpdateSelection=Get UpdateSelectionOnListRegeneration())
  ]}
 OnEventDescriptorFilterButtonSelectionChanged {nodes [
   n1 = OnEventDescriptorFilterButtonSelectionChanged!() -> (then, bSelected:bool=~) {is=title K2Node_FunctionEntry}
   > CloseDialog(bSetFocusToOptionsList=true)
   > n2 = Branch(Condition=@n1)
   > n3 = Sequence()
   FilterEntriesByTag(execute=@n3.then_2, Tags=Get FilterTags(), UpdateSelection=Get UpdateSelectionOnListRegeneration())
   RemoveGameplayTag(execute=@n3.then_0, TagContainer=Get FilterTags(), Tag=Get BasicTag())
   > RemoveGameplayTag(TagContainer=Get FilterTags(), Tag=Get ChannelTag())
   > RemoveGameplayTag(TagContainer=Get FilterTags(), Tag=Get EventFunctionTag())
   AddGameplayTag(execute=@n3.then_1, TagContainer=Get FilterTags(), Tag=Get EventDescriptorTag())
   RemoveGameplayTag(execute=@n2.else, TagContainer=Get FilterTags(), Tag=Get EventDescriptorTag())
   > n4 = Branch(Condition=HasAnyFilterTags())
   RemoveTagFilter(execute=@n4.else, UpdateSelection=Get UpdateSelectionOnListRegeneration())
  ]}
 OnEventFunctionFilterButtonSelectionChanged {nodes [
   n1 = OnEventFunctionFilterButtonSelectionChanged!() -> (then, bSelected:bool=~) {is=title K2Node_FunctionEntry}
   > CloseDialog(bSetFocusToOptionsList=true)
   > n2 = Branch(Condition=@n1)
   > n3 = Sequence()
   FilterEntriesByTag(execute=@n3.then_2, Tags=Get FilterTags(), UpdateSelection=Get UpdateSelectionOnListRegeneration())
   RemoveGameplayTag(execute=@n3.then_0, TagContainer=Get FilterTags(), Tag=Get BasicTag())
   > RemoveGameplayTag(TagContainer=Get FilterTags(), Tag=Get ChannelTag())
   > RemoveGameplayTag(TagContainer=Get FilterTags(), Tag=Get EventDescriptorTag())
   AddGameplayTag(execute=@n3.then_1, TagContainer=Get FilterTags(), Tag=Get EventFunctionTag())
   RemoveGameplayTag(execute=@n2.else, TagContainer=Get FilterTags(), Tag=Get EventFunctionTag())
   > n4 = Branch(Condition=HasAnyFilterTags())
   RemoveTagFilter(execute=@n4.else, UpdateSelection=Get UpdateSelectionOnListRegeneration())
  ]}
 OnModifiedFilterButtonSelectionChanged {nodes [
   n1 = OnModifiedFilterButtonSelectionChanged!() -> (then, bSelected:bool=~) {is=title K2Node_FunctionEntry}
   > CloseDialog(bSetFocusToOptionsList=true)
   > n2 = Branch(Condition=@n1)
   > FilterEntriesByIsModified!(execute, self:Fort Creative Message Knobs Dialog Object Reference, bShowUnmodifiedEntries:bool=false, UpdateSelection:bool=true) -> (then)
   RemoveIsModifiedFilter!(execute=@n2.else, self:Fort Creative Message Knobs Dialog Object Reference, UpdateSelection:bool=true) -> (then)
   > RegenerateAllEntries(self=Get OptionsList())
  ]}
 SetEmptyStateMessageVisibility {nodes [
   n1 = SetEmptyStateMessageVisibility!() -> (then, Show:bool=~) {is=title K2Node_FunctionEntry}
   > SetVisibility(
    self=Get Overlay_EmptyState() -> (Overlay_EmptyState:Overlay Object Reference)
    InVisibility=Select(Option 0:ESlateVisibility Enum=Collapsed, Option 1:ESlateVisibility Enum=HitTestInvisible, Index=@n1) -> (ReturnValue:ESlateVisibility Enum)
   )
  ]}
 SetItemInfo {nodes [
   n1 = SetItemInfo!() -> (then, ItemName:Text, ItemBlock:Text, ItemDescription:Text) {is=title K2Node_FunctionEntry}
   > Set ItemName(ItemName=@n1)
   > n2 = Sequence()
   n3 = Get ItemName()
   n4 = Get WBP_PropertyEditor_NameEntryWidget()
   > SetTextWithForcedSanitization(self=@n4, Text=@n3)
   > SetVisibility(self=@n4, InVisibility=Select(Option 0=SelfHitTestInvisible, Option 1=Collapsed, Index=TextIsEmpty(InText=@n3)))
   n5 = Get Text_ItemBaseName() -> (Text_ItemBaseName:Common Text Object Reference)
   "SetText (Text)"(execute=@n2.then_1, self=@n5, InText=@n1.ItemBlock)
   > SetVisibility(self=@n5, InVisibility=Select(Option 0=SelfHitTestInvisible, Option 1=Collapsed, Index=TextIsEmpty(InText=@n1.ItemBlock)))
   n6 = Get Text_ItemDescription() -> (Text_ItemDescription:Common Text Object Reference)
   "SetText (Text)"(execute=@n2.then_2, self=@n6, InText=@n1.ItemDescription)
   > SetVisibility(self=@n6, InVisibility=Select(Option 0=SelfHitTestInvisible, Option 1=Collapsed, Index=TextIsEmpty(InText=@n1.ItemDescription)))
  ]}
 SetSelectionToFirstVisibleEntryWidget {nodes [
   SetSelectionToFirstVisibleEntryWidget() {is=title K2Node_FunctionEntry}
   n1 = Get OptionsList()
   > Branch(Condition="integer > integer"(A=GetNumItems(self=@n1) -> (ReturnValue=0), B=0) {is=title K2Node_CallFunction})
   n2 = Get OptionsList()
   > SetSelectedIndex(self=@n2, Index=0)
   > NavigateToIndex(self=@n2, Index=0)
   > ForceFocusOnCurrentlySelectedItem()
  ]}
 UpdateHoldText {nodes [UpdateHoldText!() -> (then, ShowHold:bool=~) {is=title K2Node_FunctionEntry}]}
]
variables [
 AdvancedTag {type=GameplayTag, default {tagName=Creative.Property.Advanced}}
 AnimationTime {type=number, default=0.2}
 bApplyChanges {type=boolean}
 BasicTag {type=GameplayTag, default {tagName=Creative.Property.Basic}}
 bIsInCursorMode {type=boolean}
 bIsOnFocusPath {type=boolean}
 bTextCommittedByEnter {type=boolean}
 ChannelTag {type=GameplayTag, default {tagName=Creative.Property.Channel}}
 ClearMessage {type=string, default=All items removed.}
 Close Menu Sound {type=SoundBase, default=/Game/Sounds/Fort_UI_Sounds/FrontEnd/New_FrontEnd/Fortnite_UI_Creative_Menu_Close_Cue}
 EventDescriptorTag {type=GameplayTag, default {tagName=Creative.Property.Event.Descriptor}}
 EventFunctionTag {type=GameplayTag, default {tagName=Creative.Property.Event.Function}}
 FilterTags {type=GameplayTagContainer, default {gameplayTags []}}
 FilterText {type=string, default=""}
 HideGameplayTags {
  type=GameplayTagContainer
  default {gameplayTags [
    {tagName=Athena.HideForCapture}
    {tagName=HUD.InWorld.InteractionIndicator}
    {tagName=HUD.Hide.WhileLoading}
    {tagName=Gameplay.Action.Player.Inventory}
   ]}
 }
 isUsingGamePad {type=boolean}
 MoreOptionsMenu {type=Cr_PropertyEditor_MoreOptions_C, default=null}
 Open Menu Sound {type=SoundBase, default=/Game/Sounds/Fort_UI_Sounds/FrontEnd/New_FrontEnd/Fortnite_UI_Creative_Menu_Open_Cue}
 RecheckContextTagsTimeout {type=number, default=0.5}
 RecheckContextTagsTimerHandle {type=TimerHandle}
 ResetEventsMessage {type=string, default=All event settings in this device have been reset to default values.}
 ResetFunctionsMessage {type=string, default=All function settings in this device have been reset to default values.}
 ResetMessage {type=string, default=All settings in this device have been reset to default values.}
 ResetPropertiesMessage {type=string, default=All property settings in this device have been reset to default values.}
 SearchDelay {type=number, default=0.5}
 SpatialLoadingStateUpdatedHandle {type=AsyncAction_StartListeningToEvent, default=null}
 TextSearchIsOpen {type=boolean}
 TextSearchTimer {type=TimerHandle}
 UpdateSelectionOnListRegeneration {type=boolean}
]
widgets {
 info {
  inheritedWidgetCount=9
  namedSlotCount=1
  parentClass=/Script/FortniteUI.FortCreativeMessageKnobsDialog
  rootWidgetClass=&Overlay
  widgetCount=106
 }
 widgets [
  Overlay_4 {widgetClassPath=&Overlay, widget=Overlay_4}
  Lightbox {
   widgetClassPath=/Game/UI/Foundation/Lightbox/Lightbox
   parent=Overlay_4
   widget=Lightbox
   properties {dimAmount=0.6, showWindowControlButtons=false}
   slot {class=OverlaySlot, properties {horizontalAlignment=HAlign_Fill, verticalAlignment=VAlign_Fill}}
  }
  SizeBox_2 {
   widgetClassPath=&SizeBox
   namedSlotHost=Lightbox
   widget=SizeBox_2
   properties {heightOverride=800, visibility=Collapsed, widthOverride=1200}
  }
  SafeZonePanel {
   widgetClassPath=&SafeZone
   parent=Overlay_4
   widget=SafeZonePanel
   slot {class=OverlaySlot, properties {horizontalAlignment=HAlign_Fill, verticalAlignment=VAlign_Fill}}
  }
  HorizontalBox_0 {widgetClassPath=&HorizontalBox, parent=SafeZonePanel, widget=HorizontalBox_0, slot {class=SafeZoneSlot}}
  Overlay_Shell {
   widgetClassPath=&Overlay
   parent=HorizontalBox_0
   widget=Overlay_Shell
   slot {class=HorizontalBoxSlot, properties {size {sizeRule=Fill, value=1.4}}}
  }
  SafeZonePadding {
   widgetClassPath=&CommonBorder
   parent=Overlay_Shell
   widget=SafeZonePadding
   properties {
    bReducePaddingBySafezone=true
    background {
     bIsDynamicallyLoaded=false
     drawAs=Box
     imageType=NoImage
     mirroring=NoMirror
     resourceName=None
     resourceObject=null
     tiling=NoTile
     imageSize=<32 32>
     margin=<ltrb 0 0 0 0>
     outlineSettings {
      bUseBrushTransparency=false
      roundingType=HalfHeightRadius
      width=0
      color {colorUseRule=UseColor_Specified, specifiedColor=<rgba 0 0 0 0>}
      cornerRadii=<0 0 0 0>
     }
     tintColor {colorUseRule=UseColor_Specified, specifiedColor=<rgba 0.0027 0.006 0.0452 1>}
     uVRegion {bIsValid=false, max=<0 0>, min=<0 0>}
    }
    minimumPadding=<ltrb 0 48 0 16>
    padding=<ltrb 0 0 0 0>
   }
   slot {class=OverlaySlot, properties {horizontalAlignment=HAlign_Fill, verticalAlignment=VAlign_Fill}}
  }
  AnimBoundSB {
   widgetClassPath=&SizeBox
   parent=SafeZonePadding
   widget=AnimBoundSB
   properties {bOverride_WidthOverride=true, widthOverride=968}
   slot {class=BorderSlot, properties {padding=<ltrb 0 0 0 0>}}
  }
  Overlay {widgetClassPath=&Overlay, parent=AnimBoundSB, widget=Overlay, slot {class=SizeBoxSlot}}
  TabFill {
   widgetClassPath=&Border
   parent=Overlay
   widget=TabFill
   properties {
    background {
     bIsDynamicallyLoaded=false
     drawAs=Image
     imageType=NoImage
     mirroring=NoMirror
     resourceName=None
     resourceObject=/Game/UI/Foundation/Materials/M_BoltGradientBorder_LockerTabsFill
     tiling=NoTile
     imageSize=<32 32>
     margin=<ltrb 0 0 0 0>
     outlineSettings {
      bUseBrushTransparency=false
      roundingType=HalfHeightRadius
      width=0
      color {colorUseRule=UseColor_Specified, specifiedColor=<rgba 0 0 0 0>}
      cornerRadii=<0 0 0 0>
     }
     tintColor {colorUseRule=UseColor_Specified, specifiedColor=<rgba 1 1 1 1>}
     uVRegion {bIsValid=false, max=<0 0>, min=<0 0>}
    }
    padding=<ltrb 0 0 0 0>
    renderTransform {angle=0, scale=<1 1>, shear=<0 0>, translation=<0 -48>}
   }
   slot {class=OverlaySlot, properties {horizontalAlignment=HAlign_Fill}}
  }
  Spacer_4 {
   widgetClassPath=&Spacer
   parent=TabFill
   widget=Spacer_4
   properties {size=<1 15>}
   slot {class=BorderSlot, properties {padding=<ltrb 0 0 0 0>}}
  }
  HeaderFill {
   widgetClassPath=&Border
   parent=Overlay
   widget=HeaderFill
   properties {
    background {
     bIsDynamicallyLoaded=false
     drawAs=Image
     imageType=NoImage
     mirroring=NoMirror
     resourceName=None
     resourceObject=/Game/UI/Foundation/Materials/M_BoltGradientBorder_LockerHeaderFill
     tiling=NoTile
     imageSize=<32 32>
     margin=<ltrb 0 0 0 0>
     outlineSettings {
      bUseBrushTransparency=false
      roundingType=HalfHeightRadius
      width=0
      color {colorUseRule=UseColor_Specified, specifiedColor=<rgba 0 0 0 0>}
      cornerRadii=<0 0 0 0>
     }
     tintColor {colorUseRule=UseColor_Specified, specifiedColor=<rgba 1 1 1 1>}
     uVRegion {bIsValid=false, max=<0 0>, min=<0 0>}
    }
    padding=<ltrb 0 0 0 0>
   }
   slot {class=OverlaySlot, properties {horizontalAlignment=HAlign_Fill, padding=<ltrb 0 -48 0 0>}}
  }
  Spacer {
   widgetClassPath=&Spacer
   parent=HeaderFill
   widget=Spacer
   properties {size=<1 15>}
   slot {class=BorderSlot, properties {padding=<ltrb 0 0 0 0>}}
  }
  Overlay_7 {
   widgetClassPath=&Overlay
   parent=Overlay
   widget=Overlay_7
   slot {class=OverlaySlot, properties {horizontalAlignment=HAlign_Fill}}
  }
  Spacer_109 {
   widgetClassPath=&Spacer
   parent=Overlay_7
   widget=Spacer_109
   properties {size=<1 336.4525>}
   slot {class=OverlaySlot, properties {horizontalAlignment=HAlign_Fill}}
  }
  WBP_CaptureForPostBufferUpdate {
   widgetClassPath=&WBP_CaptureForPostBufferUpdate
   bIsVariable=true
   parent=Overlay_7
   slot {class=OverlaySlot, properties {horizontalAlignment=HAlign_Fill, verticalAlignment=VAlign_Fill}}
  }
  SizeBox_TextSearch {
   widgetClassPath=&SizeBox
   bIsVariable=true
   parent=Overlay_7
   widget=SizeBox_TextSearch
   properties {bOverride_HeightOverride=true, heightOverride=92, visibility=Collapsed}
   slot {class=OverlaySlot, properties {horizontalAlignment=HAlign_Fill, verticalAlignment=VAlign_Bottom}}
  }
  Overlay_0 {widgetClassPath=&Overlay, parent=SizeBox_TextSearch, widget=Overlay_0, slot {class=SizeBoxSlot}}
  EditableText_Search {
   widgetClassPath=/FortUILibrary/UIKit/InputField/WBP_UIKit_FilteredInputField
   bIsVariable=true
   parent=Overlay_0
   widget=EditableText_Search
   properties {
    bIsFocusable=true
    entryCharLimit=100
    hideValidationInfo=true
    hintText=start typing to search
    isSearchField=true
    justification=Center
    navigation=&self:WidgetTree.EditableText_Search.WidgetNavigation_0
    showTextPrompt=false
    useHintText=true
    virtualKeyboardTrigger=OnAllFocusEvents
   }
   slot {class=OverlaySlot, properties {horizontalAlignment=HAlign_Fill, verticalAlignment=VAlign_Fill, padding=<ltrb 32 20 32 18>}}
  }
  VBox_HeaderArt {
   widgetClassPath=&VerticalBox
   parent=Overlay_7
   widget=VBox_HeaderArt
   slot {class=OverlaySlot, properties {horizontalAlignment=HAlign_Fill, padding=<ltrb 0 -33 0 0>}}
  }
  Overlay_5 {
   widgetClassPath=&Overlay
   parent=VBox_HeaderArt
   widget=Overlay_5
   properties {flowDirectionPreference=Culture}
   slot {class=VerticalBoxSlot}
  }
  TabBorder {
   widgetClassPath=&Border
   parent=Overlay_5
   widget=TabBorder
   properties {bFlipForRightToLeftFlowDirection=true, brushColor=<rgba 0 0.0262 0.3372 1>, padding=<ltrb 0 0 0 0>}
   slot {class=OverlaySlot, properties {horizontalAlignment=HAlign_Fill, verticalAlignment=VAlign_Fill}}
  }
  Spacer_1 {
   widgetClassPath=&Spacer
   parent=TabBorder
   widget=Spacer_1
   properties {size=<1 279.8716>}
   slot {class=BorderSlot, properties {padding=<ltrb 0 0 0 0>}}
  }
  HeaderBorder {
   widgetClassPath=&Border
   parent=Overlay_5
   widget=HeaderBorder
   properties {
    bFlipForRightToLeftFlowDirection=true
    background {
     bIsDynamicallyLoaded=false
     drawAs=Image
     imageType=NoImage
     mirroring=NoMirror
     resourceName=None
     resourceObject=/Game/UI/Foundation/Materials/M_BoltGradientBorder_LockerHeader
     tiling=NoTile
     imageSize=<32 32>
     margin=<ltrb 0.5 0.5 0.5 0.5>
     outlineSettings {
      bUseBrushTransparency=false
      roundingType=HalfHeightRadius
      width=0
      color {colorUseRule=UseColor_Specified, specifiedColor=<rgba 0 0 0 0>}
      cornerRadii=<0 0 0 0>
     }
     tintColor {colorUseRule=UseColor_Specified, specifiedColor=<rgba 1 1 1 1>}
     uVRegion {bIsValid=false, max=<0 0>, min=<0 0>}
    }
    padding=<ltrb 0 0 0 0>
   }
   slot {class=OverlaySlot, properties {horizontalAlignment=HAlign_Fill}}
  }
  Spacer_2 {
   widgetClassPath=&Spacer
   parent=HeaderBorder
   widget=Spacer_2
   properties {size=<1 182.4451>}
   slot {class=BorderSlot, properties {padding=<ltrb 0 0 0 0>}}
  }
  SizeBox_1 {
   widgetClassPath=&SizeBox
   parent=Overlay_7
   widget=SizeBox_1
   properties {bOverride_HeightOverride=true, bOverride_WidthOverride=true, heightOverride=50, widthOverride=290}
   slot {class=OverlaySlot, properties {horizontalAlignment=HAlign_Right, verticalAlignment=VAlign_Center, padding=<ltrb 0 0 0 42>}}
  }
  Overlay_WarningMessage {
   widgetClassPath=&Overlay
   bIsVariable=true
   parent=SizeBox_1
   widget=Overlay_WarningMessage
   properties {visibility=Collapsed}
   slot {class=SizeBoxSlot}
  }
  UEFNOptionsBackground {
   widgetClassPath=&HorizontalBox
   parent=Overlay_WarningMessage
   widget=UEFNOptionsBackground
   slot {class=OverlaySlot, properties {horizontalAlignment=HAlign_Fill, verticalAlignment=VAlign_Fill}}
  }
  SheerEdge {
   widgetClassPath=&Image
   parent=UEFNOptionsBackground
   widget=SheerEdge
   properties {
    visibility=SelfHitTestInvisible
    colorAndOpacity=<rgba 1 0.15 0.0056 1>
    renderTransform {angle=0, scale=<1 1>, shear=<-12 0>, translation=<0 0>}
   }
   slot {class=HorizontalBoxSlot, properties {padding=<ltrb 0 0 -10 0>}}
  }
  Background {
   widgetClassPath=&Image
   parent=UEFNOptionsBackground
   widget=Background
   properties {visibility=SelfHitTestInvisible, colorAndOpacity=<rgba 1 0.15 0.0056 1>}
   slot {class=HorizontalBoxSlot, properties {size {sizeRule=Fill, value=1}}}
  }
  ScaleBox_0 {
   widgetClassPath=&ScaleBox
   parent=Overlay_WarningMessage
   widget=ScaleBox_0
   slot {class=OverlaySlot, properties {padding=<ltrb 10 5 10 5>}}
  }
  HorizontalBox_4 {widgetClassPath=&HorizontalBox, parent=ScaleBox_0, widget=HorizontalBox_4, slot {class=ScaleBoxSlot}}
  Icon {
   widgetClassPath=&Image
   bIsVariable=true
   parent=HorizontalBox_4
   widget=Icon
   properties {
    brush {
     bIsDynamicallyLoaded=false
     drawAs=Image
     imageType=NoImage
     mirroring=NoMirror
     resourceName=None
     resourceObject=/Game/Athena/UI/InGame/Creative/PropertyEditor/Art/Icon-PropertyEditor-UEFNOverride
     tiling=NoTile
     imageSize=<50 50>
     margin=<ltrb 0 0 0 0>
     outlineSettings {
      bUseBrushTransparency=false
      roundingType=HalfHeightRadius
      width=0
      color {colorUseRule=UseColor_Specified, specifiedColor=<rgba 0 0 0 0>}
      cornerRadii=<0 0 0 0>
     }
     tintColor {colorUseRule=UseColor_Specified, specifiedColor=<rgba 1 1 1 1>}
     uVRegion {bIsValid=false, max=<0 0>, min=<0 0>}
    }
    colorAndOpacity=<rgba 0 0 0 1>
   }
   slot {class=HorizontalBoxSlot, properties {horizontalAlignment=HAlign_Center, verticalAlignment=VAlign_Center, padding=<ltrb 0 0 10 0>}}
  }
  WarningMessage_Text_1 {
   widgetClassPath=&CommonTextBlock
   bIsVariable=true
   parent=HorizontalBox_4
   widget=WarningMessage_Text_1
   properties {
    justification=Center
    text=Options Override
    colorAndOpacity {colorUseRule=UseColor_Specified, specifiedColor=<rgba 0 0 0 1>}
    font {
     bForceMonospaced=false
     bMaterialIsStencil=false
     fontMaterial=null
     fontObject=&Burbank
     letterSpacing=0
     monospacedWidth=1
     size=24
     skewAmount=0
     typefaceFontName=Burbank-PlayerPower
     outlineSettings {
      bApplyOutlineToDropShadows=false
      bMiteredCorners=false
      bSeparateFillAlpha=false
      outlineMaterial=null
      outlineSize=0
      outlineColor=<rgba 0 0 0 1>
     }
    }
    renderTransform {angle=0, scale=<1 1>, shear=<-12 0>, translation=<0 0>}
    shadowOffset=<0 0>
   }
   slot {class=HorizontalBoxSlot, properties {horizontalAlignment=HAlign_Center, verticalAlignment=VAlign_Center}}
  }
  Button_LeftDialogMouseCapture {
   widgetClassPath=&Button
   parent=Overlay
   widget=Button_LeftDialogMouseCapture
   properties {isFocusable=false, visibility=Collapsed, backgroundColor=<rgba 1 1 1 0>, colorAndOpacity=<rgba 1 1 1 0>}
   slot {class=OverlaySlot, properties {horizontalAlignment=HAlign_Fill, verticalAlignment=VAlign_Fill, padding=<ltrb 0 -45 0 -15>}}
  }
  Image {
   widgetClassPath=&Image
   parent=Button_LeftDialogMouseCapture
   widget=Image
   properties {visibility=SelfHitTestInvisible, colorAndOpacity=<rgba 0.0331 0.855 1 1>}
   slot {class=ButtonSlot, properties {horizontalAlignment=HAlign_Fill, verticalAlignment=VAlign_Fill, padding=<ltrb 0 0 0 0>}}
  }
  VBox_Content {
   widgetClassPath=&VerticalBox
   parent=Overlay
   widget=VBox_Content
   slot {class=OverlaySlot, properties {horizontalAlignment=HAlign_Fill, verticalAlignment=VAlign_Fill, padding=<ltrb 0 -15 0 0>}}
  }
  HeaderBarSB {
   widgetClassPath=&SizeBox
   bIsVariable=true
   parent=VBox_Content
   widget=HeaderBarSB
   properties {bOverride_HeightOverride=true, flowDirectionPreference=Culture, heightOverride=96}
   slot {class=VerticalBoxSlot, properties {padding=<ltrb 64 -20 64 0>}}
  }
  HorizontalBox_2 {widgetClassPath=&VerticalBox, parent=HeaderBarSB, widget=HorizontalBox_2, slot {class=SizeBoxSlot}}
  ObjectTypeOverlay {widgetClassPath=&Overlay, parent=HorizontalBox_2, widget=ObjectTypeOverlay, slot {class=VerticalBoxSlot}}
  ObjectTypeHB {
   widgetClassPath=&HorizontalBox
   parent=ObjectTypeOverlay
   widget=ObjectTypeHB
   slot {class=OverlaySlot, properties {horizontalAlignment=HAlign_Fill, verticalAlignment=VAlign_Center}}
  }
  VerticalBox_0 {
   widgetClassPath=&VerticalBox
   parent=ObjectTypeHB
   widget=VerticalBox_0
   slot {class=HorizontalBoxSlot, properties {horizontalAlignment=HAlign_Left}}
  }
  TextBox_PropertyEditorMode {
   widgetClassPath=&CommonTextBlock
   parent=VerticalBox_0
   widget=TextBox_PropertyEditorMode
   properties {
    scrollStyle=/FortUILibrary/Styles/Typography/Foundation/TextScrollStyle-NoFade.TextScrollStyle-NoFade_C
    text=CUSTOMIZE
    font {
     bForceMonospaced=false
     bMaterialIsStencil=false
     fontMaterial=null
     fontObject=&Burbank
     letterSpacing=0
     monospacedWidth=1
     size=22
     skewAmount=0
     typefaceFontName=Burbank-Black
     outlineSettings {
      bApplyOutlineToDropShadows=false
      bMiteredCorners=false
      bSeparateFillAlpha=false
      outlineMaterial=null
      outlineSize=0
      outlineColor=<rgba 0 0 0 1>
     }
    }
    renderTransform {angle=0, scale=<1 1>, shear=<-8 0>, translation=<0 0>}
    shadowOffset=<0 0>
   }
   slot {class=VerticalBoxSlot}
  }
  SizeBox_3 {
   widgetClassPath=&SizeBox
   parent=VerticalBox_0
   widget=SizeBox_3
   properties {
    bOverride_MaxDesiredWidth=true
    bOverride_MinDesiredHeight=true
    bOverride_WidthOverride=true
    maxDesiredWidth=770
    minDesiredHeight=100
    widthOverride=840
   }
   slot {class=VerticalBoxSlot}
  }
  WBP_PropertyEditor_NameEntryWidget {
   widgetClassPath=&WBP_UIKit_FilteredInputField
   bIsVariable=true
   parent=SizeBox_3
   widget=WBP_PropertyEditor_NameEntryWidget
   properties {
    hideValidationInfo=true
    hintText=Device Name
    labelText=CUSTOMIZE
    navigation=&self:WidgetTree.WBP_PropertyEditor_NameEntryWidget.WidgetNavigation_1
    showTextPrompt=false
   }
   slot {class=SizeBoxSlot}
  }
  Button_RenameInputButton {
   widgetClassPath=&WBP_UIKit_Button_Generic
   bIsVariable=true
   parent=ObjectTypeOverlay
   widget=Button_RenameInputButton
   properties {
    bIsFocusable=false
    enableFocusedSelectedState=false
    minHeight=0
    promptHorizontalAlignment=HAlign_Right
    style=/FortUILibrary/Foundation/Buttons/ButtonStyles/ButtonStyle-Clear.ButtonStyle-Clear_C
    text=""
    useBackgroundBlock=false
    buttonPadding=<ltrb 0 0 0 0>
    promptNudge=<50 0>
    triggeringInputAction {dataTable=/FortUILibrary/Input/CreativeInputActionDataTable, rowName=CreativeDeviceRename}
   }
   slot {class=OverlaySlot, properties {horizontalAlignment=HAlign_Right, verticalAlignment=VAlign_Center, padding=<ltrb 0 16 0 0>}}
  }
  Spacer_211 {
   widgetClassPath=&Spacer
   parent=VBox_Content
   widget=Spacer_211
   properties {accessibleBehavior=Auto, bOverrideAccessibleDefaults=true, size=<1 79.8063>}
   slot {class=VerticalBoxSlot}
  }
  SizeBox_Filter {
   widgetClassPath=&SizeBox
   parent=VBox_Content
   widget=SizeBox_Filter
   properties {bOverride_HeightOverride=true, heightOverride=96}
   slot {class=VerticalBoxSlot, properties {padding=<ltrb 64 0 64 10>}}
  }
  CreativePropertyEditorFilterWidget {
   widgetClassPath=&/CreativePropertyEditorFilterWidget
   bIsVariable=true
   parent=SizeBox_Filter
   widget=CreativePropertyEditorFilterWidget
   slot {class=SizeBoxSlot, properties {verticalAlignment=VAlign_Center, padding=<ltrb 0 5 0 0>}}
  }
  Spacer_TextSearch {
   widgetClassPath=&Spacer
   bIsVariable=true
   parent=VBox_Content
   widget=Spacer_TextSearch
   properties {visibility=Collapsed, size=<1 74>}
   slot {class=VerticalBoxSlot}
  }
  Overlay_2 {
   widgetClassPath=&Overlay
   parent=VBox_Content
   widget=Overlay_2
   slot {class=VerticalBoxSlot, properties {size {sizeRule=Fill, value=1}}}
  }
  WidgetSwitcher_ListViews {
   widgetClassPath=&CommonWidgetSwitcherLegacy
   bIsVariable=true
   parent=Overlay_2
   widget=WidgetSwitcher_ListViews
   properties {
    navigation=&self:WidgetTree.WidgetSwitcher_ListViews.WidgetNavigation_0
    transitionDuration=0
    transitionType=Horizontal
   }
   slot {class=OverlaySlot, properties {horizontalAlignment=HAlign_Fill, verticalAlignment=VAlign_Fill, padding=<ltrb 16 12 16 24>}}
  }
  ListSizeBox {
   widgetClassPath=&SizeBox
   parent=WidgetSwitcher_ListViews
   widget=ListSizeBox
   properties {bOverride_WidthOverride=true, heightOverride=425, maxDesiredHeight=258, minDesiredHeight=720, widthOverride=1000}
   slot {class=WidgetSwitcherSlot}
  }
  OptionsList {
   widgetClassPath=&CommonListView
   bInherited=true
   bIsVariable=true
   parent=ListSizeBox
   widget=OptionsList
   properties {
    bEnableScrollAnimation=false
    bIsVolatile=true
    entryWidgetClass=&/CreativePropertyEditorEntry.CreativePropertyEditorEntry_C
    horizontalEntrySpacing=4
    navigation=&self:WidgetTree.OptionsList.WidgetNavigation_0
    numDesignerPreviewEntries=6
    verticalEntrySpacing=4
   }
   slot {class=SizeBoxSlot}
  }
  Overlay_EmptyState {
   widgetClassPath=&Overlay
   bIsVariable=true
   parent=WidgetSwitcher_ListViews
   widget=Overlay_EmptyState
   properties {visibility=Visible}
   slot {class=WidgetSwitcherSlot}
  }
  Text_Settings {
   widgetClassPath=&CommonTextBlock
   parent=Overlay_EmptyState
   widget=Text_Settings
   properties {
    justification=Center
    style=/FortUILibrary/Foundation/Buttons/ButtonStyles/Athena/TextStyle_BurbSmBk_20_White50pc.TextStyle_BurbSmBk_20_White50pc_C
    text=No settings to display
    colorAndOpacity {colorUseRule=UseColor_Specified, specifiedColor=<rgba 1 1 1 0.5>}
    font {
     bForceMonospaced=false
     bMaterialIsStencil=false
     fontMaterial=null
     fontObject=&BurbankSmall
     letterSpacing=0
     monospacedWidth=1
     size=20
     skewAmount=0
     typefaceFontName=Black
     outlineSettings {
      bApplyOutlineToDropShadows=false
      bMiteredCorners=false
      bSeparateFillAlpha=false
      outlineMaterial=null
      outlineSize=0
      outlineColor=<rgba 0 0 0 1>
     }
    }
    renderTransform {angle=0, scale=<1 1>, shear=<-10 0>, translation=<0 0>}
    shadowOffset=<0 0>
   }
   slot {class=OverlaySlot, properties {horizontalAlignment=HAlign_Fill, padding=<ltrb 0 64 0 0>}}
  }
  Overlay_LoadingSpinner {
   widgetClassPath=&Overlay
   bIsVariable=true
   parent=Overlay_2
   widget=Overlay_LoadingSpinner
   properties {visibility=Collapsed}
   slot {class=OverlaySlot, properties {horizontalAlignment=HAlign_Fill, verticalAlignment=VAlign_Fill}}
  }
  Image_LoadingSpinner {
   widgetClassPath=&Image
   parent=Overlay_LoadingSpinner
   widget=Image_LoadingSpinner
   properties {
    visibility=SelfHitTestInvisible
    brush {
     bIsDynamicallyLoaded=false
     drawAs=Image
     imageType=NoImage
     mirroring=NoMirror
     resourceName=None
     resourceObject=/Game/UI/Foundation/Materials/Throbber_Material-Monochrome
     tiling=NoTile
     imageSize=<64 64>
     margin=<ltrb 0 0 0 0>
     outlineSettings {
      bUseBrushTransparency=false
      roundingType=HalfHeightRadius
      width=0
      color {colorUseRule=UseColor_Specified, specifiedColor=<rgba 0 0 0 0>}
      cornerRadii=<0 0 0 0>
     }
     tintColor {colorUseRule=UseColor_Specified, specifiedColor=<rgba 1 1 1 1>}
     uVRegion {bIsValid=false, max=<0 0>, min=<0 0>}
    }
    colorAndOpacity=<rgba 0 0.624 1 1>
   }
   slot {class=OverlaySlot, properties {horizontalAlignment=HAlign_Center, verticalAlignment=VAlign_Center}}
  }
  SizeBox_4 {
   widgetClassPath=&SizeBox
   parent=VBox_Content
   widget=SizeBox_4
   properties {bOverride_HeightOverride=true, heightOverride=135}
   slot {class=VerticalBoxSlot, properties {padding=<ltrb 32 0 32 0>}}
  }
  VerticalBox_1 {widgetClassPath=&VerticalBox, parent=SizeBox_4, widget=VerticalBox_1, slot {class=SizeBoxSlot}}
  TextBlock_DescriptionText {
   widgetClassPath=&CommonTextBlock
   bInherited=true
   bIsVariable=true
   parent=VerticalBox_1
   widget=TextBlock_DescriptionText
   properties {
    autoWrapText=true
    text="[Description] Lorem ipsum dolor sit amet, consectetur adipiscing elit, sed do eiusmod tempor incididunt ut labore et dolore magna aliqua. Ut enim ad minim veniam, quis nostrud exercitation "
    visibility=Hidden
    font {
     bForceMonospaced=false
     bMaterialIsStencil=false
     fontMaterial=null
     fontObject=&BurbankSmall
     letterSpacing=0
     monospacedWidth=1
     size=18
     skewAmount=0
     typefaceFontName=Bold
     outlineSettings {
      bApplyOutlineToDropShadows=false
      bMiteredCorners=false
      bSeparateFillAlpha=false
      outlineMaterial=null
      outlineSize=0
      outlineColor=<rgba 0 0 0 1>
     }
    }
    shadowOffset=<0 0>
   }
   slot {class=VerticalBoxSlot}
  }
  Hbox_UEFNOverrideDesc {
   widgetClassPath=&HorizontalBox
   bIsVariable=true
   parent=VerticalBox_1
   widget=Hbox_UEFNOverrideDesc
   properties {visibility=Collapsed}
   slot {class=VerticalBoxSlot, properties {verticalAlignment=VAlign_Bottom, size {sizeRule=Fill, value=1}}}
  }
  Icon_1 {
   widgetClassPath=&Image
   bIsVariable=true
   parent=Hbox_UEFNOverrideDesc
   widget=Icon_1
   properties {brush {
     bIsDynamicallyLoaded=false
     drawAs=Image
     imageType=NoImage
     mirroring=NoMirror
     resourceName=None
     resourceObject=&Icon-PropertyEditor-UEFNOverride
     tiling=NoTile
     imageSize=<40 40>
     margin=<ltrb 0 0 0 0>
     outlineSettings {
      bUseBrushTransparency=false
      roundingType=HalfHeightRadius
      width=0
      color {colorUseRule=UseColor_Specified, specifiedColor=<rgba 0 0 0 0>}
      cornerRadii=<0 0 0 0>
     }
     tintColor {colorUseRule=UseColor_Specified, specifiedColor=<rgba 1 1 1 1>}
     uVRegion {bIsValid=false, max=<0 0>, min=<0 0>}
    }}
   slot {class=HorizontalBoxSlot, properties {padding=<ltrb 0 0 15 0>}}
  }
  UEFNOverride_Description {
   widgetClassPath=&CommonTextBlock
   bIsVariable=true
   parent=Hbox_UEFNOverrideDesc
   widget=UEFNOverride_Description
   properties {
    autoWrapText=true
    text="[Description] Lorem ipsum dolor sit amet, consectetur adipiscing elit, sed do eiusmod tempor incididunt ut "
    font {
     bForceMonospaced=false
     bMaterialIsStencil=false
     fontMaterial=null
     fontObject=&BurbankSmall
     letterSpacing=0
     monospacedWidth=1
     size=18
     skewAmount=0
     typefaceFontName=Bold
     outlineSettings {
      bApplyOutlineToDropShadows=false
      bMiteredCorners=false
      bSeparateFillAlpha=false
      outlineMaterial=null
      outlineSize=0
      outlineColor=<rgba 0 0 0 1>
     }
    }
    shadowOffset=<0 0>
   }
   slot {class=HorizontalBoxSlot, properties {verticalAlignment=VAlign_Center}}
  }
  Spacer_400 {widgetClassPath=&Spacer, parent=VBox_Content, widget=Spacer_400, properties {size=<1 15>}, slot {class=VerticalBoxSlot}}
  SB_BasicButtons {
   widgetClassPath=&SizeBox
   parent=VBox_Content
   widget=SB_BasicButtons
   properties {heightOverride=88}
   slot {class=VerticalBoxSlot, properties {padding=<ltrb 16 0 16 16>}}
  }
  HorizontalBox_3 {widgetClassPath=&HorizontalBox, parent=SB_BasicButtons, widget=HorizontalBox_3, slot {class=SizeBoxSlot}}
  MenuAnchor_MoreOptions {
   widgetClassPath=/Script/UMG.MenuAnchor
   bIsVariable=true
   parent=HorizontalBox_3
   widget=MenuAnchor_MoreOptions
   slot {class=HorizontalBoxSlot, properties {horizontalAlignment=HAlign_Left, verticalAlignment=VAlign_Center}}
  }
  MoreActionsButton {
   widgetClassPath=&IconTextButton
   bIsVariable=true
   parent=MenuAnchor_MoreOptions
   widget=MoreActionsButton
   properties {
    bIsFocusable=false
    buttonText=""
    contentAlignment=HAlign_Center
    controllerInputStyle=/FortUILibrary/Foundation/Buttons/ButtonStyles/ButtonStyle-BottomBar-Console.ButtonStyle-BottomBar-Console_C
    navigation=&self:WidgetTree.MoreActionsButton.WidgetNavigation_0
    onButtonBaseClicked=(&self:WidgetTree.MoreActionsButton.HandleOnButtonClicked)
    onButtonBaseDoubleClicked=(&self:WidgetTree.MoreActionsButton.HandleOnButtonDoubleClicked)
    onButtonBaseHovered=(&self:WidgetTree.MoreActionsButton.HandleOnButtonHovered)
    onButtonBaseUnhovered=(&self:WidgetTree.MoreActionsButton.HandleOnButtonUnhovered)
    onSelectedChangedBase=(&self:WidgetTree.MoreActionsButton.HandleOnSelectedChanged)
    style=/FortUILibrary/Foundation/Buttons/ButtonStyles/Athena/Athena_ButtonStyle_BlueMenuButton_Settings.Athena_ButtonStyle_BlueMenuButton_Settings_C
    iconBrush {
     bIsDynamicallyLoaded=false
     drawAs=Image
     imageType=NoImage
     mirroring=NoMirror
     resourceName=None
     resourceObject=/Game/UI/SocialPanel/Art/Textures/T_UI_Icon_LocalUserSettings
     tiling=NoTile
     imageSize=<48 48>
     margin=<ltrb 0 0 0 0>
     outlineSettings {
      bUseBrushTransparency=false
      roundingType=HalfHeightRadius
      width=0
      color {colorUseRule=UseColor_Specified, specifiedColor=<rgba 0 0 0 0>}
      cornerRadii=<0 0 0 1>
     }
     tintColor {colorUseRule=UseColor_Specified, specifiedColor=<rgba 1 1 1 1>}
     uVRegion {bIsValid=false, max=<0 0>, min=<0 0>}
    }
    triggeringInputAction {dataTable=&CreativeInputActionDataTable, rowName=MoreActionsPropertyEditor}
   }
   slot {class=PanelSlot}
  }
  HBox_EventBrowserAction {
   widgetClassPath=&HorizontalBox
   parent=HorizontalBox_3
   widget=HBox_EventBrowserAction
   slot {class=HorizontalBoxSlot, properties {padding=<ltrb 40 0 0 0>}}
  }
  Spacer_157 {widgetClassPath=&Spacer, parent=HBox_EventBrowserAction, widget=Spacer_157, slot {class=HorizontalBoxSlot}}
  Button_EventBrowser {
   widgetClassPath=/FortUILibrary/UIKit/Buttons/WBP_UIKit_Button_Quiet
   bInherited=true
   bIsVariable=true
   parent=HBox_EventBrowserAction
   widget=Button_EventBrowser
   properties {text=Event Browser, triggeringInputAction {dataTable=&CreativeInputActionDataTable, rowName=DeviceBrowser}}
   slot {class=HorizontalBoxSlot}
  }
  HBox_SelectedItemAdditionalAction {
   widgetClassPath=&HorizontalBox
   bIsVariable=true
   parent=HorizontalBox_3
   widget=HBox_SelectedItemAdditionalAction
   slot {class=HorizontalBoxSlot, properties {padding=<ltrb 32 0 32 0>}}
  }
  Action_SelectedItemAdditionalAction {
   widgetClassPath=&CommonActionWidget
   bInherited=true
   parent=HBox_SelectedItemAdditionalAction
   widget=Action_SelectedItemAdditionalAction
   properties {visibility=Collapsed}
   slot {class=HorizontalBoxSlot}
  }
  Text_SelectedItemAdditionalAction {
   widgetClassPath=&CommonTextBlock
   bInherited=true
   parent=HBox_SelectedItemAdditionalAction
   widget=Text_SelectedItemAdditionalAction
   properties {
    bAutoCollapseWithEmptyText=true
    scrollStyle=&TextScrollStyle-Base_C
    text="{Additional Action Name}"
    font {
     bForceMonospaced=false
     bMaterialIsStencil=false
     fontMaterial=null
     fontObject=&BurbankSmall
     letterSpacing=0
     monospacedWidth=1
     size=18
     skewAmount=0
     typefaceFontName=Bold
     outlineSettings {
      bApplyOutlineToDropShadows=false
      bMiteredCorners=false
      bSeparateFillAlpha=false
      outlineMaterial=null
      outlineSize=0
      outlineColor=<rgba 0 0 0 1>
     }
    }
    shadowOffset=<0 0>
   }
   slot {class=HorizontalBoxSlot, properties {verticalAlignment=VAlign_Center, padding=<ltrb 12 0 0 0>}}
  }
  OK_Button {
   widgetClassPath=/FortUILibrary/UIKit/Buttons/WBP_UIKit_Button_Loud
   bIsVariable=true
   parent=HorizontalBox_3
   widget=OK_Button
   properties {
    bIsFocusable=false
    clickMethod=MouseDown
    minWidth=160
    pressMethod=ButtonPress
    text=OK
    triggeringInputAction {dataTable=&CreativeInputActionDataTable, rowName=AcceptPropertyEditor}
   }
   slot {
    class=HorizontalBoxSlot
    properties {horizontalAlignment=HAlign_Right, verticalAlignment=VAlign_Center, padding=<ltrb 0 0 24 0>, size {sizeRule=Fill, value=1}}
   }
  }
  OverrideCancelButton {
   widgetClassPath=&WBP_UIKit_Button_Regular
   bIsVariable=true
   parent=HorizontalBox_3
   widget=OverrideCancelButton
   properties {
    bIsFocusable=false
    minWidth=160
    text=CANCEL
    visibility=Visible
    triggeringInputAction {dataTable=&CreativeInputActionDataTable, rowName=Cancel}
   }
   slot {class=HorizontalBoxSlot, properties {verticalAlignment=VAlign_Center}}
  }
  ConfirmSelectionShine {
   widgetClassPath=&Image
   parent=Overlay_Shell
   widget=ConfirmSelectionShine
   properties {
    visibility=SelfHitTestInvisible
    brush {
     bIsDynamicallyLoaded=false
     drawAs=Image
     imageType=NoImage
     mirroring=NoMirror
     resourceName=None
     resourceObject=/Game/UI/Foundation/Materials/M_Athena_Matchmaking_ShineAnim_Details
     tiling=NoTile
     imageSize=<32 32>
     margin=<ltrb 0 0 0 0>
     outlineSettings {
      bUseBrushTransparency=false
      roundingType=HalfHeightRadius
      width=0
      color {colorUseRule=UseColor_Specified, specifiedColor=<rgba 0 0 0 0>}
      cornerRadii=<0 0 0 0>
     }
     tintColor {colorUseRule=UseColor_Specified, specifiedColor=<rgba 1 1 1 1>}
     uVRegion {bIsValid=false, max=<0 0>, min=<0 0>}
    }
   }
   slot {class=OverlaySlot, properties {horizontalAlignment=HAlign_Fill, verticalAlignment=VAlign_Fill}}
  }
  Overlay_ResetMessage {
   widgetClassPath=&Overlay
   parent=Overlay_Shell
   widget=Overlay_ResetMessage
   properties {visibility=Collapsed, renderTransform {angle=0, scale=<1 1>, shear=<-8 0>, translation=<0 0>}}
   slot {class=OverlaySlot, properties {horizontalAlignment=HAlign_Fill, verticalAlignment=VAlign_Bottom, padding=<ltrb 24 0 24 116>}}
  }
  Image_120 {
   widgetClassPath=&Image
   parent=Overlay_ResetMessage
   widget=Image_120
   properties {visibility=SelfHitTestInvisible, colorAndOpacity=<rgba 0.0331 0.855 1 1>}
   slot {class=OverlaySlot, properties {horizontalAlignment=HAlign_Fill, verticalAlignment=VAlign_Fill}}
  }
  Text_Message {
   widgetClassPath=&CommonTextBlock
   bIsVariable=true
   parent=Overlay_ResetMessage
   widget=Text_Message
   properties {
    autoWrapText=true
    text=All settings in this device have been reset to default values.
    colorAndOpacity {colorUseRule=UseColor_Specified, specifiedColor=<rgba 0 0 0 1>}
    font {
     bForceMonospaced=false
     bMaterialIsStencil=false
     fontMaterial=null
     fontObject=&Burbank
     letterSpacing=0
     monospacedWidth=1
     size=22
     skewAmount=0
     typefaceFontName=Burbank-PlayerPower
     outlineSettings {
      bApplyOutlineToDropShadows=false
      bMiteredCorners=false
      bSeparateFillAlpha=false
      outlineMaterial=null
      outlineSize=0
      outlineColor=<rgba 0 0 0 1>
     }
    }
    shadowOffset=<0 0>
   }
   slot {class=OverlaySlot, properties {horizontalAlignment=HAlign_Fill, padding=<ltrb 24 24 24 24>}}
  }
  Image_LightBoxLeftSide {
   widgetClassPath=&Image
   bIsVariable=true
   parent=Overlay_Shell
   widget=Image_LightBoxLeftSide
   properties {visibility=Collapsed, colorAndOpacity=<rgba 0 0 0 0.5>}
   slot {class=OverlaySlot, properties {horizontalAlignment=HAlign_Fill, verticalAlignment=VAlign_Fill}}
  }
  Overlay_RightSide {
   widgetClassPath=&Overlay
   parent=HorizontalBox_0
   widget=Overlay_RightSide
   slot {class=HorizontalBoxSlot, properties {size {sizeRule=Fill, value=1}}}
  }
  SizeBox_0 {
   widgetClassPath=&SizeBox
   parent=Overlay_RightSide
   widget=SizeBox_0
   properties {bOverride_WidthOverride=true, flowDirectionPreference=Culture, widthOverride=500}
   slot {class=OverlaySlot, properties {horizontalAlignment=HAlign_Right, verticalAlignment=VAlign_Bottom, padding=<ltrb 0 0 0 64>}}
  }
  RootBackground {
   widgetClassPath=&CommonBorder
   parent=SizeBox_0
   widget=RootBackground
   properties {style=/FortUILibrary/Styles/Border/Border_Solid_MdBlue.Border_Solid_MdBlue_C, brushColor=<rgba 1 1 1 0.9>, padding=<ltrb 0 0 0 0>}
   slot {class=SizeBoxSlot}
  }
  Overlay_6 {widgetClassPath=&Overlay, parent=RootBackground, widget=Overlay_6, slot {class=BorderSlot, properties {padding=<ltrb 0 0 0 0>}}}
  HeaderBGContainer {
   widgetClassPath=&VerticalBox
   parent=Overlay_6
   widget=HeaderBGContainer
   slot {class=OverlaySlot, properties {horizontalAlignment=HAlign_Fill}}
  }
  HeaderBackground {
   widgetClassPath=&CommonBorder
   parent=HeaderBGContainer
   widget=HeaderBackground
   properties {style=/FortUILibrary/Styles/Border/Border-AthenaItemInfoHeader.Border-AthenaItemInfoHeader_C, padding=<ltrb 0 0 0 0>}
   slot {class=VerticalBoxSlot, properties {size {sizeRule=Fill, value=1}}}
  }
  Spacer_5 {
   widgetClassPath=&Spacer
   parent=HeaderBackground
   widget=Spacer_5
   properties {size=<1 103>}
   slot {class=BorderSlot, properties {padding=<ltrb 0 0 0 0>}}
  }
  HeaderRule {
   widgetClassPath=&CommonBorder
   parent=HeaderBGContainer
   widget=HeaderRule
   properties {
    style=/FortUILibrary/Styles/Border/Border-AthenaItemInfoRule.Border-AthenaItemInfoRule_C
    background {
     bIsDynamicallyLoaded=false
     drawAs=Box
     imageType=NoImage
     mirroring=NoMirror
     resourceName=None
     resourceObject=/FortUILibrary/Styles/Border/Materials/M-Border-ItemInfoHeader_procedural
     tiling=NoTile
     imageSize=<128 32>
     margin=<ltrb 0.1 0.7965 1 0.2611>
     outlineSettings {
      bUseBrushTransparency=false
      roundingType=HalfHeightRadius
      width=0
      color {colorUseRule=UseColor_Specified, specifiedColor=<rgba 0 0 0 0>}
      cornerRadii=<0 0 0 0>
     }
     tintColor {colorUseRule=UseColor_Specified, specifiedColor=<rgba 1 1 1 1>}
     uVRegion {bIsValid=false, max=<0 0>, min=<0 0>}
    }
    padding=<ltrb 0 0 0 0>
   }
   slot {class=VerticalBoxSlot}
  }
  Spacer_6 {
   widgetClassPath=&Spacer
   parent=HeaderRule
   widget=Spacer_6
   properties {size=<1 4>}
   slot {class=BorderSlot, properties {padding=<ltrb 0 0 0 0>}}
  }
  MainHB {
   widgetClassPath=&HorizontalBox
   parent=Overlay_6
   widget=MainHB
   slot {class=OverlaySlot, properties {horizontalAlignment=HAlign_Fill, verticalAlignment=VAlign_Fill, padding=<ltrb 8 8 8 8>}}
  }
  SizeBox_5 {
   widgetClassPath=&SizeBox
   parent=MainHB
   widget=SizeBox_5
   properties {bOverride_WidthOverride=true, widthOverride=456}
   slot {class=HorizontalBoxSlot, properties {horizontalAlignment=HAlign_Left, padding=<ltrb 8 8 8 8>, size {sizeRule=Fill, value=1}}}
  }
  TextContainer {widgetClassPath=&VerticalBox, parent=SizeBox_5, widget=TextContainer, slot {class=SizeBoxSlot}}
  SizeBox {
   widgetClassPath=&SizeBox
   parent=TextContainer
   widget=SizeBox
   properties {bOverride_HeightOverride=true, heightOverride=80}
   slot {class=VerticalBoxSlot, properties {verticalAlignment=VAlign_Top, padding=<ltrb 0 0 0 12>}}
  }
  VerticalBox {widgetClassPath=&VerticalBox, parent=SizeBox, widget=VerticalBox, slot {class=SizeBoxSlot}}
  Overlay_8 {
   widgetClassPath=&Overlay
   parent=VerticalBox
   widget=Overlay_8
   slot {class=VerticalBoxSlot, properties {horizontalAlignment=HAlign_Left, verticalAlignment=VAlign_Top, padding=<ltrb 0 8 0 0>}}
  }
  RarityGlow {
   widgetClassPath=&Image
   parent=Overlay_8
   widget=RarityGlow
   properties {
    visibility=SelfHitTestInvisible
    brush {
     bIsDynamicallyLoaded=false
     drawAs=Box
     imageType=NoImage
     mirroring=NoMirror
     resourceName=None
     resourceObject=/FortUILibrary/Styles/Border/Materials/M_Border_SoftGlow
     tiling=NoTile
     imageSize=<128 32>
     margin=<ltrb 0.5 0 0.5 0>
     outlineSettings {
      bUseBrushTransparency=false
      roundingType=HalfHeightRadius
      width=0
      color {colorUseRule=UseColor_Specified, specifiedColor=<rgba 0 0 0 0>}
      cornerRadii=<0 0 0 0>
     }
     tintColor {colorUseRule=UseColor_Specified, specifiedColor=<rgba 1 1 1 1>}
     uVRegion {bIsValid=false, max=<0 0>, min=<0 0>}
    }
    colorAndOpacity=<rgba 1 1 1 0.4>
   }
   slot {class=OverlaySlot, properties {horizontalAlignment=HAlign_Fill, verticalAlignment=VAlign_Fill, padding=<ltrb -32 -16 -32 -16>}}
  }
  Text_ItemBaseName {
   widgetClassPath=&CommonTextBlock
   bIsVariable=true
   parent=Overlay_8
   widget=Text_ItemBaseName
   properties {
    scrollStyle=&TextScrollStyle-Base_C
    style=/FortUILibrary/Styles/Typography/Foundation/TextStyle-Header-M-SimpleGradient.TextStyle-Header-M-SimpleGradient_C
    text="{ITEMNAME}"
    font {
     bForceMonospaced=false
     bMaterialIsStencil=false
     fontMaterial=/FortUILibrary/Materials/Foundation/M_Font_SimpleGradient_Dynamic
     fontObject=&Burbank
     letterSpacing=0
     monospacedWidth=1
     size=32
     skewAmount=0
     typefaceFontName=Burbank-Black
     outlineSettings {
      bApplyOutlineToDropShadows=false
      bMiteredCorners=false
      bSeparateFillAlpha=false
      outlineMaterial=null
      outlineSize=0
      outlineColor=<rgba 0 0 0 1>
     }
    }
    shadowOffset=<0 0>
   }
   slot {class=OverlaySlot, properties {horizontalAlignment=HAlign_Fill, verticalAlignment=VAlign_Center}}
  }
  Text_ItemDescription {
   widgetClassPath=&CommonTextBlock
   bIsVariable=true
   parent=TextContainer
   widget=Text_ItemDescription
   properties {
    autoWrapText=true
    text="[item description]"
    font {
     bForceMonospaced=false
     bMaterialIsStencil=false
     fontMaterial=null
     fontObject=&BurbankSmall
     letterSpacing=0
     monospacedWidth=1
     size=18
     skewAmount=0
     typefaceFontName=Bold
     outlineSettings {
      bApplyOutlineToDropShadows=false
      bMiteredCorners=false
      bSeparateFillAlpha=false
      outlineMaterial=null
      outlineSize=0
      outlineColor=<rgba 0 0 0 1>
     }
    }
    shadowOffset=<0 0>
   }
   slot {class=VerticalBoxSlot, properties {padding=<ltrb 0 8 0 0>}}
  }
  Button_RightDialogMouseCapture {
   widgetClassPath=&Button
   parent=Overlay_RightSide
   widget=Button_RightDialogMouseCapture
   properties {isFocusable=false, visibility=Collapsed, backgroundColor=<rgba 0 0 0 0>, colorAndOpacity=<rgba 1 1 1 0>}
   slot {class=OverlaySlot, properties {horizontalAlignment=HAlign_Fill, verticalAlignment=VAlign_Fill}}
  }
  CanvasPanel_OptionsDialog {
   widgetClassPath=&CanvasPanel
   bInherited=true
   bIsVariable=true
   parent=Overlay_RightSide
   widget=CanvasPanel_OptionsDialog
   properties {clipping=ClipToBounds}
   slot {class=OverlaySlot, properties {horizontalAlignment=HAlign_Fill, verticalAlignment=VAlign_Fill, padding=<ltrb 8 0 0 0>}}
  }
  BackgroundBlur_ConfirmDialog {
   widgetClassPath=&BackgroundBlur
   bInherited=true
   bIsVariable=true
   parent=Overlay_4
   widget=BackgroundBlur_ConfirmDialog
   properties {blurStrength=15, visibility=Collapsed}
   slot {class=OverlaySlot, properties {horizontalAlignment=HAlign_Fill, verticalAlignment=VAlign_Fill}}
  }
  Overlay_ConfirmDialog {
   widgetClassPath=&Overlay
   bIsVariable=true
   parent=BackgroundBlur_ConfirmDialog
   widget=Overlay_ConfirmDialog
   slot {class=BackgroundBlurSlot, properties {padding=<ltrb 0 0 0 0>}}
  }
  TextBlock_TitleText {widgetClassPath=&CommonTextBlock, bInherited=true, widget=null}
  Image_DisplayImage {widgetClassPath=&Image, bInherited=true, widget=null}
 ]
}
