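# Confirmation dialog for the Creative property editor, a fixed-size skin of AthenaConfirmationWindow.
# Inherited title, rich-text description and button entry box sit over a dark blue background; the other inherited widgets are left out of the layout.
# OnSetButtonText gives IconTextButtons the ButtonStyle class (angled blue by default) and their label.
# If that cast fails, it tries FortSimpleFeaturedYellowButton and sets only the text.
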WidgetBlueprint /Game/Athena/UI/InGame/Creative/PropertyEditor/CreativePropertyEditorConfirmationDialog
graphs [EventGraph {nodes [
   n1 = Event OnSetButtonText!() -> (OutputDelegate, then, Button:Common Button Base Object Reference, Text:"Text (by ref)") {is=title K2Node_Event}
   > n2 = Cast To IconTextButton!(execute, Object=@n1.Button) -> (then, CastFailed, AsIcon Text Button:Icon Text Button Object Reference) {is=title K2Node_DynamicCast}
   > SetStyle!(
    execute
    self:Common Button Base Object Reference=@n2
    InStyle:Common Button Style Class Reference=Get ButtonStyle() -> (ButtonStyle:Common Button Style Class Reference)
   ) -> (then)
   > Set Text!(execute, self:Icon Text Button Object Reference=@n2, ButtonText:Text=@n1.Text) -> (then) {is=title K2Node_CallFunction}
   n3 = Cast To FortSimpleFeaturedYellowButton!(execute=@n2.CastFailed, Object=@n1.Button) -> (then, CastFailed, AsFort Simple Featured Yellow Button:Fort Simple Featured Yellow Button Object Reference) {is=title K2Node_DynamicCast}
   > SetText!(execute, self:Fort Simple Featured Yellow Button Object Reference=@n3, NewText:"Text (by ref)"=@n1.Text) -> (then)
  ]}]
variables [ButtonStyle {
  type=Class
  default=/FortUILibrary/Foundation/Buttons/ButtonStyles/Athena/Athena_ButtonStyle_AngledBlueMenuButton.Athena_ButtonStyle_AngledBlueMenuButton_C
 }]
widgets {
 info {inheritedWidgetCount=7, parentClass=/Script/FortniteUI.AthenaConfirmationWindow, rootWidgetClass=&SizeBox, widgetCount=12}
 widgets [
  SizeBox_0 {
   widgetClassPath=&SizeBox
   widget=SizeBox_0
   properties {bOverride_HeightOverride=true, bOverride_WidthOverride=true, heightOverride=288, widthOverride=832}
  }
  Overlay_0 {widgetClassPath=&Overlay, parent=SizeBox_0, widget=Overlay_0, slot {class=SizeBoxSlot}}
  Background {
   widgetClassPath=&Image
   parent=Overlay_0
   widget=Background
   properties {visibility=SelfHitTestInvisible, colorAndOpacity=<rgba 0 0.0222 0.2874 1>}
   slot {class=OverlaySlot, properties {horizontalAlignment=HAlign_Fill, verticalAlignment=VAlign_Fill}}
  }
  VerticalBox_57 {
   widgetClassPath=&VerticalBox
   parent=Overlay_0
   widget=VerticalBox_57
   slot {class=OverlaySlot, properties {horizontalAlignment=HAlign_Fill, verticalAlignment=VAlign_Fill}}
  }
  CommonBorder_113 {
   widgetClassPath=&CommonBorder
   parent=VerticalBox_57
   widget=CommonBorder_113
   properties {
    verticalAlignment=VAlign_Center
    background {
     bIsDynamicallyLoaded=false
     drawAs=Image
     imageType=NoImage
     mirroring=NoMirror
     resourceName=None
     resourceObject=/Game/UI/Foundation/Materials/M_BoltGradientBorder_LockerHeader
     tiling=NoTile
     imageSize=<32 32>
     margin=<ltrb 0 0 0 0>
     outlineSettings {
      bUseBrushTransparency=false
      roundingType=HalfHeightRadius
      width=0
      color {colorUseRule=UseColor_Specified, specifiedColor=<rgba 0 0 0 0>}
      cornerRadii=<0 0 0 0>
     }
     tintColor {colorUseRule=UseColor_Specified, specifiedColor=<rgba 1 1 1 1>}
     uVRegion {bIsValid=false, max=<0 0>, min=<0 0>}
    }
    padding=<ltrb 16 12 16 12>
   }
   slot {class=VerticalBoxSlot}
  }
  Text_Title {
   widgetClassPath=&CommonTextBlock
   bInherited=true
   parent=CommonBorder_113
   widget=Text_Title
   properties {
    justification=Center
    scrollStyle=&TextScrollStyle-Base_C
    text="{Title}"
    textTransformPolicy=ToUpper
    font {
     bForceMonospaced=false
     bMaterialIsStencil=false
     fontMaterial=null
     fontObject=&Burbank
     letterSpacing=0
     monospacedWidth=1
     size=24
     skewAmount=0
     typefaceFontName=Burbank-Black
     outlineSettings {
      bApplyOutlineToDropShadows=false
      bMiteredCorners=false
      bSeparateFillAlpha=false
      outlineMaterial=null
      outlineSize=0
      outlineColor=<rgba 0 0 0 1>
     }
    }
    shadowOffset=<0 0>
   }
   slot {class=BorderSlot, properties {verticalAlignment=VAlign_Center, padding=<ltrb 16 12 16 12>}}
  }
  RichText_Description {
   widgetClassPath=&CommonRichTextBlock
   bInherited=true
   parent=VerticalBox_57
   widget=RichText_Description
   properties {
    autoWrapText=true
    bOverrideDefaultStyle=true
    clipping=ClipToBounds
    text="{Description}"
    defaultTextStyleOverride {
     overflowPolicy=Clip
     transformPolicy=None
     colorAndOpacity {colorUseRule=UseColor_Specified, specifiedColor=<rgba 1 1 1 1>}
     font {
      bForceMonospaced=false
      bMaterialIsStencil=false
      fontMaterial=null
      fontObject=&BurbankSmall
      letterSpacing=0
      monospacedWidth=1
      size=18
      skewAmount=0
      typefaceFontName=Bold
      outlineSettings {
       bApplyOutlineToDropShadows=false
       bMiteredCorners=false
       bSeparateFillAlpha=false
       outlineMaterial=null
       outlineSize=0
       outlineColor=<rgba 0 0 0 1>
      }
     }
     highlightColor {colorUseRule=UseColor_Specified, specifiedColor=<rgba 0 0 0 1>}
     highlightShape {
      bIsDynamicallyLoaded=false
      drawAs=Image
      imageType=NoImage
      mirroring=NoMirror
      resourceName=None
      resourceObject=null
      tiling=NoTile
      imageSize=<32 32>
      margin=<ltrb 0 0 0 0>
      outlineSettings {
       bUseBrushTransparency=false
       roundingType=HalfHeightRadius
       width=0
       color {colorUseRule=UseColor_Specified, specifiedColor=<rgba 0 0 0 0>}
       cornerRadii=<0 0 0 0>
      }
      tintColor {colorUseRule=UseColor_Specified, specifiedColor=<rgba 1 1 1 1>}
      uVRegion {bIsValid=false, max=<0 0>, min=<0 0>}
     }
     selectedBackgroundColor {colorUseRule=UseColor_Foreground, specifiedColor=<rgba 1 0 1 1>}
     shadowColorAndOpacity=<rgba 0 0 0 1>
     shadowOffset=<0 0>
     strikeBrush {
      bIsDynamicallyLoaded=false
      drawAs=Image
      imageType=NoImage
      mirroring=NoMirror
      resourceName=None
      resourceObject=null
      tiling=NoTile
      imageSize=<32 32>
      margin=<ltrb 0 0 0 0>
      outlineSettings {
       bUseBrushTransparency=false
       roundingType=HalfHeightRadius
       width=0
       color {colorUseRule=UseColor_Specified, specifiedColor=<rgba 0 0 0 0>}
       cornerRadii=<0 0 0 0>
      }
      tintColor {colorUseRule=UseColor_Specified, specifiedColor=<rgba 1 1 1 1>}
      uVRegion {bIsValid=false, max=<0 0>, min=<0 0>}
     }
     underlineBrush {
      bIsDynamicallyLoaded=false
      drawAs=Image
      imageType=NoImage
      mirroring=NoMirror
      resourceName=None
      resourceObject=null
      tiling=NoTile
      imageSize=<32 32>
      margin=<ltrb 0 0 0 0>
      outlineSettings {
       bUseBrushTransparency=false
       roundingType=HalfHeightRadius
       width=0
       color {colorUseRule=UseColor_Specified, specifiedColor=<rgba 0 0 0 0>}
       cornerRadii=<0 0 0 0>
      }
      tintColor {colorUseRule=UseColor_Specified, specifiedColor=<rgba 1 1 1 1>}
      uVRegion {bIsValid=false, max=<0 0>, min=<0 0>}
     }
    }
   }
   slot {class=VerticalBoxSlot, properties {verticalAlignment=VAlign_Center, padding=<ltrb 24 32 24 32>, size {sizeRule=Fill, value=1}}}
  }
  EntryBox_Buttons {
   widgetClassPath=&DynamicEntryBox
   bInherited=true
   parent=VerticalBox_57
   widget=EntryBox_Buttons
   properties {entryWidgetClass=/FortUILibrary/Foundation/Buttons/IconTextButton.IconTextButton_C, entrySpacing=<8 0>}
   slot {class=VerticalBoxSlot, properties {horizontalAlignment=HAlign_Center, padding=<ltrb 0 0 0 16>}}
  }
  Text_PostConfirmation {widgetClassPath=&CommonTextBlock, bInherited=true, widget=null}
  Image_MainIcon {widgetClassPath=&FortLazyImage, bInherited=true, widget=null}
  Button_CloseTouch {widgetClassPath=&CommonButtonBase, bInherited=true, widget=null}
  Border_TapToCloseZone {widgetClassPath=&CommonBorder, bInherited=true, widget=null}
 ]
}
