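# Creative property editor background widget with a center fill, an optional child marker and spacer, and optional skewed left and right edges.
# PreConstruct assigns edge materials, sets widths, outward translations, skew and feather through dynamic material parameters, then applies colors and visibility.
# Unless OverrideEnums is set, UpdateBackground then maps Type, MarkerState and FocusState enums to colors and part visibility.
# The Update Background Style event runs the same mapping on demand.
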WidgetBlueprint /Game/Athena/UI/InGame/Creative/PropertyEditor/WBP_CreativePropertyEditor_Background_SkewedWithMarker
graphs [
 EventGraph {nodes [
   n1 = Event Update Background Style!() -> (
    OutputDelegate
    then
    InType:E_UI_CreativePropertyEditor_Type Enum=NewEnumerator4
    InMarkerState:E_UI_CreativePropertyEditor_MarkerStates Enum=NewEnumerator0
    InFocusState:E_UI_CreativePropertyEditor_FocusStates Enum=NewEnumerator0
   ) {is=title K2Node_CustomEvent}
   > UpdateBackground!(
    execute
    self
    InType:E_UI_CreativePropertyEditor_Type Enum=@n1.InType
    InMarkerState:E_UI_CreativePropertyEditor_MarkerStates Enum=@n1.InMarkerState
    InFocusState:E_UI_CreativePropertyEditor_FocusStates Enum=@n1.InFocusState
   ) -> (then)
   Event PreConstruct()
   > SetMaterialsInstances()
   > PreconstructWidgetSettings()
   > n2 = Branch(Condition=Get OverrideEnums() -> (OverrideEnums:bool))
   UpdateBackground(
    execute=@n2.else
    InType=Get Type() -> (Type:E_UI_CreativePropertyEditor_Type Enum=NewEnumerator4)
    InMarkerState=Get MarkerState() -> (MarkerState:E_UI_CreativePropertyEditor_MarkerStates Enum=NewEnumerator0)
    InFocusState=Get FocusState() -> (FocusState:E_UI_CreativePropertyEditor_FocusStates Enum=NewEnumerator0)
   )
   Event Tick()
  ]}
 PreconstructWidgetSettings {nodes [
   PreconstructWidgetSettings() {is=title K2Node_FunctionEntry}
   > SetSkewEdges()
   > SetChildMarker()
   > Update Colors!(
    execute
    self
    bIsMarkerDisplayed:bool=Get DisplayMarker() -> (DisplayMarker:bool)
    InBackgroundColor:Linear Color Structure=Get BackgroundColor() -> (BackgroundColor:Linear Color Structure)
    InMarkerColor:Linear Color Structure=Get MarkerColor() -> (MarkerColor:Linear Color Structure)
   ) -> (then)
   > UpdateSkewAndMarkerVisibility!(
    execute
    self
    inDisplayMarker:bool=Get DisplayMarker()
    InDisplayLeftSkew:bool=Get DisplayLeftSkew() -> (DisplayLeftSkew:bool)
    InDisplayRightSkew:bool=Get DisplayRightSkew() -> (DisplayRightSkew:bool)
   ) -> (then)
  ]}
 SetChildMarker {nodes [
   SetChildMarker() {is=title K2Node_FunctionEntry}
   > SetDesiredSizeOverride!(
    execute
    self:Image Object Reference=Get Image_ChildMarker() -> (Image_ChildMarker:Image Object Reference)
    DesiredSize_X:double=Get ChildMarkerWidth() -> (ChildMarkerWidth:double)
    DesiredSize_Y:double=0
   ) -> (then)
   > SetSize!(
    execute
    self:Spacer Object Reference=Get Spacer() -> (Spacer:Spacer Object Reference)
    InSize_X:double=Get ChildMarkerSpacerWidth() -> (ChildMarkerSpacerWidth:double)
    InSize_Y:double=0
   ) -> (then)
  ]}
 SetMaterialsInstances {nodes [
   SetMaterialsInstances() {is=title K2Node_FunctionEntry}
   > SetBrushFromMaterial(
    self=Get Image_SkewedEdge_Left() -> (Image_SkewedEdge_Left:Image Object Reference)
    Material=Get LeftSkewEdgeMaterial() -> (LeftSkewEdgeMaterial:Material Instance Object Reference)
   )
   > SetBrushFromMaterial(
    self=Get Image_SkewedEdge_Right() -> (Image_SkewedEdge_Right:Image Object Reference)
    Material=Get RightSkewEdgeMaterial() -> (RightSkewEdgeMaterial:Material Instance Object Reference)
   )
  ]}
 SetSkewEdges {nodes [
   SetPadding!(
    execute
    self:Horizontal Box Slot Object Reference=SlotAsHorizontalBoxSlot!(Widget:Widget Object Reference=Get Image_SkewedEdge_Left()) -> (ReturnValue:Horizontal Box Slot Object Reference)
    InPadding_Left:float=float * float(A=Get LeftSkewEdgeWidth() -> (LeftSkewEdgeWidth:double), B=-1) {is=title K2Node_PromotableOperator}
    InPadding_Top:float=0
    InPadding_Right:float=0
    InPadding_Bottom:float=0
   ) -> (then)
   > SetPadding(
    self=SlotAsHorizontalBoxSlot(Widget=Get Image_SkewedEdge_Right())
    InPadding_Left=0
    InPadding_Top=0
    InPadding_Right=float * float(A=Get RightSkewEdgeWidth() -> (RightSkewEdgeWidth:double), B=-1)
    InPadding_Bottom=0
   )
   SetSkewEdges() {is=title K2Node_FunctionEntry}
   > n1 = Sequence!(execute) -> (then_0, then_1, then_2:Exec, then_3:Exec, then_4:Exec)
   SetRenderTranslation!(
    execute=@n1.then_1
    self=Get Image_SkewedEdge_Left()
    Translation_X:double=float * float(A=Get LeftSkewEdgeWidth(), B=-1)
    Translation_Y:double=0
   ) -> (then)
   n2 = Get RightSkewEdgeWidth()
   > SetRenderTranslation(self=Get Image_SkewedEdge_Right(), Translation_X=@n2, Translation_Y=0)
   SetDesiredSizeOverride(execute=@n1.then_0, self=Get Image_SkewedEdge_Left(), DesiredSize_X=Get LeftSkewEdgeWidth(), DesiredSize_Y=0)
   > SetDesiredSizeOverride(self=Get Image_SkewedEdge_Right(), DesiredSize_X=Get RightSkewEdgeWidth(), DesiredSize_Y=0)
   n3 = GetDynamicMaterial(execute=@n1.then_3, self=Get Image_SkewedEdge_Left())
   > SetScalarParameterValue(self=@n3, ParameterName=FeatherMin, Value=Get LeftFeatherMin() -> (LeftFeatherMin:double))
   > SetScalarParameterValue(self=@n3, ParameterName="FeatherMax (Length)", Value="Get LeftFeatherMax (Length)"() -> ("LeftFeatherMax (Length)":double))
   > n4 = GetDynamicMaterial(self=Get Image_SkewedEdge_Right())
   > SetScalarParameterValue(self=@n4, ParameterName=FeatherMin, Value=Get RightFeatherMin() -> (RightFeatherMin:double))
   > SetScalarParameterValue(self=@n4, ParameterName="FeatherMax (Length)", Value="Get RightFeatherMax (Length)"() -> ("RightFeatherMax (Length)":double))
   n5 = GetDynamicMaterial(execute=@n1.then_2, self=Get Image_SkewedEdge_Left())
   > SetScalarParameterValue(self=@n5, ParameterName=SkewAmount, Value=Get LeftSkewAmount() -> (LeftSkewAmount:double))
   > n6 = GetDynamicMaterial(self=Get Image_SkewedEdge_Right())
   > SetScalarParameterValue(self=@n6, ParameterName=SkewAmount, Value=Get RightSkewAmount() -> (RightSkewAmount:double))
   float * float(A=@n2, B=-1)
  ]}
 Update Colors {nodes [
   n1 = Update Colors!() -> (then, bIsMarkerDisplayed:bool=~, InBackgroundColor:Linear Color Structure, InMarkerColor:Linear Color Structure) {is=title K2Node_FunctionEntry}
   > n2 = Branch(Condition=@n1)
   > SetColorAndOpacity(
    self:Image Object Reference=[Get Image_Background_Center() -> (Image_Background_Center:Image Object Reference), Get Image_SkewedEdge_Right()]
    InColorAndOpacity:Linear Color Structure=Get InBackgroundColor() -> (InBackgroundColor:Linear Color Structure)
   )
   > SetColorAndOpacity(self=[Get Image_SkewedEdge_Left(), Get Image_ChildMarker()], InColorAndOpacity=Get InMarkerColor() -> (InMarkerColor:Linear Color Structure))
   SetColorAndOpacity(
    execute=@n2.else
    self=[Get Image_Background_Center(), Get Image_SkewedEdge_Right(), Get Image_SkewedEdge_Left()]
    InColorAndOpacity=Get InBackgroundColor()
   )
   > SetColorAndOpacity(self=Get Image_ChildMarker(), InColorAndOpacity=Get InMarkerColor())
  ]}
 UpdateBackground {nodes [
   UpdateBackground!() -> (
    then
    InType:E_UI_CreativePropertyEditor_Type Enum=NewEnumerator4
    InMarkerState:E_UI_CreativePropertyEditor_MarkerStates Enum=NewEnumerator0
    InFocusState:E_UI_CreativePropertyEditor_FocusStates Enum=NewEnumerator0
   ) {is=title K2Node_FunctionEntry}
   n1 = Get MarkerColor()
   n2 = Get InFocusState() -> (InFocusState:E_UI_CreativePropertyEditor_FocusStates Enum=NewEnumerator0)
   n3 = Select!(
    NewEnumerator0:Linear Color Structure=Get BackgroundColor()
    NewEnumerator1:Linear Color Structure=(R=1,G=1,B=1,A=1)
    Index:E_UI_CreativePropertyEditor_FocusStates Enum=@n2
   ) -> (ReturnValue:Linear Color Structure)
   n4 = Get InType() -> (InType:E_UI_CreativePropertyEditor_Type Enum=NewEnumerator4)
   n5 = Select!(
    NewEnumerator4:bool=false
    NewEnumerator1:bool=true
    NewEnumerator2:bool=true
    NewEnumerator3:bool=false
    Index:E_UI_CreativePropertyEditor_Type Enum=@n4
   ) -> (ReturnValue:bool)
   > Update Colors(
    bIsMarkerDisplayed=@n5
    InBackgroundColor=@n3
    InMarkerColor=Select!(
     NewEnumerator0:Linear Color Structure=Select(NewEnumerator0=@n1, NewEnumerator1=@n3, Index:E_UI_CreativePropertyEditor_FocusStates Enum=@n2)
     NewEnumerator1:Linear Color Structure=(R=1,G=1,B=1,A=1)
     Index:E_UI_CreativePropertyEditor_MarkerStates Enum=Get InMarkerState() -> (InMarkerState:E_UI_CreativePropertyEditor_MarkerStates Enum=NewEnumerator0)
    ) -> (ReturnValue:Linear Color Structure)
   ) {is=title K2Node_CallFunction}
   n6 = Select!(
    NewEnumerator4:bool=false
    NewEnumerator1:bool=false
    NewEnumerator2:bool=true
    NewEnumerator3:bool=true
    Index:E_UI_CreativePropertyEditor_Type Enum=@n4
   ) -> (ReturnValue:bool)
   n7 = Select!(NewEnumerator0:bool=false, NewEnumerator1=true, Index:E_UI_CreativePropertyEditor_FocusStates Enum=@n2) -> (ReturnValue)
   n8 = Get InType() -> (InType=NewEnumerator4)
   > UpdateSkewAndMarkerVisibility(
    inDisplayMarker=@n5
    InDisplayLeftSkew=Select!(
     NewEnumerator4:bool=false
     NewEnumerator1=false
     NewEnumerator2:bool=true
     NewEnumerator3:bool=true
     Index:E_UI_CreativePropertyEditor_Type Enum=@n4
    ) -> (ReturnValue)
    InDisplayRightSkew=OR Boolean(
     A=AND Boolean(A=@n6, B=@n7)
     B="Equal (Enum)"(A:E_UI_CreativePropertyEditor_Type Enum=@n8, B:E_UI_CreativePropertyEditor_Type Enum=NewEnumerator3)
    )
   )
   Find!(
    TargetMap:Map of E_UI_CreativePropertyEditor_BackgroundPresets Enums to S UI Creative Property Editor Background Styles Structures=Get ChannelsBindingsBackgroundPresets_Map!(self:UI PDA Creative Property Editor Object Reference=Get PrimaryDataAsset() -> (PrimaryDataAsset:UI PDA Creative Property Editor Object Reference)) -> (ChannelsBindingsBackgroundPresets_Map:Map of E_UI_CreativePropertyEditor_BackgroundPresets Enums to S UI Creative Property Editor Background Styles Structures=NewEnumerator0)
    Key:"E_UI_CreativePropertyEditor_BackgroundPresets Enum (by ref)"=NewEnumerator0
   ) -> (
    Value_DisplayMarker_1_#1:bool=~
    Value_CollapseSpacer_14_#2:bool=~
    Value_DisplayLeftSkew_3_#3:bool=~
    Value_DisplayRightSkew_5_#4:bool=~
    Value_MarkerColor_8_#5:Linear Color Structure
    Value_BackgroundColor_11_#6:Linear Color Structure
    ReturnValue
   )
  ]}
 UpdateSkewAndMarkerVisibility {nodes [
   UpdateSkewAndMarkerVisibility!() -> (then, inDisplayMarker:bool=~, InDisplayLeftSkew:bool=~, InDisplayRightSkew:bool=~) {is=title K2Node_FunctionEntry}
   > n1 = Set Visibility(
    Visibility=Select!(
     Option 0:ESlateVisibility Enum=Collapsed
     Option 1:ESlateVisibility Enum=HitTestInvisible
     Index:bool=Get inDisplayMarker() -> (inDisplayMarker:bool)
    ) -> (ReturnValue:ESlateVisibility Enum)
    self=Get Image_ChildMarker()
   )
   > Set Visibility(Visibility=@n1, self=Get Spacer())
   > Set Visibility(
    Visibility=Select(Option 0=Collapsed, Option 1=HitTestInvisible, Index=Get InDisplayLeftSkew() -> (InDisplayLeftSkew:bool))
    self=Get Image_SkewedEdge_Left()
   )
   > Set Visibility(
    Visibility=Select(Option 0=Collapsed, Option 1=HitTestInvisible, Index=Get InDisplayRightSkew() -> (InDisplayRightSkew:bool))
    self=Get Image_SkewedEdge_Right()
   )
  ]}
 UpdateStyleVariables {nodes [
   UpdateStyleVariables!() -> (
    then
    bInDisplayMarker:bool=~
    bInDisplayLeftSkew:bool=~
    bInDisplayRightSkew:bool=~
    inBackgroundColor:Linear Color Structure
    inMarkerColor:Linear Color Structure
   ) {is=title K2Node_FunctionEntry}
   > n1 = Sequence!(execute) -> (then_0, then_1)
   > Set DisplayMarker(DisplayMarker:bool=Get bInDisplayMarker() -> (bInDisplayMarker:bool)) -> (Output_Get:bool)
   > Set DisplayLeftSkew(DisplayLeftSkew:bool=Get bInDisplayLeftSkew() -> (bInDisplayLeftSkew:bool)) -> (Output_Get:bool)
   > Set DisplayRightSkew(DisplayRightSkew:bool=Get bInDisplayRightSkew() -> (bInDisplayRightSkew:bool)) -> (Output_Get:bool)
   Set MarkerColor(execute=@n1.then_1, MarkerColor:Linear Color Structure=Get inMarkerColor() -> (inMarkerColor:Linear Color Structure)) -> (Output_Get:Linear Color Structure)
   > Set BackgroundColor(BackgroundColor:Linear Color Structure=Get inBackgroundColor() -> (inBackgroundColor:Linear Color Structure)) -> (Output_Get:Linear Color Structure)
  ]}
]
variables [
 BackgroundColor {type=LinearColor, default=<rgba 0.0232 0.0273 0.0908 1>}
 ChildMarkerSpacerWidth {type=number, default=2}
 ChildMarkerWidth {type=number, default=9}
 CollapseSpacer {type=boolean}
 DisplayLeftSkew {type=boolean, default=true}
 DisplayMarker {type=boolean, default=true}
 DisplayRightSkew {type=boolean, default=true}
 FocusState {type=E_UI_CreativePropertyEditor_FocusStates, default=Default}
 "LeftFeatherMax (Length)" {type=number, default=1}
 LeftFeatherMin {type=number, default=0.9}
 LeftSkewAmount {type=number, default=0.18}
 LeftSkewEdgeMaterial {type=MaterialInstance, default=/Game/Athena/UI/InGame/Creative/PropertyEditor/Art/MI_UI_SingleEdgeSkew_CreativePropertyEditor_LeftMarker}
 LeftSkewEdgeWidth {type=number, default=16}
 MarkerColor {type=LinearColor, default=<rgba 0.1022 0.1413 0.2874 1>}
 MarkerState {type=E_UI_CreativePropertyEditor_MarkerStates, default=Default}
 OverrideEnums {type=boolean}
 PrimaryDataAsset {type=UI_PDA_Creative_PropertyEditor_C, default=/Game/Athena/UI/InGame/Creative/PropertyEditor/DataAssets/UI_DA_Creative_PropertyEditor}
 "RightFeatherMax (Length)" {type=number, default=1}
 RightFeatherMin {type=number, default=0.9}
 RightSkewAmount {type=number, default=0.18}
 RightSkewEdgeMaterial {type=MaterialInstance, default=/Game/Athena/UI/InGame/Creative/PropertyEditor/Art/MI_UI_SingleEdgeSkew_CreativePropertyEditor_RightMarker}
 RightSkewEdgeWidth {type=number, default=16}
 Type {type=E_UI_CreativePropertyEditor_Type, default=Skewed}
]
widgets {
 info {inheritedWidgetCount=0, parentClass=&UserWidget, rootWidgetClass=&Overlay, widgetCount=7}
 widgets [
  Root {widgetClassPath=&Overlay, widget=Root}
  Image_SkewedEdge_Left {
   widgetClassPath=&Image
   bIsVariable=true
   parent=Root
   widget=Image_SkewedEdge_Left
   properties {visibility=HitTestInvisible}
   slot {class=OverlaySlot, properties {verticalAlignment=VAlign_Fill}}
  }
  Main {
   widgetClassPath=&HorizontalBox
   parent=Root
   widget=Main
   properties {visibility=HitTestInvisible}
   slot {class=OverlaySlot, properties {horizontalAlignment=HAlign_Fill, verticalAlignment=VAlign_Fill}}
  }
  Image_ChildMarker {
   widgetClassPath=&Image
   bIsVariable=true
   parent=Main
   widget=Image_ChildMarker
   properties {visibility=HitTestInvisible}
   slot {class=HorizontalBoxSlot}
  }
  Spacer {
   widgetClassPath=&Spacer
   bIsVariable=true
   parent=Main
   widget=Spacer
   properties {visibility=HitTestInvisible, size=<2 1>}
   slot {class=HorizontalBoxSlot, properties {horizontalAlignment=HAlign_Left}}
  }
  Image_Background_Center {
   widgetClassPath=&Image
   bIsVariable=true
   parent=Main
   widget=Image_Background_Center
   properties {visibility=HitTestInvisible}
   slot {class=HorizontalBoxSlot, properties {size {sizeRule=Fill, value=1}}}
  }
  Image_SkewedEdge_Right {
   widgetClassPath=&Image
   bIsVariable=true
   parent=Root
   widget=Image_SkewedEdge_Right
   properties {visibility=HitTestInvisible}
   slot {class=OverlaySlot, properties {horizontalAlignment=HAlign_Right, verticalAlignment=VAlign_Fill}}
  }
 ]
}
