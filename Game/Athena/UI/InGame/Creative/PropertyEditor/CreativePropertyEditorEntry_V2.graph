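# Creative property editor row (V2), a FortCreativePropertyKnobEntry widget showing one property's name, value and modified marker.
# A switcher holds the summary row, with a gradient fade over the name, an extra-content slot and collapsed debug text, alongside an empty content panel.
# SetDesiredValueWidth resizes the value column.
# Summary init and modify events call UpdateModifiedState, which shows or hides the icon and passes the flag to SetIsModified.
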WidgetBlueprint /Game/Athena/UI/InGame/Creative/PropertyEditor/CreativePropertyEditorEntry_V2
graphs [
 EventGraph {nodes [
   n1 = Event SetDesiredValueWidth!() -> (OutputDelegate, then, DesiredValueWidth:float) {is=title K2Node_Event}
   > SetWidthOverride(self=Get SizeBox_Value() -> (SizeBox_Value:Size Box Object Reference), InWidthOverride=@n1.DesiredValueWidth)
   Event BP_PostInitSummaryData!() -> (OutputDelegate, then, CustomEditWidget:Creative Property Edit Widget Base Object Reference) {is=title K2Node_Event}
   > UpdateModifiedState!(execute, self, isModifiedFromDefaultValue:bool=GetIsModified!() -> (ReturnValue:bool)) -> (then)
   n2 = Event BP_OnModified!() -> (OutputDelegate, then, bModified:bool) {is=title K2Node_Event}
   > UpdateModifiedState(isModifiedFromDefaultValue=@n2.bModified)
  ]}
 UpdateModifiedState {nodes [
   n1 = UpdateModifiedState!() -> (then, isModifiedFromDefaultValue:bool=~) {is=title K2Node_FunctionEntry}
   > SetVisibility(
    self=Get Image_Modified() -> (Image_Modified:Image Object Reference)
    InVisibility=Select(Option 0=Hidden, Option 1=SelfHitTestInvisible, Index=@n1)
   )
   > SetIsModified!(
    execute
    self:Fort Creative Property Knob Entry Object Reference
    bModified:bool=Get isModifiedFromDefaultValue() -> (isModifiedFromDefaultValue:bool)
   ) -> (then)
  ]}
]
variables []
widgets {
 info {
  inheritedWidgetCount=12
  parentClass=/Script/FortniteUI.FortCreativePropertyKnobEntry
  rootWidgetClass=&CommonVisibilitySwitcher
  widgetCount=25
 }
 widgets [
  Switcher_Content {widgetClassPath=&CommonVisibilitySwitcher, bInherited=true, widget=Switcher_Content}
  Panel_SummaryBox {
   widgetClassPath=&Overlay
   bInherited=true
   parent=Switcher_Content
   widget=Panel_SummaryBox
   slot {class=CommonVisibilitySwitcherSlot}
  }
  SizeBox_1 {
   widgetClassPath=&SizeBox
   parent=Panel_SummaryBox
   widget=SizeBox_1
   properties {bOverride_HeightOverride=true, heightOverride=82, minDesiredHeight=56, widthOverride=1000}
   slot {class=OverlaySlot, properties {horizontalAlignment=HAlign_Fill, verticalAlignment=VAlign_Center}}
  }
  Overlay_2 {widgetClassPath=&Overlay, parent=SizeBox_1, widget=Overlay_2, slot {class=SizeBoxSlot, properties {padding=<ltrb 16 2 16 2>}}}
  WBP_Background_SkewedWithMarker {
   widgetClassPath=&/WBP_CreativePropertyEditor_Background_SkewedWithMarker
   parent=Overlay_2
   widget=WBP_Background_SkewedWithMarker
   properties {type=Default}
   slot {class=OverlaySlot, properties {horizontalAlignment=HAlign_Fill, verticalAlignment=VAlign_Fill}}
  }
  HorizontalBox_0 {
   widgetClassPath=&HorizontalBox
   parent=Overlay_2
   widget=HorizontalBox_0
   slot {class=OverlaySlot, properties {horizontalAlignment=HAlign_Fill, verticalAlignment=VAlign_Fill}}
  }
  Image_Modified {
   widgetClassPath=&Image
   bIsVariable=true
   parent=HorizontalBox_0
   widget=Image_Modified
   properties {
    visibility=Collapsed
    brush {
     bIsDynamicallyLoaded=false
     drawAs=Image
     imageType=NoImage
     mirroring=NoMirror
     resourceName=None
     resourceObject=/Game/Athena/UI/InGame/Creative/PropertyEditor/Art/Icon-PropertyEditor-Modified
     tiling=NoTile
     imageSize=<20 20>
     margin=<ltrb 0 0 0 0>
     outlineSettings {
      bUseBrushTransparency=false
      roundingType=HalfHeightRadius
      width=0
      color {colorUseRule=UseColor_Specified, specifiedColor=<rgba 0 0 0 0>}
      cornerRadii=<0 0 0 0>
     }
     tintColor {colorUseRule=UseColor_Specified, specifiedColor=<rgba 1 1 1 1>}
     uVRegion {bIsValid=false, max=<0 0>, min=<0 0>}
    }
    colorAndOpacity=<rgba 0.3372 0.402 0.5906 1>
   }
   slot {class=HorizontalBoxSlot, properties {horizontalAlignment=HAlign_Left, verticalAlignment=VAlign_Center, padding=<ltrb 6 0 4 0>}}
  }
  SizeBox_0 {
   widgetClassPath=&SizeBox
   parent=HorizontalBox_0
   widget=SizeBox_0
   properties {clipping=ClipToBounds, maxDesiredWidth=669.8986, widthOverride=550}
   slot {class=HorizontalBoxSlot, properties {verticalAlignment=VAlign_Center, padding=<ltrb 0 0 24 0>, size {sizeRule=Fill, value=1}}}
  }
  Overlay_0 {
   widgetClassPath=&Overlay
   parent=SizeBox_0
   widget=Overlay_0
   slot {class=SizeBoxSlot, properties {verticalAlignment=VAlign_Center}}
  }
  Text_DisplayName {
   widgetClassPath=&CommonTextBlock
   bInherited=true
   parent=Overlay_0
   widget=Text_DisplayName
   properties {
    text="[Parameter name]"
    colorAndOpacity {colorUseRule=UseColor_Specified, specifiedColor=<rgba 0.3372 0.402 0.5906 1>}
    font {
     bForceMonospaced=false
     bMaterialIsStencil=false
     fontMaterial=null
     fontObject=&BurbankSmall
     letterSpacing=0
     monospacedWidth=1
     size=18
     skewAmount=0
     typefaceFontName=Bold
     outlineSettings {
      bApplyOutlineToDropShadows=false
      bMiteredCorners=false
      bSeparateFillAlpha=false
      outlineMaterial=null
      outlineSize=0
      outlineColor=<rgba 0.0052 0.0052 0.0056 0.2>
     }
    }
   }
   slot {class=OverlaySlot, properties {horizontalAlignment=HAlign_Fill, verticalAlignment=VAlign_Center}}
  }
  Image_TextNameMask {
   widgetClassPath=&Image
   parent=Overlay_0
   widget=Image_TextNameMask
   properties {
    visibility=SelfHitTestInvisible
    brush {
     bIsDynamicallyLoaded=false
     drawAs=Image
     imageType=NoImage
     mirroring=NoMirror
     resourceName=None
     resourceObject=/Game/UI/Style/Materials/BasicLinearUIGradient_CommonBorder_Inst
     tiling=NoTile
     imageSize=<32 36.3048>
     margin=<ltrb 0 0 0 0>
     outlineSettings {
      bUseBrushTransparency=false
      roundingType=HalfHeightRadius
      width=0
      color {colorUseRule=UseColor_Specified, specifiedColor=<rgba 0 0 0 0>}
      cornerRadii=<0 0 0 0>
     }
     tintColor {colorUseRule=UseColor_Specified, specifiedColor=<rgba 1 1 1 1>}
     uVRegion {bIsValid=false, max=<0 0>, min=<0 0>}
    }
    colorAndOpacity=<rgba 0.0232 0.0273 0.0908 1>
   }
   slot {class=OverlaySlot, properties {horizontalAlignment=HAlign_Right, verticalAlignment=VAlign_Center}}
  }
  SizeBox_Value {
   widgetClassPath=&SizeBox
   bIsVariable=true
   parent=HorizontalBox_0
   widget=SizeBox_Value
   properties {bOverride_WidthOverride=true, heightOverride=92, widthOverride=320}
   slot {class=HorizontalBoxSlot, properties {horizontalAlignment=HAlign_Right, padding=<ltrb 0 0 12 0>}}
  }
  Overlay_1 {widgetClassPath=&Overlay, parent=SizeBox_Value, widget=Overlay_1, properties {clipping=ClipToBounds}, slot {class=SizeBoxSlot}}
  HorizontalBox_1 {
   widgetClassPath=&HorizontalBox
   parent=Overlay_1
   widget=HorizontalBox_1
   properties {flowDirectionPreference=RightToLeft}
   slot {class=OverlaySlot, properties {horizontalAlignment=HAlign_Center, verticalAlignment=VAlign_Center, padding=<ltrb 20 0 20 0>}}
  }
  NamedSlot_ExtraContent {
   widgetClassPath=&NamedSlot
   bInherited=true
   parent=HorizontalBox_1
   widget=NamedSlot_ExtraContent
   properties {bExposeOnInstanceOnly=true}
   slot {class=HorizontalBoxSlot, properties {padding=<ltrb 0 0 10 0>}}
  }
  Overlay_3 {widgetClassPath=&Overlay, parent=HorizontalBox_1, widget=Overlay_3, slot {class=HorizontalBoxSlot}}
  Text_DisplayedValue {
   widgetClassPath=&CommonTextBlock
   bInherited=true
   parent=Overlay_3
   widget=Text_DisplayedValue
   properties {
    clipping=ClipToBounds
    text="[VALUE]"
    colorAndOpacity {colorUseRule=UseColor_Specified, specifiedColor=<rgba 0.9823 0.9823 0.9823 1>}
    font {
     bForceMonospaced=false
     bMaterialIsStencil=false
     fontMaterial=null
     fontObject=&BurbankSmall
     letterSpacing=0
     monospacedWidth=1
     size=18
     skewAmount=0
     typefaceFontName=Bold
     outlineSettings {
      bApplyOutlineToDropShadows=false
      bMiteredCorners=false
      bSeparateFillAlpha=false
      outlineMaterial=null
      outlineSize=0
      outlineColor=<rgba 0 0 0 0>
     }
    }
    shadowColorAndOpacity=<rgba 0.0052 0.0052 0.0056 0>
   }
   slot {class=OverlaySlot, properties {horizontalAlignment=HAlign_Center, verticalAlignment=VAlign_Center}}
  }
  Image_TextMask {
   widgetClassPath=&Image
   parent=Overlay_1
   widget=Image_TextMask
   properties {
    renderOpacity=0
    visibility=SelfHitTestInvisible
    brush {
     bIsDynamicallyLoaded=false
     drawAs=Image
     imageType=NoImage
     mirroring=NoMirror
     resourceName=None
     resourceObject=&BasicLinearUIGradient_CommonBorder_Inst
     tiling=NoTile
     imageSize=<32 36.3048>
     margin=<ltrb 0 0 0 0>
     outlineSettings {
      bUseBrushTransparency=false
      roundingType=HalfHeightRadius
      width=0
      color {colorUseRule=UseColor_Specified, specifiedColor=<rgba 0 0 0 0>}
      cornerRadii=<0 0 0 0>
     }
     tintColor {colorUseRule=UseColor_Specified, specifiedColor=<rgba 1 1 1 1>}
     uVRegion {bIsValid=false, max=<0 0>, min=<0 0>}
    }
    colorAndOpacity=<rgba 0.0232 0.0273 0.0908 1>
   }
   slot {class=OverlaySlot, properties {horizontalAlignment=HAlign_Right, verticalAlignment=VAlign_Center, padding=<ltrb 0 0 19 0>}}
  }
  Panel_DebugText {
   widgetClassPath=&Overlay
   bInherited=true
   parent=Overlay_2
   widget=Panel_DebugText
   properties {visibility=Collapsed}
   slot {class=OverlaySlot, properties {horizontalAlignment=HAlign_Fill, verticalAlignment=VAlign_Fill, padding=<ltrb 14 0 14 0>}}
  }
  DebugText_Data {
   widgetClassPath=&CommonTextBlock
   bInherited=true
   parent=Panel_DebugText
   widget=DebugText_Data
   properties {
    bAutoCollapseWithEmptyText=true
    justification=Right
    text="[Debug Data]"
    visibility=Collapsed
    colorAndOpacity {colorUseRule=UseColor_Specified, specifiedColor=<rgba 1 0 0 1>}
    font {
     bForceMonospaced=false
     bMaterialIsStencil=false
     fontMaterial=null
     fontObject=&BurbankSmall
     letterSpacing=0
     monospacedWidth=1
     size=12
     skewAmount=0
     typefaceFontName=Bold
     outlineSettings {
      bApplyOutlineToDropShadows=false
      bMiteredCorners=false
      bSeparateFillAlpha=false
      outlineMaterial=null
      outlineSize=1
      outlineColor=<rgba 0 0 0 1>
     }
    }
   }
   slot {class=OverlaySlot, properties {horizontalAlignment=HAlign_Fill, verticalAlignment=VAlign_Bottom}}
  }
  DebugText_ClassName {
   widgetClassPath=&CommonTextBlock
   bInherited=true
   parent=Panel_DebugText
   widget=DebugText_ClassName
   properties {
    bAutoCollapseWithEmptyText=true
    scrollStyle=&TextScrollStyle-Base_C
    text="[Debug Class Name]"
    visibility=Collapsed
    colorAndOpacity {colorUseRule=UseColor_Specified, specifiedColor=<rgba 1 0 0 1>}
    font {
     bForceMonospaced=false
     bMaterialIsStencil=false
     fontMaterial=null
     fontObject=&BurbankSmall
     letterSpacing=0
     monospacedWidth=1
     size=12
     skewAmount=0
     typefaceFontName=Bold
     outlineSettings {
      bApplyOutlineToDropShadows=false
      bMiteredCorners=false
      bSeparateFillAlpha=false
      outlineMaterial=null
      outlineSize=1
      outlineColor=<rgba 0 0 0 1>
     }
    }
   }
   slot {class=OverlaySlot, properties {horizontalAlignment=HAlign_Fill, verticalAlignment=VAlign_Bottom}}
  }
  DebugText_OptionValueName {
   widgetClassPath=&CommonTextBlock
   bInherited=true
   parent=Panel_DebugText
   widget=DebugText_OptionValueName
   properties {
    bAutoCollapseWithEmptyText=true
    scrollStyle=&TextScrollStyle-Base_C
    text="[Debug Option Value Name]"
    visibility=Collapsed
    colorAndOpacity {colorUseRule=UseColor_Specified, specifiedColor=<rgba 1 0 0 1>}
    font {
     bForceMonospaced=false
     bMaterialIsStencil=false
     fontMaterial=null
     fontObject=&BurbankSmall
     letterSpacing=0
     monospacedWidth=1
     size=12
     skewAmount=0
     typefaceFontName=Bold
     outlineSettings {
      bApplyOutlineToDropShadows=false
      bMiteredCorners=false
      bSeparateFillAlpha=false
      outlineMaterial=null
      outlineSize=1
      outlineColor=<rgba 0 0 0 1>
     }
    }
   }
   slot {class=OverlaySlot}
  }
  DebugText_WidgetName {
   widgetClassPath=&CommonTextBlock
   bInherited=true
   parent=Panel_DebugText
   widget=DebugText_WidgetName
   properties {
    bAutoCollapseWithEmptyText=true
    scrollStyle=&TextScrollStyle-Base_C
    text="[Debug Widget Name]"
    visibility=Collapsed
    colorAndOpacity {colorUseRule=UseColor_Specified, specifiedColor=<rgba 1 0 0 1>}
    font {
     bForceMonospaced=false
     bMaterialIsStencil=false
     fontMaterial=null
     fontObject=&BurbankSmall
     letterSpacing=0
     monospacedWidth=1
     size=12
     skewAmount=0
     typefaceFontName=Bold
     outlineSettings {
      bApplyOutlineToDropShadows=false
      bMiteredCorners=false
      bSeparateFillAlpha=false
      outlineMaterial=null
      outlineSize=1
      outlineColor=<rgba 0 0 0 1>
     }
    }
   }
   slot {class=OverlaySlot, properties {horizontalAlignment=HAlign_Center, verticalAlignment=VAlign_Bottom}}
  }
  Panel_ContentBox {
   widgetClassPath=&Overlay
   bInherited=true
   parent=Switcher_Content
   widget=Panel_ContentBox
   slot {class=CommonVisibilitySwitcherSlot}
  }
  InputActionWidget {widgetClassPath=&CommonActionWidget, bInherited=true, widget=null}
 ]
}
