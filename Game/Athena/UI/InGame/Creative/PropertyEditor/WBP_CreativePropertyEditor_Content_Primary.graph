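# Primary content row widget for the Creative property editor, showing a modified icon, a channel-binding icon and rich text.
# A binding-state enum picks the binding icon material from the data asset map and the default or highlighted color for text and modified icon.
# Text color goes through a TextColor parameter on the rich text's dynamic font material.
# The binding icon hides in two states, the only ones where the modified icon can show.
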WidgetBlueprint /Game/Athena/UI/InGame/Creative/PropertyEditor/WBP_CreativePropertyEditor_Content_Primary
graphs [
 EventGraph {nodes [
   n1 = Event Update IsModified!() -> (OutputDelegate, then, bIsModified:bool=~) {is=title K2Node_CustomEvent}
   > Set bUpdatedIsModified(bUpdatedIsModified:bool=@n1.bIsModified) -> (Output_Get:bool)
   > UpdateIsModifiedVisibilityAlsoDependingState()
   n2 = Event Update States!() -> (OutputDelegate, then, State:E_UI_ChannelsBindingStates Enum=NewEnumerator7) {is=title K2Node_CustomEvent}
   > n3 = Set UpdatedState(UpdatedState:E_UI_ChannelsBindingStates Enum=@n2.State) -> (Output_Get:E_UI_ChannelsBindingStates Enum=NewEnumerator7)
   > UpdateBindingIcon!(execute, self, IconState:E_UI_ChannelsBindingStates Enum=@n3) -> (then)
   > UpdateTextState!(
    execute
    self
    TextColorState:E_UI_ChannelsBindingStates Enum=Get UpdatedState() -> (UpdatedState:E_UI_ChannelsBindingStates Enum=NewEnumerator7)
   ) -> (then)
   > Update Is Modified State!(execute, self, ModifiedIconState:E_UI_ChannelsBindingStates Enum=Get UpdatedState() -> (UpdatedState=NewEnumerator7)) -> (then)
   Event PreConstruct()
   > SetFontMaterial()
   > PreConstructWidgetSettings()
   UpdateText!(execute, self, Text:"Text (by ref)") -> (then)
   Event Construct()
   Event Tick()
  ]}
 GetColorsByState {nodes [
   GetColorsByState!() -> (then, ColorStates:E_UI_ChannelsBindingStates Enum=NewEnumerator7) {is=title K2Node_FunctionEntry}
   n1 = Get DefaultColor() -> (DefaultColor:Linear Color Structure)
   n2 = Get HighlightedColor() -> (HighlightedColor:Linear Color Structure)
   > Return Node!(
    execute
    ReturnValue:Linear Color Structure=Select!(
     NewEnumerator7:Linear Color Structure=@n1
     NewEnumerator8:Linear Color Structure=@n2
     NewEnumerator0:Linear Color Structure=@n1
     NewEnumerator1:Linear Color Structure=@n2
     NewEnumerator2:Linear Color Structure=@n1
     NewEnumerator3:Linear Color Structure=@n2
     Index:E_UI_ChannelsBindingStates Enum=Get ColorStates() -> (ColorStates:E_UI_ChannelsBindingStates Enum=NewEnumerator7)
    ) -> (ReturnValue:Linear Color Structure)
   )
  ]}
 PreConstructWidgetSettings {nodes [
   PreConstructWidgetSettings() {is=title K2Node_FunctionEntry}
   > UpdateBindingIcon(IconState=Get States() -> (States:E_UI_ChannelsBindingStates Enum=NewEnumerator7))
   > UpdateTextState(TextColorState=Get States() -> (States=NewEnumerator7))
   > Update Is Modified State(ModifiedIconState=Get States() -> (States=NewEnumerator7))
   > Event Update IsModified!(execute, self, bIsModified:bool=Get bIsModified() -> (bIsModified:bool)) -> (then) {is=title K2Node_CallFunction}
   > UpdateText(Text=Get RichText() -> (RichText:Text))
  ]}
 SetFontMaterial {nodes [
   SetFontMaterial() {is=title K2Node_FunctionEntry}
   > SetDefaultMaterial!(
    execute
    self:Rich Text Block Object Reference=Get RichText_Content() -> (RichText_Content:Common Rich Text Block Object Reference)
    InMaterial:Material Interface Object Reference=Get RichTextFontMaterial() -> (RichTextFontMaterial:Material Interface Object Reference)
   ) -> (then)
  ]}
 Update Is Modified State {nodes [
   n1 = Update Is Modified State!() -> (then, ModifiedIconState:E_UI_ChannelsBindingStates Enum=NewEnumerator7) {is=title K2Node_FunctionEntry}
   > n2 = GetColorsByState!(execute, self, ColorStates:E_UI_ChannelsBindingStates Enum=@n1) -> (then, ReturnValue:Linear Color Structure) {is=title K2Node_CallFunction}
   > SetColorAndOpacity(
    self:Image Object Reference=Get Image_isModifiedIcon() -> (Image_isModifiedIcon:Image Object Reference)
    InColorAndOpacity:Linear Color Structure=@n2
   )
   > UpdateIsModifiedVisibilityAlsoDependingState()
  ]}
 UpdateBindingIcon {nodes [
   UpdateBindingIcon!() -> (then, IconState:E_UI_ChannelsBindingStates Enum=NewEnumerator7) {is=title K2Node_FunctionEntry}
   n1 = Get PrimaryDataAsset() -> (PrimaryDataAsset:UI PDA Creative Property Editor Object Reference)
   > SetBrushFromMaterial(
    self=Get Image_BindingIcon() -> (Image_BindingIcon:Image Object Reference)
    Material=Find(
     TargetMap:Map of E_UI_ChannelsBindingStates Enums to Material Instance Object References=Get ChannelsBindingMaterials_Map!(self:UI PDA Creative Property Editor Object Reference=@n1) -> (ChannelsBindingMaterials_Map:Map of E_UI_ChannelsBindingStates Enums to Material Instance Object References=NewEnumerator7)
     Key:"E_UI_ChannelsBindingStates Enum (by ref)"=Get IconState() -> (IconState:E_UI_ChannelsBindingStates Enum=NewEnumerator7)
    ) -> (Value:Material Instance Object Reference)
   )
   > UpdateBindingIconVisibility!(execute, self, IconState:E_UI_ChannelsBindingStates Enum=Get IconState() -> (IconState=NewEnumerator7)) -> (then)
  ]}
 UpdateBindingIconVisibility {nodes [
   UpdateBindingIconVisibility!() -> (then, IconState:E_UI_ChannelsBindingStates Enum=NewEnumerator7) {is=title K2Node_FunctionEntry}
   > SetVisibility(
    self=Get Image_BindingIcon()
    InVisibility=Select(
     NewEnumerator7:ESlateVisibility Enum=Collapsed
     NewEnumerator8:ESlateVisibility Enum=Collapsed
     NewEnumerator0:ESlateVisibility Enum=SelfHitTestInvisible
     NewEnumerator1:ESlateVisibility Enum=SelfHitTestInvisible
     NewEnumerator2:ESlateVisibility Enum=SelfHitTestInvisible
     NewEnumerator3:ESlateVisibility Enum=SelfHitTestInvisible
     Index=Get IconState() -> (IconState=NewEnumerator7)
    ) -> (ReturnValue:ESlateVisibility Enum)
   )
  ]}
 UpdateIsModifiedVisibilityAlsoDependingState {nodes [
   UpdateIsModifiedVisibilityAlsoDependingState() {is=title K2Node_FunctionEntry}
   n1 = Select!(
    Option 0:ESlateVisibility Enum=Collapsed
    Option 1:ESlateVisibility Enum=SelfHitTestInvisible
    Index:bool=Get bUpdatedIsModified() -> (bUpdatedIsModified:bool)
   ) -> (ReturnValue)
   > Set Visibility(
    Visibility=Select!(
     NewEnumerator7:ESlateVisibility Enum=@n1
     NewEnumerator8:ESlateVisibility Enum=@n1
     NewEnumerator0:ESlateVisibility Enum=Collapsed
     NewEnumerator1:ESlateVisibility Enum=Collapsed
     NewEnumerator2:ESlateVisibility Enum=Collapsed
     NewEnumerator3:ESlateVisibility Enum=Collapsed
     Index:E_UI_ChannelsBindingStates Enum=Get UpdatedState() -> (UpdatedState=NewEnumerator7)
    ) -> (ReturnValue)
    self=Get Image_isModifiedIcon()
   )
  ]}
 UpdateText {nodes [
   n1 = UpdateText!() -> (then, Text:"Text (by ref)") {is=title K2Node_FunctionEntry}
   > Set RichText(RichText:Text=@n1) -> (Output_Get:Text)
   > SetText(self=Get RichText_Content(), InText=@n1)
  ]}
 UpdateTextState {nodes [
   n1 = UpdateTextState!() -> (then, TextColorState:E_UI_ChannelsBindingStates Enum=NewEnumerator7) {is=title K2Node_FunctionEntry}
   > n2 = GetColorsByState(ColorStates=@n1)
   > n3 = GetDefaultDynamicMaterial!(execute, self:Rich Text Block Object Reference=Get RichText_Content()) -> (then, ReturnValue:Material Instance Dynamic Object Reference)
   > SetVectorParameterValue(self=@n3, ParameterName=TextColor, Value=@n2)
  ]}
]
variables [
 bIsModified {type=boolean}
 bUpdatedIsModified {type=boolean}
 DefaultColor {type=LinearColor, default=<rgba 0.3372 0.402 0.5906 1>}
 DisabledColor {type=LinearColor, default=<rgba 0.9682 0 1 1>}
 HighlightedColor {type=LinearColor, default=<rgba 0.0666 0.0908 0.1274 1>}
 PrimaryDataAsset {type=UI_PDA_Creative_PropertyEditor_C, default=/Game/Athena/UI/InGame/Creative/PropertyEditor/DataAssets/UI_DA_Creative_PropertyEditor}
 RichText {type=string, default="<Functions>Functions</> Rich Text"}
 RichTextFontMaterial {type=MaterialInterface, default=/Game/Athena/UI/InGame/Creative/PropertyEditor/Art/MI_UI_CreativePropertyEvent_FontColor_PrimaryContent}
 States {type=E_UI_ChannelsBindingStates, default=FunctionNormal}
 UpdatedState {type=E_UI_ChannelsBindingStates, default=Default}
]
widgets {
 info {inheritedWidgetCount=0, parentClass=&UserWidget, rootWidgetClass=&HorizontalBox, widgetCount=4}
 widgets [
  Root {widgetClassPath=&HorizontalBox, widget=Root}
  Image_isModifiedIcon {
   widgetClassPath=&Image
   bIsVariable=true
   parent=Root
   widget=Image_isModifiedIcon
   properties {brush {
     bIsDynamicallyLoaded=false
     drawAs=Image
     imageType=NoImage
     mirroring=NoMirror
     resourceName=None
     resourceObject=/Game/Athena/UI/InGame/Creative/PropertyEditor/Art/Icon-PropertyEditor-Modified
     tiling=NoTile
     imageSize=<20 20>
     margin=<ltrb 0 0 0 0>
     outlineSettings {
      bUseBrushTransparency=false
      roundingType=HalfHeightRadius
      width=0
      color {colorUseRule=UseColor_Specified, specifiedColor=<rgba 0 0 0 0>}
      cornerRadii=<0 0 0 0>
     }
     tintColor {colorUseRule=UseColor_Specified, specifiedColor=<rgba 1 1 1 1>}
     uVRegion {bIsValid=false, max=<0 0>, min=<0 0>}
    }}
   slot {class=HorizontalBoxSlot, properties {horizontalAlignment=HAlign_Left, verticalAlignment=VAlign_Center, padding=<ltrb -24 0 4 0>}}
  }
  Image_BindingIcon {
   widgetClassPath=&Image
   bIsVariable=true
   parent=Root
   widget=Image_BindingIcon
   properties {brush {
     bIsDynamicallyLoaded=false
     drawAs=Image
     imageType=NoImage
     mirroring=NoMirror
     resourceName=None
     resourceObject=/Game/Athena/UI/InGame/Creative/PropertyEditor/Art/MI_UI_Binding_FunctionNormal
     tiling=NoTile
     imageSize=<32 32>
     margin=<ltrb 0 0 0 0>
     outlineSettings {
      bUseBrushTransparency=false
      roundingType=HalfHeightRadius
      width=0
      color {colorUseRule=UseColor_Specified, specifiedColor=<rgba 0 0 0 0>}
      cornerRadii=<0 0 0 0>
     }
     tintColor {colorUseRule=UseColor_Specified, specifiedColor=<rgba 1 1 1 1>}
     uVRegion {bIsValid=false, max=<0 0>, min=<0 0>}
    }}
   slot {class=HorizontalBoxSlot, properties {horizontalAlignment=HAlign_Left, verticalAlignment=VAlign_Center, padding=<ltrb 0 0 8 0>}}
  }
  RichText_Content {
   widgetClassPath=&CommonRichTextBlock
   bIsVariable=true
   parent=Root
   widget=RichText_Content
   properties {
    bOverrideDefaultStyle=true
    scrollStyle=&TextScrollStyle-Base_C
    text="<functions>Functions</> or <Events>Events</> or Default Rich Text"
    textStyleSet=/Game/Athena/UI/InGame/Creative/PropertyEditor/DataAssets/UI_DT_Creative_PropertyEditor_Textstyle
    decoratorClasses [null]
    defaultTextStyleOverride {
     overflowPolicy=Clip
     transformPolicy=None
     colorAndOpacity {colorUseRule=UseColor_Specified, specifiedColor=<rgba 1 1 1 1>}
     font {
      bForceMonospaced=false
      bMaterialIsStencil=false
      fontMaterial=null
      fontObject=&BurbankSmall
      letterSpacing=0
      monospacedWidth=1
      size=18
      skewAmount=0
      typefaceFontName=Bold
      outlineSettings {
       bApplyOutlineToDropShadows=false
       bMiteredCorners=false
       bSeparateFillAlpha=false
       outlineMaterial=null
       outlineSize=0
       outlineColor=<rgba 0 0 0 1>
      }
     }
     highlightColor {colorUseRule=UseColor_Specified, specifiedColor=<rgba 0 0 0 1>}
     highlightShape {
      bIsDynamicallyLoaded=false
      drawAs=Image
      imageType=NoImage
      mirroring=NoMirror
      resourceName=None
      resourceObject=null
      tiling=NoTile
      imageSize=<32 32>
      margin=<ltrb 0 0 0 0>
      outlineSettings {
       bUseBrushTransparency=false
       roundingType=HalfHeightRadius
       width=0
       color {colorUseRule=UseColor_Specified, specifiedColor=<rgba 0 0 0 0>}
       cornerRadii=<0 0 0 0>
      }
      tintColor {colorUseRule=UseColor_Specified, specifiedColor=<rgba 1 1 1 1>}
      uVRegion {bIsValid=false, max=<0 0>, min=<0 0>}
     }
     selectedBackgroundColor {colorUseRule=UseColor_Foreground, specifiedColor=<rgba 1 0 1 1>}
     shadowColorAndOpacity=<rgba 0 0 0 0>
     shadowOffset=<0 0>
     strikeBrush {
      bIsDynamicallyLoaded=false
      drawAs=Image
      imageType=NoImage
      mirroring=NoMirror
      resourceName=None
      resourceObject=null
      tiling=NoTile
      imageSize=<32 32>
      margin=<ltrb 0 0 0 0>
      outlineSettings {
       bUseBrushTransparency=false
       roundingType=HalfHeightRadius
       width=0
       color {colorUseRule=UseColor_Specified, specifiedColor=<rgba 0 0 0 0>}
       cornerRadii=<0 0 0 0>
      }
      tintColor {colorUseRule=UseColor_Specified, specifiedColor=<rgba 1 1 1 1>}
      uVRegion {bIsValid=false, max=<0 0>, min=<0 0>}
     }
     underlineBrush {
      bIsDynamicallyLoaded=false
      drawAs=Image
      imageType=NoImage
      mirroring=NoMirror
      resourceName=None
      resourceObject=null
      tiling=NoTile
      imageSize=<32 32>
      margin=<ltrb 0 0 0 0>
      outlineSettings {
       bUseBrushTransparency=false
       roundingType=HalfHeightRadius
       width=0
       color {colorUseRule=UseColor_Specified, specifiedColor=<rgba 0 0 0 0>}
       cornerRadii=<0 0 0 0>
      }
      tintColor {colorUseRule=UseColor_Specified, specifiedColor=<rgba 1 1 1 1>}
      uVRegion {bIsValid=false, max=<0 0>, min=<0 0>}
     }
    }
   }
   slot {class=HorizontalBoxSlot, properties {horizontalAlignment=HAlign_Left, verticalAlignment=VAlign_Center, size {sizeRule=Fill, value=1}}}
  }
 ]
}
